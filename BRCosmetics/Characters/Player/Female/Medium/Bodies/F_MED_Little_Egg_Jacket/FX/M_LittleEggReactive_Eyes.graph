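# Unlit, two-sided additive Niagara mesh-particle material for the eye effect of the reactive Little Egg Jacket cosmetic.
# A streak texture panned by particle relative time modulates particle color brightness.
# The HueShift dynamic parameter, weighted by a vertical UV gradient, shifts that color's hue before the shared emissive adjustment function.
# Particle alpha, the same gradient, inverted Fresnel and vertex-color red multiply into a mask that softens the edges.
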Material /BRCosmetics/Characters/Player/Female/Medium/Bodies/F_MED_Little_Egg_Jacket/FX/M_LittleEggReactive_Eyes
graph {
 properties {
  bUsedWithNiagaraMeshParticles=true
  blendMode=BLEND_Additive
  pixelDepthOffsetMode=PDOM_Legacy
  shadingModel=MSM_Unlit
  twoSided=true
  usedShadingModels=Unlit
 }
 nodes [
  n1 = TextureCoordinate()
  n2 = Power(Base=OneMinus(None=ComponentMask(None=@n1) {g=true}))
  n3 = DynamicParameter() {paramNames [HueShift,Param2,Param3,Param4]}
  n4 = ParticleRelativeTime()
  n5 = LinearInterpolate(Alpha=TextureSample(UVs=Panner(Time=@n4) {speedX=-1.5}) {texture=/Game/Effects/Fort_Effects/Textures/Tile/MotionBlur}.R) {constA=0.25, constB=1.5}
  n6 = ParticleColor()
  MaterialFunctionCall(Emissive=MaterialFunctionCall(Hue Shift Percentage=Multiply(A=@n2, B=@n3.HueShift), Texture=Multiply(A=@n5, B=@n6.RGB)) {materialFunction=&HueShift}.Result) {materialFunction=&MF_FX_EmissiveAdjustments}
  n7 = OneMinus(None=Fresnel() {baseReflectFraction=0, exponent=1})
  Multiply(A=Multiply(A=Multiply(A=@n6.A, B=@n2), B=Power(Base=@n7)), B=VertexColor().R)
 ]
}
