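# Niagara overlay effect for the female medium Little Egg Jacket cosmetic, playing a sparkle burst and glow scaled by reactive level.
# User.ReactiveIndex selects the GPU sprite emitter's spawn rate and maximum box height; the box grows as spawning tapers.
# Sparks take gradient colors from height relative to the owner, then flicker, drift and inherit source movement.
# A hue-cycling glow follows the pelvis, with front-end settings adjusting brightness, spawn rate and placement.
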NiagaraSystem /BRCosmetics/Characters/Player/Female/Medium/Bodies/F_MED_Little_Egg_Jacket/FX/NS_F_Med_LittleEgg_Jacket_ReactiveOverlay
dependencies {
 usedDataInterfaces [&NiagaraDataInterfaceArrayFloat, &NiagaraDataInterfaceCurve, &NiagaraDataInterfaceColorCurve]
 usedDynamicInputs [
  &self:ScratchDynamicInput
  &Abs_Float
 ]
 usedModules [
  &self:SystemSpawnScript.SystemScriptSource.SystemScriptGraph.NiagaraNodeAssignment_1.SetVariables_#1
  &SampleSkeletalMesh
  &self:SystemSpawnScript.SystemScriptSource.SystemScriptGraph.NiagaraNodeAssignment_0.SetVariables_#2
  &self:SystemSpawnScript.SystemScriptSource.SystemScriptGraph.NiagaraNodeAssignment_2.SetVariables_#3
  &self:SystemSpawnScript.SystemScriptSource.SystemScriptGraph.NiagaraNodeAssignment_3.SetVariables_#4
  &SystemState
 ]
}
emitters {
 Glow {
  emitterUpdateScript [
   &EmitterState(Life Cycle Mode=::NewEnumerator1, Loop Behavior=::NewEnumerator1, Loop Duration Mode=::NewEnumerator1, UseLoopDelay=false)
   &SpawnBurst_Instantaneous(Use Spawn Probability=false, UseLoopCountLimit=false)
  ]
  particleSpawnScript [&InitializeParticle(
    Color=<rgba 0.01 0 0 1>
    Lifetime=1.5
    Sprite Size Mode=::NewEnumerator3
    Uniform Sprite Size=120
    UsePositionOffset=false
    Write Material Random=true
   )]
  particleUpdateScript [
   &ParticleState()
   /SetVariables(Particles.Position=&ConvertVectorToPosition(
     Input Position=&Add_Vector(A=$System.PelvisSrc, B=<0 -7.5 -25>)
     Vector Is In Absolute Large World Space:NiagaraBool=true
    )) {isSetParameters=true}
   &CameraOffset(Camera Offset Amount=-40)
   &Color(Color=&HueShiftLinearColor(
     Hue=&Multiply_Float(A=$Particles.NormalizedAge, B=-1.25)
     Saturation=0.5
     Value=&MakeCustomFloatFromBool(Bool=$User.bIsFrontEnd, True Float=10, False Float=5)
     Linear Color=<rgba 0 0.01 0.1 1>
    ))
   &ScaleColor(
    Color Value To Scale=$Particles.Color
    Scale RGB=&VectorFromFloat(Value=&FloatFromCurve(FloatCurve=DataInterface() {propertyValues {Curve {keys [{arriveTangent=-0.25, interpMode=RCIM_Cubic, time=0.1, value=0.875}, <at 1 0 RCIM_Cubic>]}}}))
    ScaleA=true
    ScaleRGB=true
   )
   &ScaleSpriteSize(Uniform Curve Sprite Scale=DataInterface() {propertyValues {Curve {keys [{arriveTangent=3.948, interpMode=RCIM_Cubic, leaveTangent=3.948, tangentMode=RCTM_User, time=0, value=0}, <at 1 1 RCIM_Cubic>]}}})
  ]
  rendererClasses [&NiagaraSpriteRendererProperties]
  renderers [
   {rendererClass=&NiagaraSpriteRendererProperties}
   {rendererClass=&NiagaraSpriteRendererProperties, rendererIndex=1}
  ]
 }
 SimpleSpriteBurst {
  simTarget=GPUComputeSim
  emitterSpawnScript [/SetVariables(Emitter.MaxBoxSpawnHeight:NiagaraFloat=&SelectFloatFromArray(Direct Array Index=$User.ReactiveIndex, Float Selection Array=DataInterface() {propertyValues {FloatData [15,30,50,70,70]}})) {isSetParameters=true}]
  emitterUpdateScript [
   &EmitterState(Life Cycle Mode=::NewEnumerator1, Loop Behavior=::NewEnumerator1, Loop Duration=0.5, UseLoopDelay=false)
   &SpawnRate(
    SpawnRate=&Multiply_Float(
     A=&FloatFromCurve(
      FloatCurve=DataInterface() {propertyValues {Curve {keys [<at 0 1>,<at 0.2 0.35>,<at 1 0>]}}}
      CurveIndex=$Emitter.NormalizedLoopAge
      Scale Curve=&FN_FrontEndFloatModifier(Gameplay=3.5, Front End=2)
     )
     B=&SelectFloatFromArray(Direct Array Index=$User.ReactiveIndex, Float Selection Array=DataInterface() {propertyValues {FloatData [10,20,30,40,50]}})
    )
    Use Spawn Probability=false
   )
   /SetVariables(Emitter.CurrBoxHeight:NiagaraFloat=&FloatFromCurve(
     FloatCurve=DataInterface() {propertyValues {Curve {keys [
         {arriveTangent=3.5766, interpMode=RCIM_Cubic, leaveTangent=3.5766, tangentMode=RCTM_User, time=0, value=0}
         {interpMode=RCIM_Cubic, leaveTangent=1.9257, time=1, value=1}
        ]}}}
     CurveIndex=$Emitter.NormalizedLoopAge
     Scale Curve=$Emitter.MaxBoxSpawnHeight
    )) {isSetParameters=true}
   &SpawnRate(SpawnRate=250, Use Spawn Probability=false) {enabled=false, module=SpawnRate001}
  ]
  particleSpawnScript [
   &InitializeParticle(
    Lifetime=&RandomRangeFloat(Minimum=0.35, Maximum=1.5)
    Position Offset=<0 -30 100>
    Position Offset Coordinate Space=::Local
    Sprite Rotation Mode=::NewEnumerator2
    Sprite Size Mode=::NewEnumerator3
    Uniform Sprite Size=&RandomRangeFloat(Minimum=4, Maximum=6)
    UsePositionOffset=true
    Write Material Random=true
   )
   &AddVelocity(
    Cone Axis=<0 -1 0>
    Module.UseConeFalloff=false
    Module.UseVelDistribution=false
    Override Seed=false
    Velocity Mode=::NewEnumerator2
    Velocity Speed=&RandomRangeFloat(Minimum=5, Maximum=50)
   )
   &AddVelocity(Rotation Coordinate Space=::World, Velocity=&RandomRangeVector(Minimum=<0 0 0>, Maximum=<1 1 -20>)) {module=AddVelocity001}
   &BoxLocation(
    Box Offset=&MakeVector(
     X=&FN_FrontEndFloatModifier(Gameplay=0, Front End=-15)
     Y=&FN_FrontEndFloatModifier(Gameplay=0, Front End=-10)
     Z=&Subtract_Float(A=10, B=&Multiply_Float(A=0.5, B=$Emitter.MaxBoxSpawnHeight))
    )
    Box Size=&MakeVector(X=50, Y=&FN_FrontEndFloatModifier(Gameplay=50, Front End=5), Z=$Emitter.CurrBoxHeight)
    MaskSpawnGroup=false
    Module.Override Seed=false
    Rotate Box=false
    Surface Only=false
   )
   /SetVariables(Particles.Color=&ColorFromCurve(
     ColorCurve=DataInterface() {propertyValues {
       AlphaCurve {keys [<at 0 1 RCIM_Cubic>]}
       BlueCurve {keys [<at 0 0.6094 RCIM_Cubic>,<at 0.2096 0.7448>,<at 0.448 0.0134>,<at 0.6961 0>,<at 0.8729 0>]}
       GreenCurve {keys [<at 0 0.5035 RCIM_Cubic>,<at 0.2096 0>,<at 0.448 0>,<at 0.6961 0.4162>,<at 0.8729 0.875>]}
       RedCurve {keys [<at 0 0 RCIM_Cubic>,<at 0.2096 0.169>,<at 0.448 0.7448>,<at 0.6961 0.9323>,<at 0.8729 0.9323>]}
      }}
     CurveIndex=&Subtract_Float(
      A=1
      B=&Divide_Float(
       A=&MakeFloatFromVector(
        Vector=&Subtract_Vector(
         A=&ConvertPositionToVector(Input Position=$Particles.Position, Convert To Absolute Large World Vector=false)
         B=&ConvertPositionToVector(
          Input Position=&ConvertVectorToPosition(
           Input Position=&Add_Vector(
            A=&ConvertPositionToVector(Input Position=$Engine.Owner.Position, Convert To Absolute Large World Vector=false)
            B=<0 0 40>
           )
           Vector Is In Absolute Large World Space:NiagaraBool=false
          )
          Convert To Absolute Large World Vector=false
         )
        )
        Channel=::NewEnumerator2
       )
       B=64
      )
     )
    )) {isSetParameters=true}
   /SetVariables(Particles.Color=&HueShiftLinearColor(
     Hue=&FloatFromCurve(
      FloatCurve=DataInterface() {propertyValues {Curve {keys [<at 0 -0.25 RCIM_Cubic>,<at 0.4 0>,<at 1 0 RCIM_Cubic>]}}}
      CurveIndex=$Emitter.NormalizedLoopAge
     )
     Saturation=1
     Value=1
     Linear Color=$Particles.Color
    )) {isSetParameters=true}
   /Niagara/Modules/Math/PlaceParticlesOnDepthBuffer_GPU(Max Z Depth=100, Module.Use Max Z Depth=true) {enabled=false}
   &AddVelocity(
    Module.Use Falloff Distance=false
    Override Seed=false
    Velocity Mode=::NewEnumerator1
    Velocity Origin=&ConvertPositionToVector(Input Position=$Particles.Position, Convert To Absolute Large World Vector=false)
    Velocity Speed=&RandomRangeFloat(Minimum=0, Maximum=20)
   ) {module=AddVelocity002}
  ]
  particleUpdateScript [
   &ParticleState()
   &ScaleColor(
    Scale Alpha=&FloatFromCurve(FloatCurve=DataInterface() {propertyValues {Curve {keys [<at -0.0046 1.0113 RCIM_Cubic>,<at 1.0011 0.002 RCIM_Cubic>]}}}, Scale Curve=5)
    Scale RGB=&VectorFromFloat(Value=&FloatFromCurve(
      FloatCurve=DataInterface() {propertyValues {Curve {keys [<at 0 1>,<at 0.1 1.3>,<at 0.2 0.45>,<at 0.3 1.25>,<at 0.6 0.375>,<at 0.8 0.975>,<at 1 0>]}}}
      Scale Curve=8
     ))
    ScaleA=true
    ScaleRGB=true
   )
   &AccelerationForce(Acceleration=&RandomRangeVector(Minimum=<0 0 -10>, Maximum=<0 0 20>))
   &Drag(Drag=&RandomRangeFloat(Minimum=1, Maximum=2), Use Linear Drag=true, Use Rotational Drag=false)
   &InheritSourceMovement(
    Applied Force Scale=&VectorFromFloat(Value=&FloatFromCurve(FloatCurve=DataInterface() {propertyValues {Curve {keys [<at 0 1 RCIM_Cubic>, {arriveTangent=-1, interpMode=RCIM_Cubic, leaveTangent=-1, time=0.2, value=0.15}, <at 1 0 RCIM_Cubic>]}}}))
    Applied Position Scale=&VectorFromFloat(Value=&FloatFromCurve(FloatCurve=DataInterface() {propertyValues {Curve {keys [<at 0 1 RCIM_Cubic>, {arriveTangent=-1, interpMode=RCIM_Cubic, leaveTangent=-1, time=0.1, value=0.3}, <at 1 0 RCIM_Cubic>]}}}))
    ApplyForce=true
    ApplyForceFalloff=false
    ApplyPosition=true
    ApplyPositionFalloff=false
    Module.LimitForceSpeed=false
   )
   &SolveForcesAndVelocity(Clamp Velocity=false, Limit Acceleration=false)
   &CameraOffset(Camera Offset Amount=35)
   &ScaleSpriteSize(Uniform Curve Sprite Scale=DataInterface() {propertyValues {Curve {keys [
        <at 0 0.975 RCIM_Cubic>
        {arriveTangent=-1.1875, interpMode=RCIM_Cubic, leaveTangent=-1.1875, time=0.1, value=1.225}
        {arriveTangent=-0.3929, interpMode=RCIM_Cubic, leaveTangent=-0.3929, time=0.4, value=0.5}
        {arriveTangent=-0.8333, interpMode=RCIM_Cubic, leaveTangent=-0.8333, time=0.8, value=0.95}
        <at 1 0 RCIM_Cubic>
       ]}}})
   &SpriteRotationRate(Rotation Rate=&RandomRangeFloat(Minimum=-300, Maximum=300))
  ]
  rendererClasses [&NiagaraSpriteRendererProperties]
  renderers [{rendererClass=&NiagaraSpriteRendererProperties}]
 }
}
summary {
 systemName=NS_F_Med_LittleEgg_Jacket_ReactiveOverlay
 emitters [
  {emitterName=SimpleSpriteBurst, simTarget=GPUComputeSim, rendererClasses [&NiagaraSpriteRendererProperties]}
  {emitterName=Glow, rendererClasses [&NiagaraSpriteRendererProperties]}
 ]
 userVariables [
  User.bIsFrontEnd {type=NiagaraBool}
  User.bIsFrontEndPreview {type=NiagaraBool}
  User.ReactiveIndex {type=NiagaraInt32, defaultValue=3}
  User.SkeletalMesh {
   type=NiagaraDataInterfaceSkeletalMesh
   defaultValue {
    struct=NiagaraExt_VariableValue_DataInterface
    value {dataInterface=&self:NiagaraDataInterfaceSkeletalMesh_0}
   }
  }
 ]
}
