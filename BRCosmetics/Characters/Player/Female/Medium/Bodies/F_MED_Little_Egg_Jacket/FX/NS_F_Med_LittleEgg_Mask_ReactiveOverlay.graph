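# Niagara overlay effect for the Little Egg female cosmetic's mask, flashing glows and sparks at the eyes.
# EyeGlows (sprite and mesh) and two offset EyeGlowsSprites fade over life, alongside a hue-shifting central Glow.
# SimpleSpriteBurst emits hue-shifting sprites from both eye offsets via a decaying spawn rate, inheriting source movement.
# Every emitter sets VisibilityTag from User.bisHeadHidden; User.ReactiveIndex is declared but unread here.
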NiagaraSystem /BRCosmetics/Characters/Player/Female/Medium/Bodies/F_MED_Little_Egg_Jacket/FX/NS_F_Med_LittleEgg_Mask_ReactiveOverlay
dependencies {
 usedDataInterfaces [&NiagaraDataInterfaceCurve, &NiagaraDataInterfaceArrayFloat3, &NiagaraDataInterfaceColorCurve]
 usedDynamicInputs []
 usedModules [&SystemState]
}
emitters {
 EyeGlows {
  emitterUpdateScript [
   &EmitterState()
   &SpawnBurst_Instantaneous(Use Spawn Probability=false, UseLoopCountLimit=false)
  ]
  particleSpawnScript [
   &InitializeParticle(Color=<rgba 5 0.2083 0.2519 1>, Position Offset=<0 -1 0>, UsePositionOffset=true, Write Material Random=true)
   /SetVariables(Particles.MeshOrientation=&MakeQuatFromAxisAngle(Axis=<0 1 0>, AngleInDegrees=90)) {isSetParameters=true}
  ]
  particleUpdateScript [
   &ParticleState()
   &ScaleColor(
    Scale Alpha=&FloatFromCurve(FloatCurve=DataInterface() {propertyValues {Curve {keys [
         <at 0 0 RCIM_Cubic>
         {arriveTangent=1.25, interpMode=RCIM_Cubic, leaveTangent=1.25, time=0.1, value=1.25}
         {arriveTangent=-1.3889, interpMode=RCIM_Cubic, leaveTangent=-1.3889, time=0.7, value=0.875}
         <at 1 0 RCIM_Cubic>
        ]}}})
    ScaleA=true
    ScaleRGB=true
   )
   &DynamicMaterialParameters(
    Index 0 Param 1=&FloatFromCurve(
     FloatCurve=DataInterface() {propertyValues {Curve {keys [{interpMode=RCIM_Cubic, tangentMode=RCTM_User, time=0, value=0}, {interpMode=RCIM_Cubic, tangentMode=RCTM_User, time=1, value=1}]}}}
     Scale Curve=0.05
    )
    Module.Param0WriteEnabled=true
    Module.Param1WriteEnabled=true
    Module.Param2WriteEnabled=true
    Module.Param3WriteEnabled=true
   )
   /SetVariables(Particles.VisibilityTag=&MakeCustomIntFromBool(Bool=$User.bisHeadHidden, True INT=1, False INT=0)) {isSetParameters=true, module=SetVariables_#1}
  ]
  rendererClasses [&NiagaraSpriteRendererProperties, &NiagaraMeshRendererProperties]
  renderers [
   {rendererClass=&NiagaraSpriteRendererProperties}
   {rendererClass=&NiagaraMeshRendererProperties, rendererIndex=1}
  ]
 }
 EyeGlowsSprites {
  emitterUpdateScript [
   &EmitterState()
   &SpawnBurst_Instantaneous(Spawn Count=2, Use Spawn Probability=false, UseLoopCountLimit=false)
  ]
  particleSpawnScript [
   &FN_SetSpawnGroup(Number of Spawn Groups=2)
   &InitializeParticle(
    Color=<rgba 1.5 0.0733 0.0464 1>
    Position Offset=&SelectVectorFromArray(
     Direct Array Index=&SpawnGroup()
     Vector Selection Array=DataInterface() {propertyValues {FloatData [<2 7.5 3>, <2 7.5 -3>]}}
    )
    Sprite Size=<50 12>
    Sprite Size Mode=::NewEnumerator4
    UsePositionOffset=true
    Write Material Random=true
   )
  ]
  particleUpdateScript [
   &ParticleState()
   &ScaleColor(
    Scale Alpha=&FloatFromCurve(FloatCurve=DataInterface() {propertyValues {Curve {keys [
         <at 0 0 RCIM_Cubic>
         {arriveTangent=1.25, interpMode=RCIM_Cubic, leaveTangent=1.25, time=0.1, value=1.25}
         {arriveTangent=-1.3889, interpMode=RCIM_Cubic, leaveTangent=-1.3889, time=0.7, value=0.875}
         <at 1 0 RCIM_Cubic>
        ]}}})
    ScaleA=true
    ScaleRGB=true
   )
   &CameraOffset(Camera Offset Amount=5)
   /SetVariables(Particles.VisibilityTag=&MakeCustomIntFromBool(Bool=$User.bisHeadHidden, True INT=1, False INT=0)) {isSetParameters=true, module=SetVariables_#1}
  ]
  rendererClasses [&NiagaraSpriteRendererProperties]
  renderers [{rendererClass=&NiagaraSpriteRendererProperties}]
 }
 Glow {
  emitterUpdateScript [
   &EmitterState(Life Cycle Mode=::NewEnumerator1, Loop Behavior=::NewEnumerator1, Loop Duration Mode=::NewEnumerator1, UseLoopDelay=false)
   &SpawnBurst_Instantaneous(Use Spawn Probability=false, UseLoopCountLimit=false)
  ]
  particleSpawnScript [&InitializeParticle(Color=<rgba 0.01 0 0 1>, Sprite Size Mode=::NewEnumerator3, Uniform Sprite Size=50, UsePositionOffset=false, Write Material Random=true)]
  particleUpdateScript [
   &ParticleState()
   /SetVariables(Particles.Position=&ConvertVectorToPosition(Input Position=<5 10 0>, Vector Is In Absolute Large World Space:NiagaraBool=false)) {isSetParameters=true}
   &CameraOffset()
   &Color(Color=&HueShiftLinearColor(
     Hue=&Multiply_Float(A=$Particles.NormalizedAge, B=-1.25)
     Saturation=0.75
     Value=2
     Linear Color=<rgba 0 0.01 0.1 1>
    ))
   &ScaleColor(
    Color Value To Scale=$Particles.Color
    Scale RGB=&VectorFromFloat(Value=&FloatFromCurve(FloatCurve=DataInterface() {propertyValues {Curve {keys [<at 0 1 RCIM_Cubic>, {arriveTangent=-1, interpMode=RCIM_Cubic, leaveTangent=-1, time=0.2, value=0.275}, <at 1 0 RCIM_Cubic>]}}}))
    ScaleA=true
    ScaleRGB=true
   )
   &ScaleSpriteSize(Uniform Curve Sprite Scale=DataInterface() {propertyValues {Curve {keys [<at 0 0 RCIM_Cubic>, {arriveTangent=1, interpMode=RCIM_Cubic, leaveTangent=1, time=0.3, value=0.9}, <at 1 1 RCIM_Cubic>]}}})
   /SetVariables(Particles.VisibilityTag=&MakeCustomIntFromBool(Bool=$User.bisHeadHidden, True INT=1, False INT=0)) {isSetParameters=true, module=SetVariables_#1}
  ]
  rendererClasses [&NiagaraSpriteRendererProperties]
  renderers [
   {rendererClass=&NiagaraSpriteRendererProperties}
   {rendererClass=&NiagaraSpriteRendererProperties, rendererIndex=1}
  ]
 }
 SimpleSpriteBurst {
  emitterUpdateScript [
   &EmitterState(Life Cycle Mode=::NewEnumerator1, Loop Behavior=::NewEnumerator1, Loop Duration=0.5, UseLoopDelay=false)
   &SpawnRate(
    SpawnRate=&Multiply_Float(
     A=&FloatFromCurve(
      FloatCurve=DataInterface() {propertyValues {Curve {keys [<at 0 1 RCIM_Cubic>, {arriveTangent=-1, interpMode=RCIM_Cubic, leaveTangent=-1, time=0.1, value=1.5}, <at 1 0 RCIM_Cubic>]}}}
      CurveIndex=$Emitter.NormalizedLoopAge
      Scale Curve=14
     )
     B=1
    )
    Use Spawn Probability=false
   )
   &SpawnRate(SpawnRate=50, Use Spawn Probability=false) {enabled=false, module=SpawnRate001}
   &SpawnBurst_Instantaneous(Spawn Count=566, Use Spawn Probability=false, UseLoopCountLimit=false) {enabled=false}
  ]
  particleSpawnScript [
   &InitializeParticle(
    Color=<rgba 1 0 0 1>
    Lifetime=&RandomRangeFloat(Minimum=0.35, Maximum=1.35)
    Sprite Rotation Mode=::NewEnumerator2
    Sprite Size Mode=::NewEnumerator3
    Uniform Sprite Size=&RandomRangeFloat(Minimum=6, Maximum=12)
    UsePositionOffset=false
    Write Material Random=true
   )
   &AddVelocity(
    Cone Angle=90
    Cone Axis=<0 1 0>
    Module.UseConeFalloff=false
    Module.UseVelDistribution=false
    Override Seed=false
    Velocity Mode=::NewEnumerator2
    Velocity Speed=&RandomRangeFloat(Minimum=1, Maximum=10)
   )
   &SphereLocation(
    Hemisphere Y=true
    MaskSpawnGroup=false
    Non Uniform Scale=<0.61 0.75 1>
    Offset=&SelectVectorFromArray(
     Array Sampling Mode=::NewEnumerator0
     Direct Array Index=?
     Vector Selection Array=DataInterface() {propertyValues {FloatData [<2 7 3>, <2 7 -3>]}}
     Random Seed=0
    )
    Override Seed=false
    Sphere Orientation Axis:Vector3f=<0 0 1>
    Sphere Radius=1
    Surface Only=false
    UseNonUniformScale=true
    UseOffset=true
    UseSphereOrientation=true
   )
   /SetVariables(Particles.Color=&ColorFromCurve(
     ColorCurve=DataInterface() {propertyValues {
       AlphaCurve {keys [<at 0 1 RCIM_Cubic>]}
       BlueCurve {keys [<at 0 0.0556>,<at 0.2442 0>,<at 0.8009 0.0011>,<at 1 0.7448>]}
       GreenCurve {keys [<at 0 0>,<at 0.2442 0.4162>,<at 0.8009 0.7448>,<at 1 0>]}
       RedCurve {keys [<at 0 0.9323>,<at 0.2442 0.9323>,<at 0.8009 0.1234>,<at 1 0.169>]}
      }}
     CurveIndex=&RandomRangeFloat(Minimum=0, Maximum=0.1)
    )) {enabled=false, isSetParameters=true}
   /Niagara/Modules/Math/PlaceParticlesOnDepthBuffer_GPU(Max Z Depth=100, Module.Use Max Z Depth=true) {enabled=false}
   &AddVelocity(
    Module.Use Falloff Distance=false
    Origin Offset=<3 0 0>
    Override Seed=false
    Velocity Mode=::NewEnumerator1
    Velocity Origin=&ConvertPositionToVector(Input Position=$Engine.Owner.Position, Convert To Absolute Large World Vector=false)
    Velocity Speed=&RandomRangeFloat(Minimum=0, Maximum=30)
   ) {module=AddVelocity002}
  ]
  particleUpdateScript [
   &ParticleState()
   &Color(Color=&HueShiftLinearColor(
     Hue=&Multiply_Float(
      A=&FloatFromCurve(FloatCurve=DataInterface() {propertyValues {Curve {keys [
           {interpMode=RCIM_Cubic, tangentMode=RCTM_User, time=0, value=0}
           {arriveTangent=1, interpMode=RCIM_Cubic, leaveTangent=1, time=0.5, value=0}
           <at 1 1 RCIM_Cubic>
          ]}}})
      B=&Add_Float(A=&RandomRangeFloat(Minimum=0, Maximum=1), B=$Particles.Age)
     )
     Saturation=1
     Value=1
     Linear Color=$Particles.Initial.Color
    ))
   &ScaleColor(
    Color Value To Scale=$Particles.Color
    Scale Alpha=&FloatFromCurve(FloatCurve=DataInterface() {propertyValues {Curve {keys [<at -0.0046 1.0113 RCIM_Cubic>,<at 1.0011 0.002 RCIM_Cubic>]}}}, Scale Curve=5)
    Scale RGB=&VectorFromFloat(Value=&FloatFromCurve(
      FloatCurve=DataInterface() {propertyValues {Curve {keys [
          <at 0 1 RCIM_Cubic>
          {arriveTangent=-1.0417, interpMode=RCIM_Cubic, leaveTangent=-1.0417, time=0.3, value=1.25}
          {arriveTangent=-0.55, interpMode=RCIM_Cubic, leaveTangent=-0.55, time=0.6, value=0.375}
          {arriveTangent=-0.9375, interpMode=RCIM_Cubic, leaveTangent=-0.9375, time=0.8, value=0.975}
          <at 1 0 RCIM_Cubic>
         ]}}}
      Scale Curve=8
     ))
    ScaleA=true
    ScaleRGB=true
   )
   &AccelerationForce(Acceleration=&RandomRangeVector(Minimum=<0 0 0>, Maximum=<1 1 50>))
   &Drag(Drag=&RandomRangeFloat(Minimum=1, Maximum=2), Use Linear Drag=true, Use Rotational Drag=false)
   &InheritSourceMovement(
    Applied Force Scale=&VectorFromFloat(Value=&FloatFromCurve(FloatCurve=DataInterface() {propertyValues {Curve {keys [<at 0 1 RCIM_Cubic>, {arriveTangent=-1, interpMode=RCIM_Cubic, leaveTangent=-1, time=0.2, value=0.15}, <at 1 0 RCIM_Cubic>]}}}))
    Applied Position Scale=&VectorFromFloat(Value=&FloatFromCurve(FloatCurve=DataInterface() {propertyValues {Curve {keys [<at 0 1 RCIM_Cubic>,<at 1 0 RCIM_Cubic>]}}}))
    ApplyForce=true
    ApplyForceFalloff=false
    ApplyPosition=true
    ApplyPositionFalloff=false
    Module.LimitForceSpeed=false
   )
   &SolveForcesAndVelocity(Clamp Velocity=false, Limit Acceleration=false)
   &CameraOffset(Camera Offset Amount=5)
   &ScaleSpriteSize(Uniform Curve Sprite Scale=DataInterface() {propertyValues {Curve {keys [
        <at 0 0.975 RCIM_Cubic>
        {arriveTangent=-1.1875, interpMode=RCIM_Cubic, leaveTangent=-1.1875, time=0.1, value=1.225}
        {arriveTangent=-0.3929, interpMode=RCIM_Cubic, leaveTangent=-0.3929, time=0.4, value=0.5}
        {arriveTangent=-0.8333, interpMode=RCIM_Cubic, leaveTangent=-0.8333, time=0.8, value=0.95}
        <at 1 0 RCIM_Cubic>
       ]}}})
   &SpriteRotationRate(Rotation Rate=&RandomRangeFloat(Minimum=-50, Maximum=50))
   /SetVariables(Particles.VisibilityTag=&MakeCustomIntFromBool(Bool=$User.bisHeadHidden, True INT=1, False INT=0)) {isSetParameters=true, module=SetVariables_#1}
  ]
  rendererClasses [&NiagaraSpriteRendererProperties]
  renderers [{rendererClass=&NiagaraSpriteRendererProperties}]
 }
}
summary {
 systemName=NS_F_Med_LittleEgg_Mask_ReactiveOverlay
 emitters [
  {emitterName=SimpleSpriteBurst, rendererClasses [&NiagaraSpriteRendererProperties]}
  {emitterName=Glow, rendererClasses [&NiagaraSpriteRendererProperties]}
  {emitterName=EyeGlows, rendererClasses [&NiagaraSpriteRendererProperties, &NiagaraMeshRendererProperties]}
  {emitterName=EyeGlowsSprites, rendererClasses [&NiagaraSpriteRendererProperties]}
 ]
 userVariables [
  User.bisHeadHidden {type=NiagaraBool}
  User.ReactiveIndex {type=NiagaraInt32, defaultValue=3}
 ]
}
