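# Character master material for the female Little Egg Jacket skin, adding a reactive rainbow effect to the standard character FX function chain.
# The Reactive switch blends in mask-driven curve-atlas colors and an emissive glow built from panning noise, a timed pulse wave and Fresnel.
# A stepped reveal mask follows Reactive_Level, or a Blueprint-supplied start time under Use_BP_Time.
# The Pickaxe_HueShift switch instead cycles base color hue over time within the pickaxe's FX mask.
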Material /BRCosmetics/Characters/Player/Female/Medium/Bodies/F_MED_Little_Egg_Jacket/Materials/F_MED_LittleEgg_Jacket_Reactive_Master
graph {
 properties {
  bAutomaticallySetUsageInEditor=false
  bUseMaterialAttributes=true
  bUsedWithMeshDeformer=true
  bUsedWithMorphTargets=true
  bUsedWithSkeletalMesh=true
  pixelDepthOffsetMode=PDOM_Legacy
  refractionMethod=RM_IndexOfRefraction
  shadingModel=MSM_FromMaterialExpression
  subsurfaceProfile=&SS_HeroSkin_02
  usedShadingModels=Default Lit | Subsurface | Clear Coat | Subsurface Profile | Cloth | Eye
 }
 nodes [
  n1 = MaterialFunctionCall() {materialFunction=&BaseFN_CharacterFunction_PreFX}
  n2 = MaterialFunctionCall(
   PreskinnedPosition=@n1.Pre skinned Local Position
   Gmap=@n1.Pre skinned Local Position
   MaterialAttributes=@n1.Pre skinned Local Position
   AO=@n1.Pre skinned Local Position
  ) {materialFunction=&CharacterSkinFX}
  n3 = GetMaterialAttributes(None=@n2.Result) {attributeGetTypes [#1]}
  n4 = TextureSample() {
   samplerSource=SSM_Wrap_WorldGroupSettings
   samplerType=SAMPLERTYPE_Masks
   texture=/BRCosmetics/Weapons/FORT_Melee/Pickaxe_Little_Egg_Female/Skins/Textures/T_Pickaxe_Little_Egg_Dazzle_Female_FX
  }
  n5 = Time()
  n6 = ScalarParameter() {defaultValue=0.25, parameterName=HueShift_Speed}
  n7 = MaterialFunctionCall(Hue Shift Percentage=Multiply(A=Ceil(None=@n4.R), B=Multiply(A=@n5, B=@n6)), Texture=@n3.Base Color) {materialFunction=&HueShift}
  n8 = GetMaterialAttributes(None=@n2.Result) {attributeGetTypes [#1, #2]}
  n9 = Max(A=NamedRerouteUsage()) {constA=0.3, constB=0.15}
  n10 = NamedRerouteUsage()
  n11 = TextureSampleParameter2D() {
   group=LittleEgg
   parameterName=Rainbow_Mask
   samplerSource=SSM_Wrap_WorldGroupSettings
   samplerType=SAMPLERTYPE_Masks
   texture=/BRCosmetics/Characters/Player/Female/Medium/Bodies/F_MED_Little_Egg_Jacket/Textures/F_MED_Little_Egg_Body_Reactive_Mask
  }
  n12 = Multiply(A=MaterialFunctionCall(X=@n11.G) {materialFunction=&MakeFloat2}.Result) {constB=2}
  n13 = Multiply(A=TextureCoordinate(), B=ScalarParameter() {defaultValue=2, group=LittleEgg, parameterName=EnergyNoise_tile})
  n14 = TextureSample(UVs=Panner(Coordinate=Add(A=@n12, B=@n13)) {speedY=0.1}) {
   samplerSource=SSM_Wrap_WorldGroupSettings
   samplerType=SAMPLERTYPE_LinearColor
   texture=/Game/Global/Textures/Default/Noise/T_Curie_NoisePack
  }
  n15 = Multiply(A=Time() {bOverride_Period=true, period=6.5}, B=ScalarParameter() {defaultValue=2, group=LittleEgg, parameterName=Pulse_Speed}) {constB=0.25}
  n16 = Saturate(None=@n15)
  n17 = Divide(A=Add(A=OneMinus(None=@n16), B=Subtract(A=@n11.G) {constB=0.25})) {constB=3.14}
  n18 = Saturate(None=Subtract(A=Sine(None=@n17)) {constB=0.95})
  n19 = Constant() {r=0.5}
  n20 = Saturate(None=@n15)
  n21 = NamedRerouteDeclaration(None=LinearInterpolate(
    A=MaterialFunctionCall!(In=@n18, Ramp Width=@n19) {materialFunction=/Engine/Functions/Engine_MaterialFunctions01/Math/SmoothCeil}.Result
    Alpha=Floor(None=@n20)
   ) {constB=0}) {name=Mask_PulseWave, nodeColor=<rgba 0 1 0.3438 1>}
  n22 = Multiply(A=NamedRerouteDeclaration(None=@n14.R) {name=Mask_EnergyNoise, nodeColor=<rgba 1 0.5156 0 1>}, B=@n21)
  n23 = ScalarParameter() {defaultValue=0.3, group=LittleEgg, parameterName=EnergyNoise_Intensity}
  n24 = ScalarParameter() {defaultValue=0.4, group=LittleEgg, parameterName=Pulse_Intensity}
  n25 = Multiply(A=NamedRerouteUsage(), B=Add(A=Multiply(A=@n22, B=@n23), B=Multiply(A=@n21, B=@n24)))
  n26 = NamedRerouteUsage()
  n27 = Max(A=@n26) {constB=0.1}
  n28 = ScalarParameter() {defaultValue=1, group=LittleEgg, parameterName=Fresnel_Intensity}
  n29 = Multiply(A=Add(A=Multiply(A=@n25, B=@n27), B=Multiply(A=@n26, B=@n28)), B=NamedRerouteUsage())
  n30 = NamedRerouteUsage()
  n31 = SetMaterialAttributes!(
   MaterialAttributes=@n8.MaterialAttributes
   Base Color=LinearInterpolate(A=@n8.Base Color, B=Multiply(A=@n9, B=@n10), Alpha=NamedRerouteUsage())
   Emissive Color=Add(A=@n8.Emissive Color, B=Multiply(A=@n29, B=@n30))
  ) {attributeSetTypes [#1, #2]}
  n32 = MaterialFunctionCall(
   AO=@n1.AO
   Distance Blend=@n1.AO
   Fresnel=@n1.AO
   Gmap=@n1.AO
   Pre Skinned Local Position=@n1.AO
   Material=StaticSwitchParameter(
    True=SetMaterialAttributes!(MaterialAttributes=@n3.MaterialAttributes, Base Color=@n7.Result) {attributeSetTypes [#1]}
    False=StaticSwitchParameter(True=@n31, False=@n2.Result) {group=LittleEgg, parameterName=Reactive}
   ) {group=LittleEgg, parameterName=Pickaxe_HueShift}
  ) {materialFunction=&BaseFN_CharacterFunction_PostFX}
  MaterialFunctionCall(In=@n32.Result) {materialFunction=&CINE_MF_CharPostFX}
  n33 = Ceil(None=Subtract(A=@n11.G) {constB=0.05})
  n34 = Subtract(
   A=StaticSwitchParameter(True=OneMinus(None=@n11.G), False=@n11.G) {group=LittleEgg, parameterName=Use_BP_Time}
   B=StaticSwitchParameter(True=Constant() {r=0.8}, False=Constant() {r=0.4}) {group=LittleEgg, parameterName=Use_BP_Time}
  ) {constB=0.4}
  n35 = Divide(
   A=Subtract(A=Time(), B=ScalarParameter() {group=LittleEgg, parameterName=Reactive_TimeFromBP})
   B=ScalarParameter() {defaultValue=0.8, group=LittleEgg, parameterName=RevealSpeed}
  )
  n36 = ScalarParameter() {defaultValue=1, group=LittleEgg, parameterName=Reactive_Level}
  n37 = Multiply(A=@n36) {constB=0.1}
  n38 = StaticSwitchParameter(True=Multiply(A=Saturate(None=@n35), B=NamedRerouteUsage()), False=Subtract(A=Saturate(None=@n37)) {constB=0.1}) {group=LittleEgg, parameterName=Use_BP_Time}
  n39 = Multiply(A=Saturate(None=Add(A=@n34, B=@n38))) {constB=10}
  n40 = MaterialFunctionCall(In=Multiply(A=Round(None=@n39)) {constB=0.1}) {materialFunction=&SmoothCeil}
  n41 = Subtract(A=@n36) {constB=2}
  n42 = StaticSwitchParameter(True=Add(A=@n40.Result, B=Saturate(None=@n41)), False=@n40.Result) {group=LittleEgg, parameterName=Use_BP_Time}
  NamedRerouteDeclaration(None=Saturate(None=Multiply(A=@n33, B=@n42))) {name=Mask_ReactiveLevel, nodeColor=<rgba 0.7109 0 1 1>}
  n43 = Fresnel() {baseReflectFraction=0.05, exponent=2}
  n44 = DotProduct(A=PixelNormalWS(), B=CameraVectorWS())
  n45 = Multiply(A=Saturate(None=Power(Base=@n44) {constExponent=10})) {constB=0.25}
  NamedRerouteDeclaration(None=Saturate(None=Add(A=@n43, B=@n45))) {name=Mask_Fresnel, nodeColor=<rgba 0.2422 1 0 1>}
  n46 = CurveAtlasRowParameter(CurveTime=@n11.B) {
   atlas=/BRCosmetics/Characters/Player/Female/Medium/Bodies/F_MED_Little_Egg_Jacket/Textures/LittleEgg_CurveAtlas
   curve=/BRCosmetics/Characters/Player/Female/Medium/Bodies/F_MED_Little_Egg_Jacket/Textures/LittleEgg_ReactiveColors
  }
  NamedRerouteDeclaration(None=MaterialFunctionCall(Hue Shift Percentage=ScalarParameter() {group=LittleEgg, parameterName=Reactive_ColorShift, sliderMax=1}, Texture=Multiply(A=@n11.R, B=@n46)) {materialFunction=&HueShift}.Result) {name=Rainbow_BaseColor, nodeColor=<rgba 0.3594 0 1 1>}
  n47 = Subtract(A=@n36)
  NamedRerouteDeclaration(None=Saturate(None=Floor(None=@n47))) {name=Mask_ReactiveActivated, nodeColor=<rgba 0.8516 0 1 1>}
 ]
}
