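# Crouch physics asset for the female medium Little Egg Jacket body, setting up collision and joints for its dynamic coat bones.
# Capsules cover legs, pelvis and spine, while a huge sphere below root approximates a floor.
# Front chains hang from spine_04, back and side chains from spine_01.
# Front and back chains have angular limits; side chains and cross-links tying back panels to each other and the front rotate freely.
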PhysicsAsset /BRCosmetics/Characters/Player/Female/Medium/Bodies/F_MED_Little_Egg_Jacket/Meshes/F_MED_Little_Egg_Jacket_Physics_Crouch
bodies {
 calf_l [{shapeType=Capsule, length=31.75, radius=8.5, shapeName=calf_l_capsule, center=<-17.5 -2 0>, rotation=<rot 90 0 0>}]
 calf_r [{shapeType=Capsule, length=31.75, radius=8.5, shapeName=calf_r_capsule, center=<17.5 2 0>, rotation=<rot 90 0 0>}]
 dyn_coat_bk_01 [{shapeType=Capsule, length=22, radius=8, shapeName=dyn_coat_bk_r_02_capsule, center=<9 0 7>, rotation=<rot 0 0 90>}]
 dyn_coat_bk_02 [{shapeType=Capsule, length=20, radius=8.25, shapeName=dyn_coat_bk_02_capsule, center=<9 0 7>, rotation=<rot 0 0 90>}]
 dyn_coat_bk_04 [{shapeType=Capsule, length=20, radius=8, shapeName=dyn_coat_bk_04_capsule, center=<9 0 7>, rotation=<rot 0 0 90>}]
 dyn_coat_bk_MSTR [{shapeType=Capsule, length=10, radius=8, shapeName=dyn_coat_bk_MSTR_capsule, center=<7 0 0>, rotation=<rot 90 0 0>}]
 dyn_coat_bk_l_01 [{shapeType=Capsule, length=22, radius=8, shapeName=dyn_coat_bk_r_02_capsule, center=<9 0 7>, rotation=<rot 0 0 90>}]
 dyn_coat_bk_l_02 [{shapeType=Capsule, length=20, radius=8.25, shapeName=dyn_coat_bk_l_02_capsule, center=<9 0 7>, rotation=<rot 0 0 90>}]
 dyn_coat_bk_l_04 [{shapeType=Capsule, length=20, radius=8, shapeName=dyn_coat_bk_l_04_capsule, center=<9 0 7>, rotation=<rot 0 0 90>}]
 dyn_coat_bk_l_05 [{shapeType=Capsule, length=20, radius=8, shapeName=dyn_coat_bk_l_05_capsule, center=<9 0 5>, rotation=<rot 0 0 90>}]
 dyn_coat_bk_l_06 [{shapeType=Capsule, length=20, radius=8, shapeName=dyn_coat_bk_l_06_capsule, center=<7 0 5>, rotation=<rot 0 0 90>}]
 dyn_coat_bk_l_MSTR [{shapeType=Capsule, length=10, radius=8, shapeName=dyn_coat_bk_l_MSTR_capsule, center=<7 0 4>, rotation=<rot 90 0 0>}]
 dyn_coat_bk_r_01 [{shapeType=Capsule, length=22, radius=8, shapeName=dyn_coat_bk_r_02_capsule, center=<9 0 7>, rotation=<rot 0 0 90>}]
 dyn_coat_bk_r_02 [{shapeType=Capsule, length=20, radius=8.25, shapeName=dyn_coat_bk_r_02_capsule, center=<9 0 7>, rotation=<rot 0 0 90>}]
 dyn_coat_bk_r_04 [{shapeType=Capsule, length=20, radius=8, shapeName=dyn_coat_bk_r_04_capsule, center=<9 0 7>, rotation=<rot 0 0 90>}]
 dyn_coat_bk_r_05 [{shapeType=Capsule, length=20, radius=8, shapeName=dyn_coat_bk_r_05_capsule, center=<9 0 5>, rotation=<rot 0 0 90>}]
 dyn_coat_bk_r_06 [{shapeType=Capsule, length=20, radius=8, shapeName=dyn_coat_bk_r_06_capsule, center=<7 0 5>, rotation=<rot 0 0 90>}]
 dyn_coat_bk_r_MSTR [{shapeType=Capsule, length=10, radius=8, shapeName=dyn_coat_bk_r_MSTR_capsule, center=<7 0 4>, rotation=<rot 90 0 0>}]
 dyn_coat_bk_split_01 [{shapeType=Capsule, length=20, radius=8, shapeName=dyn_coat_bk_split_01_capsule, center=<9 0 5>, rotation=<rot 0 0 90>}]
 dyn_coat_bk_split_02 [{shapeType=Capsule, length=20, radius=8, shapeName=dyn_coat_bk_split_02_capsule, center=<7 0 5>, rotation=<rot 0 0 90>}]
 dyn_coat_l_01 [{shapeType=Capsule, length=8, radius=2, shapeName=dyn_coat_l_01_capsule, center=<7 0 0>, rotation=<rot 90 0 0>}]
 dyn_coat_l_02 [
  {shapeType=Capsule, length=8, radius=5.5, shapeName=dyn_coat_l_03_capsule, center=<8.1867 -7 4>, rotation=<rot 90 0 0>}
  {shapeType=Capsule, length=12, radius=3, shapeName=dyn_coat_l_03_capsule, center=<7.1867 0 1>, rotation=<rot 90 0 0>}
 ]
 dyn_coat_l_03 [
  {shapeType=Capsule, length=7, radius=5.25, shapeName=dyn_coat_l_03_capsule, center=<7.8827 -4 5>, rotation=<rot 90 0 0>}
  {shapeType=Capsule, length=13, radius=2.75, shapeName=dyn_coat_l_03_capsule, center=<7.8827 3.75 -4>, rotation=<rot 90 0 0>}
 ]
 dyn_coat_l_04 [
  {shapeType=Capsule, length=10, radius=5, shapeName=dyn_coat_l_04_capsule, center=<10 -4 5>, rotation=<rot 90 0 0>}
  {shapeType=Capsule, length=16, radius=2.5, shapeName=dyn_coat_l_04_capsule, center=<8 3.75 -4>, rotation=<rot 90 0 0>}
 ]
 dyn_coat_l_05 [
  {shapeType=Capsule, length=10, radius=5, shapeName=dyn_coat_l_05_capsule, center=<10 -4 5>, rotation=<rot 90 0 0>}
  {shapeType=Capsule, length=16, radius=2.5, shapeName=dyn_coat_l_05_capsule, center=<8 3.75 -4>, rotation=<rot 90 0 0>}
 ]
 dyn_coat_l_06 [
  {shapeType=Capsule, length=12.75, radius=5, shapeName=dyn_coat_l_06_capsule, center=<6.6612 -4 5>, rotation=<rot 90 -179.9604 -179.9604>}
  {shapeType=Capsule, length=18.75, radius=2.5, shapeName=dyn_coat_l_06_capsule, center=<8.6612 3.75 -4>, rotation=<rot 90 -179.9604 -179.9604>}
 ]
 dyn_coat_l_MSTR [{shapeType=Capsule, length=5, radius=5, shapeName=dyn_coat_l_MSTR_capsule, center=<7 0 4>, rotation=<rot 90 0 0>}]
 dyn_coat_r_01 [{shapeType=Capsule, length=8, radius=2, shapeName=dyn_coat_r_01_capsule, center=<7 0 0>, rotation=<rot 90 0 0>}]
 dyn_coat_r_02 [{shapeType=Capsule, length=8, radius=2, shapeName=dyn_coat_l_01_capsule, center=<7 0 0>, rotation=<rot 90 0 0>}]
 dyn_coat_r_03 [
  {shapeType=Capsule, length=10, radius=5, shapeName=dyn_coat_r_03_capsule, center=<8 4 5>, rotation=<rot 90 0 0>}
  {shapeType=Capsule, length=16, radius=2.25, shapeName=dyn_coat_r_03_capsule, center=<8 -3.75 -4>, rotation=<rot 90 0 0>}
 ]
 dyn_coat_r_04 [
  {shapeType=Capsule, length=10, radius=5, shapeName=dyn_coat_r_04_capsule, center=<10 4 5>, rotation=<rot 90 0 0>}
  {shapeType=Capsule, length=16, radius=2.25, shapeName=dyn_coat_r_04_capsule, center=<8 -3.75 -4>, rotation=<rot 90 0 0>}
 ]
 dyn_coat_r_05 [
  {shapeType=Capsule, length=10, radius=5, shapeName=dyn_coat_r_05_capsule, center=<10 4 5>, rotation=<rot 90 0 0>}
  {shapeType=Capsule, length=16, radius=2.25, shapeName=dyn_coat_r_05_capsule, center=<8 -3.75 -4>, rotation=<rot 90 25.0898 25.0898>}
 ]
 dyn_coat_r_06 [
  {shapeType=Capsule, length=12.75, radius=5, shapeName=dyn_coat_r_06_capsule, center=<6.6612 4 5>, rotation=<rot 90 0 0>}
  {shapeType=Capsule, length=18.75, radius=2.25, shapeName=dyn_coat_r_06_capsule, center=<8.6612 -3.75 -4>, rotation=<rot 90 179.9951 179.9951>}
 ]
 dyn_coat_r_MSTR [{shapeType=Capsule, length=5, radius=5, shapeName=dyn_coat_r_MSTR_capsule, center=<7 0 4>, rotation=<rot 90 0 0>}]
 dyn_coat_side_l_01 [{shapeType=Capsule, length=5, radius=3, shapeName=dyn_coat_side_l_01_capsule, center=<3 0 0>, rotation=<rot 90 0 0>}]
 dyn_coat_side_l_02 [{shapeType=Capsule, length=5, radius=3, shapeName=dyn_coat_side_l_02_capsule, center=<3 0 0>, rotation=<rot 90 0 0>}]
 dyn_coat_side_l_03 [{shapeType=Capsule, length=5, radius=3, shapeName=dyn_coat_side_l_03_capsule, center=<3 0 0>, rotation=<rot 90 0 0>}]
 dyn_coat_side_l_04 [{shapeType=Capsule, length=5, radius=3, shapeName=dyn_coat_side_l_04_capsule, center=<3 0 0>, rotation=<rot 90 0 0>}]
 dyn_coat_side_l_05 [{shapeType=Capsule, length=5, radius=3, shapeName=dyn_coat_side_l_05_capsule, center=<3 0 0>, rotation=<rot 90 0 0>}]
 dyn_coat_side_l_MSTR [{shapeType=Capsule, length=5, radius=3, shapeName=dyn_coat_side_l_MSTR_capsule, center=<3 0 0>, rotation=<rot 90 0 0>}]
 dyn_coat_side_r_01 [{shapeType=Capsule, length=5, radius=3, shapeName=dyn_coat_side_r_01_capsule, center=<3 0 0>, rotation=<rot 90 0 0>}]
 dyn_coat_side_r_02 [{shapeType=Capsule, length=5, radius=3, shapeName=dyn_coat_side_r_02_capsule, center=<3 0 0>, rotation=<rot 90 0 0>}]
 dyn_coat_side_r_03 [{shapeType=Capsule, length=5, radius=3, shapeName=dyn_coat_side_r_03_capsule, center=<3 0 0>, rotation=<rot 90 0 0>}]
 dyn_coat_side_r_04 [{shapeType=Capsule, length=5, radius=3, shapeName=dyn_coat_side_r_04_capsule, center=<3 0 0>, rotation=<rot 90 0 0>}]
 dyn_coat_side_r_05 [{shapeType=Capsule, length=5, radius=3, shapeName=dyn_coat_side_r_05_capsule, center=<3 0 0>, rotation=<rot 90 0 0>}]
 dyn_coat_side_r_MSTR [{shapeType=Capsule, length=5, radius=3, shapeName=dyn_coat_side_r_MSTR_capsule, center=<3 0 0>, rotation=<rot 90 0 0>}]
 foot_l [{shapeType=Capsule, length=14.25, radius=10.5, shapeName=foot_l_capsule, center=<0.1414 5 0>, rotation=<rot 0 0 90>}]
 foot_r [{shapeType=Capsule, length=14.25, radius=10.5, shapeName=foot_r_capsule, center=<0.1414 -5 0>, rotation=<rot 0 0 90>}]
 pelvis [
  {shapeType=Capsule, length=9.25, radius=12, shapeName=pelvis_capsule, center=<-2.1944 -1 0>, rotation=<rot 90 0 0>}
  {
   shapeType=Capsule
   length=15.3785
   radius=7.9107
   shapeName=pelvis_capsule
   center=<-5.5488 -1.7063 -4.6532>
   rotation=<rot 80.001 -176.6924 -180>
  }
  {shapeType=Capsule, length=15.6212, radius=7.6681, shapeName=pelvis_capsule, center=<-3.4687 -1.5861 4.6995>, rotation=<rot 77.8519 3.3074 0>}
 ]
 root [
  {shapeType=Sphere, radius=500, shapeName=root_sphere, center=<0 0 -498>}
  {shapeType=Sphere, radius=50, shapeName=root_sphere, center=<0 -2 -11.3869>}
 ]
 spine_01 [{shapeType=Capsule, length=6, radius=8.05, shapeName=spine_01_capsule}]
 spine_02 [{shapeType=Capsule, length=4.5, radius=8, shapeName=spine_02_capsule}]
 spine_03 [{shapeType=Capsule, length=6, radius=8, shapeName=spine_03_capsule}]
 spine_04 [{shapeType=Capsule, length=6.5, radius=8.5, shapeName=spine_03_capsule, center=<0 0.6281 0>}]
 thigh_l [{shapeType=Capsule, length=30, radius=7, shapeName=thigh_l_capsule, center=<-21 0 0.5>, rotation=<rot 92 0 0>}]
 thigh_r [{shapeType=Capsule, length=30.5, radius=6.5, shapeName=thigh_r_capsule, center=<21 0 -0.5>, rotation=<rot 92 0 0>}]
}
constraints [
 {
  bone1Name=thigh_l
  bone2Name=pelvis
  swing1LimitDegrees=45
  swing1Motion=Free
  swing2LimitDegrees=45
  swing2Motion=Free
  twistLimitDegrees=45
  twistMotion=Free
 }
 {
  bone1Name=thigh_r
  bone2Name=pelvis
  swing1LimitDegrees=45
  swing1Motion=Free
  swing2LimitDegrees=45
  swing2Motion=Free
  twistLimitDegrees=45
  twistMotion=Free
 }
 {
  bone1Name=calf_r
  bone2Name=thigh_r
  swing1LimitDegrees=45
  swing1Motion=Free
  swing2LimitDegrees=45
  swing2Motion=Free
  twistLimitDegrees=45
  twistMotion=Free
 }
 {
  bone1Name=calf_l
  bone2Name=thigh_l
  swing1LimitDegrees=45
  swing1Motion=Free
  swing2LimitDegrees=45
  swing2Motion=Free
  twistLimitDegrees=45
  twistMotion=Free
 }
 {
  bone1Name=pelvis
  bone2Name=root
  swing1LimitDegrees=45
  swing1Motion=Free
  swing2LimitDegrees=45
  swing2Motion=Free
  twistLimitDegrees=45
  twistMotion=Free
 }
 {
  bone1Name=spine_01
  bone2Name=pelvis
  swing1LimitDegrees=45
  swing1Motion=Free
  swing2LimitDegrees=45
  swing2Motion=Free
  twistLimitDegrees=45
  twistMotion=Free
 }
 {
  bone1Name=spine_02
  bone2Name=spine_01
  swing1LimitDegrees=45
  swing1Motion=Free
  swing2LimitDegrees=45
  swing2Motion=Free
  twistLimitDegrees=45
  twistMotion=Free
 }
 {
  bone1Name=spine_03
  bone2Name=spine_02
  swing1LimitDegrees=45
  swing1Motion=Free
  swing2LimitDegrees=45
  swing2Motion=Free
  twistLimitDegrees=45
  twistMotion=Free
 }
 {
  bone1Name=spine_04
  bone2Name=spine_03
  swing1LimitDegrees=45
  swing1Motion=Free
  swing2LimitDegrees=45
  swing2Motion=Free
  twistLimitDegrees=45
  twistMotion=Free
 }
 {
  bone1Name=dyn_coat_r_MSTR
  bone2Name=spine_04
  swing1LimitDegrees=45
  swing1Motion=Free
  swing2LimitDegrees=45
  swing2Motion=Free
  twistLimitDegrees=45
  twistMotion=Free
 }
 {bone1Name=dyn_coat_r_01, bone2Name=dyn_coat_r_MSTR, swing1LimitDegrees=10, swing2LimitDegrees=10, twistLimitDegrees=5}
 {bone1Name=dyn_coat_r_02, bone2Name=dyn_coat_r_01, swing1LimitDegrees=15, swing2LimitDegrees=15, twistLimitDegrees=10}
 {bone1Name=dyn_coat_r_03, bone2Name=dyn_coat_r_02, swing1LimitDegrees=30, swing2LimitDegrees=60, twistLimitDegrees=5}
 {bone1Name=dyn_coat_r_04, bone2Name=dyn_coat_r_03, swing1LimitDegrees=60, swing2LimitDegrees=60, twistLimitDegrees=5}
 {bone1Name=dyn_coat_r_05, bone2Name=dyn_coat_r_04, swing1LimitDegrees=45, swing2LimitDegrees=45, twistLimitDegrees=5}
 {bone1Name=dyn_coat_r_06, bone2Name=dyn_coat_r_05, swing1LimitDegrees=45, swing2LimitDegrees=45, twistLimitDegrees=5}
 {
  bone1Name=dyn_coat_l_MSTR
  bone2Name=spine_04
  swing1LimitDegrees=45
  swing1Motion=Free
  swing2LimitDegrees=45
  swing2Motion=Free
  twistLimitDegrees=45
  twistMotion=Free
 }
 {bone1Name=dyn_coat_l_01, bone2Name=dyn_coat_l_MSTR, swing1LimitDegrees=10, swing2LimitDegrees=10, twistLimitDegrees=5}
 {bone1Name=dyn_coat_l_02, bone2Name=dyn_coat_l_01, swing1LimitDegrees=50, swing2LimitDegrees=50, twistLimitDegrees=10}
 {bone1Name=dyn_coat_l_03, bone2Name=dyn_coat_l_02, swing1LimitDegrees=40, swing2LimitDegrees=60, twistLimitDegrees=5}
 {bone1Name=dyn_coat_l_04, bone2Name=dyn_coat_l_03, swing1LimitDegrees=60, swing2LimitDegrees=60, twistLimitDegrees=5}
 {bone1Name=dyn_coat_l_05, bone2Name=dyn_coat_l_04, swing1LimitDegrees=45, swing2LimitDegrees=45, twistLimitDegrees=5}
 {bone1Name=dyn_coat_l_06, bone2Name=dyn_coat_l_05, swing1LimitDegrees=45, swing2LimitDegrees=45, twistLimitDegrees=5}
 {
  bone1Name=dyn_coat_bk_r_MSTR
  bone2Name=spine_01
  swing1LimitDegrees=45
  swing1Motion=Free
  swing2LimitDegrees=45
  swing2Motion=Free
  twistLimitDegrees=45
  twistMotion=Free
 }
 {bone1Name=dyn_coat_bk_r_01, bone2Name=dyn_coat_bk_r_MSTR, swing1LimitDegrees=15, swing2LimitDegrees=15, twistLimitDegrees=5}
 {bone1Name=dyn_coat_bk_r_02, bone2Name=dyn_coat_bk_r_01, swing1LimitDegrees=35, swing2LimitDegrees=45, twistLimitDegrees=5}
 {bone1Name=dyn_coat_bk_r_04, bone2Name=dyn_coat_bk_r_02, swing1LimitDegrees=35, swing2LimitDegrees=35, twistLimitDegrees=5}
 {bone1Name=dyn_coat_bk_r_05, bone2Name=dyn_coat_bk_r_04, swing1LimitDegrees=35, swing2LimitDegrees=35, twistLimitDegrees=5}
 {bone1Name=dyn_coat_bk_r_06, bone2Name=dyn_coat_bk_r_05, swing1LimitDegrees=35, swing2LimitDegrees=35, twistLimitDegrees=5}
 {
  bone1Name=dyn_coat_bk_MSTR
  bone2Name=spine_01
  swing1LimitDegrees=45
  swing1Motion=Free
  swing2LimitDegrees=45
  swing2Motion=Free
  twistLimitDegrees=45
  twistMotion=Free
 }
 {bone1Name=dyn_coat_bk_01, bone2Name=dyn_coat_bk_MSTR, swing1LimitDegrees=15, swing2LimitDegrees=15, twistLimitDegrees=5}
 {bone1Name=dyn_coat_bk_02, bone2Name=dyn_coat_bk_01, swing1LimitDegrees=25, swing2LimitDegrees=45, twistLimitDegrees=5}
 {bone1Name=dyn_coat_bk_04, bone2Name=dyn_coat_bk_02, swing1LimitDegrees=25, swing2LimitDegrees=35, twistLimitDegrees=5}
 {bone1Name=dyn_coat_bk_split_01, bone2Name=dyn_coat_bk_04, swing1LimitDegrees=25, swing2LimitDegrees=35, twistLimitDegrees=5}
 {bone1Name=dyn_coat_bk_split_02, bone2Name=dyn_coat_bk_split_01, swing1LimitDegrees=25, swing2LimitDegrees=35, twistLimitDegrees=5}
 {
  bone1Name=dyn_coat_bk_l_MSTR
  bone2Name=spine_01
  swing1LimitDegrees=45
  swing1Motion=Free
  swing2LimitDegrees=45
  swing2Motion=Free
  twistLimitDegrees=45
  twistMotion=Free
 }
 {bone1Name=dyn_coat_bk_l_01, bone2Name=dyn_coat_bk_l_MSTR, swing1LimitDegrees=15, swing2LimitDegrees=15, twistLimitDegrees=5}
 {bone1Name=dyn_coat_bk_l_02, bone2Name=dyn_coat_bk_l_01, swing1LimitDegrees=35, swing2LimitDegrees=45, twistLimitDegrees=5}
 {bone1Name=dyn_coat_bk_l_04, bone2Name=dyn_coat_bk_l_02, swing1LimitDegrees=35, swing2LimitDegrees=35, twistLimitDegrees=5}
 {bone1Name=dyn_coat_bk_l_05, bone2Name=dyn_coat_bk_l_04, swing1LimitDegrees=35, swing2LimitDegrees=35, twistLimitDegrees=5}
 {bone1Name=dyn_coat_bk_l_06, bone2Name=dyn_coat_bk_l_05, swing1LimitDegrees=35, swing2LimitDegrees=35, twistLimitDegrees=5}
 {
  bone1Name=foot_r
  bone2Name=calf_r
  swing1LimitDegrees=45
  swing1Motion=Free
  swing2LimitDegrees=45
  swing2Motion=Free
  twistLimitDegrees=45
  twistMotion=Free
 }
 {
  bone1Name=foot_l
  bone2Name=calf_l
  swing1LimitDegrees=45
  swing1Motion=Free
  swing2LimitDegrees=45
  swing2Motion=Free
  twistLimitDegrees=45
  twistMotion=Free
 }
 {
  bone1Name=dyn_coat_bk_l_06
  bone2Name=dyn_coat_bk_split_02
  swing1LimitDegrees=45
  swing1Motion=Free
  swing2LimitDegrees=45
  swing2Motion=Free
  twistLimitDegrees=45
  twistMotion=Free
 }
 {
  bone1Name=dyn_coat_bk_r_06
  bone2Name=dyn_coat_bk_split_02
  swing1LimitDegrees=45
  swing1Motion=Free
  swing2LimitDegrees=45
  swing2Motion=Free
  twistLimitDegrees=45
  twistMotion=Free
 }
 {
  bone1Name=dyn_coat_bk_r_04
  bone2Name=dyn_coat_r_04
  swing1LimitDegrees=45
  swing1Motion=Free
  swing2LimitDegrees=45
  swing2Motion=Free
  twistLimitDegrees=45
  twistMotion=Free
 }
 {
  bone1Name=dyn_coat_bk_l_04
  bone2Name=dyn_coat_l_04
  swing1LimitDegrees=45
  swing1Motion=Free
  swing2LimitDegrees=45
  swing2Motion=Free
  twistLimitDegrees=45
  twistMotion=Free
 }
 {
  bone1Name=dyn_coat_side_l_MSTR
  bone2Name=spine_01
  swing1LimitDegrees=45
  swing1Motion=Free
  swing2LimitDegrees=45
  swing2Motion=Free
  twistLimitDegrees=45
  twistMotion=Free
 }
 {
  bone1Name=dyn_coat_side_l_01
  bone2Name=dyn_coat_side_l_MSTR
  swing1LimitDegrees=45
  swing1Motion=Free
  swing2LimitDegrees=45
  swing2Motion=Free
  twistLimitDegrees=45
  twistMotion=Free
 }
 {
  bone1Name=dyn_coat_side_l_02
  bone2Name=dyn_coat_side_l_01
  swing1LimitDegrees=45
  swing1Motion=Free
  swing2LimitDegrees=45
  swing2Motion=Free
  twistLimitDegrees=45
  twistMotion=Free
 }
 {
  bone1Name=dyn_coat_side_l_03
  bone2Name=dyn_coat_side_l_02
  swing1LimitDegrees=45
  swing1Motion=Free
  swing2LimitDegrees=45
  swing2Motion=Free
  twistLimitDegrees=45
  twistMotion=Free
 }
 {
  bone1Name=dyn_coat_side_l_04
  bone2Name=dyn_coat_side_l_03
  swing1LimitDegrees=45
  swing1Motion=Free
  swing2LimitDegrees=45
  swing2Motion=Free
  twistLimitDegrees=45
  twistMotion=Free
 }
 {
  bone1Name=dyn_coat_side_l_05
  bone2Name=dyn_coat_side_l_04
  swing1LimitDegrees=45
  swing1Motion=Free
  swing2LimitDegrees=45
  swing2Motion=Free
  twistLimitDegrees=45
  twistMotion=Free
 }
 {
  bone1Name=dyn_coat_side_r_MSTR
  bone2Name=spine_01
  swing1LimitDegrees=45
  swing1Motion=Free
  swing2LimitDegrees=45
  swing2Motion=Free
  twistLimitDegrees=45
  twistMotion=Free
 }
 {
  bone1Name=dyn_coat_side_r_01
  bone2Name=dyn_coat_side_r_MSTR
  swing1LimitDegrees=45
  swing1Motion=Free
  swing2LimitDegrees=45
  swing2Motion=Free
  twistLimitDegrees=45
  twistMotion=Free
 }
 {
  bone1Name=dyn_coat_side_r_02
  bone2Name=dyn_coat_side_r_01
  swing1LimitDegrees=45
  swing1Motion=Free
  swing2LimitDegrees=45
  swing2Motion=Free
  twistLimitDegrees=45
  twistMotion=Free
 }
 {
  bone1Name=dyn_coat_side_r_03
  bone2Name=dyn_coat_side_r_02
  swing1LimitDegrees=45
  swing1Motion=Free
  swing2LimitDegrees=45
  swing2Motion=Free
  twistLimitDegrees=45
  twistMotion=Free
 }
 {
  bone1Name=dyn_coat_side_r_04
  bone2Name=dyn_coat_side_r_03
  swing1LimitDegrees=45
  swing1Motion=Free
  swing2LimitDegrees=45
  swing2Motion=Free
  twistLimitDegrees=45
  twistMotion=Free
 }
 {
  bone1Name=dyn_coat_side_r_05
  bone2Name=dyn_coat_side_r_04
  swing1LimitDegrees=45
  swing1Motion=Free
  swing2LimitDegrees=45
  swing2Motion=Free
  twistLimitDegrees=45
  twistMotion=Free
 }
]
