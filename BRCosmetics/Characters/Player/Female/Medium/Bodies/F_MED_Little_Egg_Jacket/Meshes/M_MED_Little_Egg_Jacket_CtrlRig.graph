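# Control Rig for the Little Egg Jacket character body that adjusts six lapel bones (right side named "rapel") as the arms move.
# Spherical pose readers measure each upper arm against spine_04, plus the right clavicle against spine_05 for two extra right-side corrections.
# Their outputs weight a shared "Set Transforms for Bone" function, which adds translate, Euler rotate and scale offsets in local space.
# Enabled debug axes draw each lapel before and after correction.
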ControlRigBlueprint /BRCosmetics/Characters/Player/Female/Medium/Bodies/F_MED_Little_Egg_Jacket/Meshes/M_MED_Little_Egg_Jacket_CtrlRig
graphs [
 ContainedGraph {nodes [
   n1 = RigVMFunctionEntryNode()
   RigUnit_GetTransform = &RigUnit_GetTransform(Space=LocalSpace, bInitial=false, Item=@n1.Joint)
   n2 = &RigVMFunction_MathQuaternionToEuler(RotationOrder=ZYX, Value=@RigUnit_GetTransform.Transform.Rotation)
   > RigUnit_SetTransform = &RigUnit_SetTransform(
    Space=LocalSpace
    Value=(Rotation=(X=0.0132,Y=0.4099,Z=-0.0222,W=0.9118),Translation=(X=23.6045,Y=-6.37,Z=140.1115),Scale3D=(X=1,Y=1,Z=1))
    bInitial=false
    Item=@n1.Joint
    Weight=@n1.Weight
    Value.Translation.Z=&RigVMFunction_MathFloatAdd(B=@n1.Translate Z, A=@RigUnit_GetTransform.Transform.Translation.Z) {template="Add::Execute(in A,in B,out Result)"}.Result
    Value.Scale3D.Z=&RigVMFunction_MathFloatAdd(B=@n1.Scale Z, A=@RigUnit_GetTransform.Transform.Scale3D.Z) {template="Add::Execute(in A,in B,out Result)"}.Result
    Value.Translation.Y=&RigVMFunction_MathFloatAdd(B=@n1.Translate Y, A=@RigUnit_GetTransform.Transform.Translation.Y) {template="Add::Execute(in A,in B,out Result)"}.Result
    Value.Scale3D.Y=&RigVMFunction_MathFloatAdd(B=@n1.Scale Y, A=@RigUnit_GetTransform.Transform.Scale3D.Y) {template="Add::Execute(in A,in B,out Result)"}.Result
    Value.Translation.X=&RigVMFunction_MathFloatAdd(B=@n1.Translate X, A=@RigUnit_GetTransform.Transform.Translation.X) {template="Add::Execute(in A,in B,out Result)"}.Result
    Value.Scale3D.X=&RigVMFunction_MathFloatAdd(B=@n1.Scale X, A=@RigUnit_GetTransform.Transform.Scale3D.X) {template="Add::Execute(in A,in B,out Result)"}.Result
    Value.Rotation=&RigVMFunction_MathQuaternionFromEuler(
     Euler=(X=0,Y=0,Z=0)
     Euler.X=&RigVMFunction_MathFloatAdd(B=@n1.Rotate X, A=@n2.Result.X) {template="Add::Execute(in A,in B,out Result)"}.Result
     Euler.Y=&RigVMFunction_MathFloatAdd(B=@n1.Rotate Y, A=@n2.Result.Y) {template="Add::Execute(in A,in B,out Result)"}.Result
     Euler.Z=&RigVMFunction_MathFloatAdd(B=@n1.Rotate Z, A=@n2.Result.Z) {template="Add::Execute(in A,in B,out Result)"}.Result
    ).Result
   )
   > RigVMFunctionReturnNode()
  ]}
 RigVMFunctionLibrary {nodes [Set Transforms for Bone = RigVMCollapseNode(
    ExecuteContext=()
    Joint=()
    Rotate X=0
    Rotate Y=0
    Rotate Z=0
    Scale X=0
    Scale Y=0
    Scale Z=0
    Translate X=0
    Translate Y=0
    Translate Z=0
    Weight=1
   )]}
 RigVMModel {nodes [
   RigUnit_BeginExecution = &RigUnit_BeginExecution()
   /Script/ControlRig.RigUnit_DebugTransformMutableItemSpace(
    Color=(R=1,G=0,B=0,A=1)
    DebugDrawSettings=(DepthPriority=SDPG_Foreground,Lifetime=-1)
    Mode=Axes
    Scale=5
    Space=(Type=Bone)
    Thickness=0.1
    WorldOffset=(Rotation=(X=0,Y=0,Z=0,W=1),Translation=(X=0,Y=0,Z=0),Scale3D=(X=1,Y=1,Z=1))
    bEnabled=True
    ExecutePin=@RigUnit_BeginExecution.ExecutePin
    Transform=&RigUnit_GetTransform(Item=(Type=Bone,Name="dyn_lapel_l")).Transform
   ) {is=struct RigVMUnitNode}
   > &RigUnit_DebugTransformMutableItemSpace(
    Color=(R=1,G=0,B=0,A=1)
    DebugDrawSettings=(DepthPriority=SDPG_Foreground,Lifetime=-1)
    Mode=Axes
    Scale=5
    Space=(Type=Bone)
    Thickness=0.1
    WorldOffset=(Rotation=(X=0,Y=0,Z=0,W=1),Translation=(X=0,Y=0,Z=0),Scale3D=(X=1,Y=1,Z=1))
    bEnabled=True
    Transform=&RigUnit_GetTransform(Item=(Type=Bone,Name="dyn_lapel_low_l")).Transform
   )
   > &RigUnit_DebugTransformMutableItemSpace(
    Color=(R=1,G=0,B=0,A=1)
    DebugDrawSettings=(DepthPriority=SDPG_Foreground,Lifetime=-1)
    Mode=Axes
    Scale=5
    Space=(Type=Bone)
    Thickness=0.1
    WorldOffset=(Rotation=(X=0,Y=0,Z=0,W=1),Translation=(X=0,Y=0,Z=0),Scale3D=(X=1,Y=1,Z=1))
    bEnabled=True
    Transform=&RigUnit_GetTransform(Item=(Type=Bone,Name="dyn_lapel_high_l")).Transform
   )
   > n1 = &RigUnit_SphericalPoseReader(
    ActiveRegionScaleFactors=(PositiveWidth=0.6,NegativeWidth=0.6,PositiveHeight=1,NegativeHeight=1)
    ActiveRegionSize=0.15
    Debug=(bDrawDebug=false,bDraw2D=false,bDrawLocalAxes=true,DebugScale=25,DebugSegments=20,DebugThickness=0.25)
    DriverAxis=(X=1,Y=0,Z=0)
    DriverItem=(Type=Bone,Name="upperarm_l")
    FalloffRegionScaleFactors=(PositiveWidth=0.6,NegativeWidth=0.6,PositiveHeight=1,NegativeHeight=1)
    FalloffSize=0.45
    OptionalParentItem=(Type=Bone,Name="spine_04")
    RotationOffset=(X=40,Y=25,Z=-40)
   )
   > Set Transforms for Bone = RigVMFunctionReferenceNode(
    Joint=(Type=Bone,Name="dyn_lapel_high_l")
    Rotate X=0
    Rotate Y=-70
    Rotate Z=50
    Scale X=0
    Scale Y=0
    Scale Z=0
    Translate X=-2
    Translate Y=0
    Translate Z=1
    Weight=@n1.OutputParam
   )
   > n2 = &RigUnit_SphericalPoseReader(
    ActiveRegionScaleFactors=(PositiveWidth=0.5,NegativeWidth=1,PositiveHeight=1,NegativeHeight=2)
    ActiveRegionSize=0.08
    Debug=(bDrawDebug=false,bDraw2D=false,bDrawLocalAxes=true,DebugScale=25,DebugSegments=20,DebugThickness=0.25)
    DriverAxis=(X=1,Y=0,Z=0)
    DriverItem=(Type=Bone,Name="upperarm_l")
    FalloffRegionScaleFactors=(PositiveWidth=0.5,NegativeWidth=1,PositiveHeight=1,NegativeHeight=1)
    FalloffSize=0.45
    OptionalParentItem=(Type=Bone,Name="spine_04")
    RotationOffset=(X=80,Y=30,Z=-25)
   )
   > Set Transforms for Bone_1 = RigVMFunctionReferenceNode(
    Joint=(Type=Bone,Name="dyn_lapel_l")
    Rotate X=0
    Rotate Y=-65
    Rotate Z=0
    Scale X=0
    Scale Y=0
    Scale Z=0
    Translate X=-1
    Translate Y=0
    Translate Z=1
    Weight=@n2.OutputParam
   )
   > n3 = &RigUnit_SphericalPoseReader(
    ActiveRegionScaleFactors=(PositiveWidth=1,NegativeWidth=1,PositiveHeight=3,NegativeHeight=2)
    ActiveRegionSize=0.05
    Debug=(bDrawDebug=false,bDraw2D=false,bDrawLocalAxes=true,DebugScale=25,DebugSegments=20,DebugThickness=0.25)
    DriverAxis=(X=1,Y=0,Z=0)
    DriverItem=(Type=Bone,Name="upperarm_l")
    FalloffRegionScaleFactors=(PositiveWidth=0.5,NegativeWidth=0.5,PositiveHeight=2,NegativeHeight=2)
    FalloffSize=0.3
    OptionalParentItem=(Type=Bone,Name="spine_04")
    RotationOffset=(X=110,Y=30,Z=-30)
   )
   > Set Transforms for Bone_1_1 = RigVMFunctionReferenceNode(
    Joint=(Type=Bone,Name="dyn_lapel_low_l")
    Rotate X=0
    Rotate Y=-40
    Rotate Z=0
    Scale X=0
    Scale Y=0
    Scale Z=0
    Translate X=-1
    Translate Y=0
    Translate Z=0
    Weight=@n3.OutputParam
   )
   > &RigUnit_DebugTransformMutableItemSpace(
    Color=(R=1,G=0,B=0,A=1)
    DebugDrawSettings=(DepthPriority=SDPG_Foreground,Lifetime=-1)
    Mode=Axes
    Scale=3
    Space=(Type=Bone)
    Thickness=0.3
    WorldOffset=(Rotation=(X=0,Y=0,Z=0,W=1),Translation=(X=0,Y=0,Z=0),Scale3D=(X=1,Y=1,Z=1))
    bEnabled=True
    Transform=&RigUnit_GetTransform(Item=(Type=Bone,Name="dyn_lapel_l")).Transform
   )
   > &RigUnit_DebugTransformMutableItemSpace(
    Color=(R=1,G=0,B=0,A=1)
    DebugDrawSettings=(DepthPriority=SDPG_Foreground,Lifetime=-1)
    Mode=Axes
    Scale=3
    Space=(Type=Bone)
    Thickness=0.3
    WorldOffset=(Rotation=(X=0,Y=0,Z=0,W=1),Translation=(X=0,Y=0,Z=0),Scale3D=(X=1,Y=1,Z=1))
    bEnabled=True
    Transform=&RigUnit_GetTransform(Item=(Type=Bone,Name="dyn_lapel_low_l")).Transform
   )
   > &RigUnit_DebugTransformMutableItemSpace(
    Color=(R=1,G=0,B=0,A=1)
    DebugDrawSettings=(DepthPriority=SDPG_Foreground,Lifetime=-1)
    Mode=Axes
    Scale=3
    Space=(Type=Bone)
    Thickness=0.3
    WorldOffset=(Rotation=(X=0,Y=0,Z=0,W=1),Translation=(X=0,Y=0,Z=0),Scale3D=(X=1,Y=1,Z=1))
    bEnabled=True
    Transform=&RigUnit_GetTransform(Item=(Type=Bone,Name="dyn_lapel_high_l")).Transform
   )
   > &RigUnit_DebugTransformMutableItemSpace(
    Color=(R=1,G=0,B=0,A=1)
    DebugDrawSettings=(DepthPriority=SDPG_Foreground,Lifetime=-1)
    Mode=Axes
    Scale=5
    Space=(Type=Bone)
    Thickness=0.1
    WorldOffset=(Rotation=(X=0,Y=0,Z=0,W=1),Translation=(X=0,Y=0,Z=0),Scale3D=(X=1,Y=1,Z=1))
    bEnabled=True
    Transform=&RigUnit_GetTransform(Item=(Type=Bone,Name="dyn_rapel_r")).Transform
   )
   > &RigUnit_DebugTransformMutableItemSpace(
    Color=(R=1,G=0,B=0,A=1)
    DebugDrawSettings=(DepthPriority=SDPG_Foreground,Lifetime=-1)
    Mode=Axes
    Scale=5
    Space=(Type=Bone)
    Thickness=0.1
    WorldOffset=(Rotation=(X=0,Y=0,Z=0,W=1),Translation=(X=0,Y=0,Z=0),Scale3D=(X=1,Y=1,Z=1))
    bEnabled=True
    Transform=&RigUnit_GetTransform(Item=(Type=Bone,Name="dyn_rapel_row_r")).Transform
   )
   > &RigUnit_DebugTransformMutableItemSpace(
    Color=(R=1,G=0,B=0,A=1)
    DebugDrawSettings=(DepthPriority=SDPG_Foreground,Lifetime=-1)
    Mode=Axes
    Scale=5
    Space=(Type=Bone)
    Thickness=0.1
    WorldOffset=(Rotation=(X=0,Y=0,Z=0,W=1),Translation=(X=0,Y=0,Z=0),Scale3D=(X=1,Y=1,Z=1))
    bEnabled=True
    Transform=&RigUnit_GetTransform(Item=(Type=Bone,Name="dyn_rapel_high_r")).Transform
   )
   > n4 = &RigUnit_SphericalPoseReader(
    ActiveRegionScaleFactors=(PositiveWidth=0.6,NegativeWidth=0.6,PositiveHeight=1,NegativeHeight=1)
    ActiveRegionSize=0.15
    Debug=(bDrawDebug=false,bDraw2D=false,bDrawLocalAxes=true,DebugScale=25,DebugSegments=20,DebugThickness=0.25)
    DriverAxis=(X=-1,Y=0,Z=0)
    DriverItem=(Type=Bone,Name="upperarm_r")
    FalloffRegionScaleFactors=(PositiveWidth=0.6,NegativeWidth=0.6,PositiveHeight=2,NegativeHeight=1)
    FalloffSize=0.45
    FlipHeightScaling=false
    FlipWidthScaling=true
    OptionalParentItem=(Type=Bone,Name="spine_04")
    RotationOffset=(X=-100,Y=25,Z=40)
   )
   > Set Transforms for Bone_1_1_1 = RigVMFunctionReferenceNode(
    Joint=(Type=Bone,Name="dyn_rapel_high_r")
    Rotate X=0
    Rotate Y=-70
    Rotate Z=0
    Scale X=0
    Scale Y=0
    Scale Z=0
    Translate X=0
    Translate Y=0
    Translate Z=0
    Weight=@n4.OutputParam
   )
   > n5 = &RigUnit_SphericalPoseReader(
    ActiveRegionScaleFactors=(PositiveWidth=0.5,NegativeWidth=1,PositiveHeight=1.8,NegativeHeight=2)
    ActiveRegionSize=0.08
    Debug=(bDrawDebug=false,bDraw2D=false,bDrawLocalAxes=true,DebugScale=25,DebugSegments=20,DebugThickness=0.25)
    DriverAxis=(X=-1,Y=0,Z=0)
    DriverItem=(Type=Bone,Name="upperarm_r")
    FalloffRegionScaleFactors=(PositiveWidth=0.5,NegativeWidth=1,PositiveHeight=2,NegativeHeight=1)
    FalloffSize=0.45
    FlipHeightScaling=false
    FlipWidthScaling=true
    OptionalParentItem=(Type=Bone,Name="spine_04")
    RotationOffset=(X=-80,Y=40,Z=25)
   )
   > Set Transforms for Bone_1_1_1_1 = RigVMFunctionReferenceNode(
    Joint=(Type=Bone,Name="dyn_rapel_r")
    Rotate X=0
    Rotate Y=-65
    Rotate Z=0
    Scale X=-0.3
    Scale Y=0
    Scale Z=0
    Translate X=-3
    Translate Y=0
    Translate Z=2
    Weight=@n5.OutputParam
   )
   > n6 = &RigUnit_SphericalPoseReader(
    ActiveRegionScaleFactors=(PositiveWidth=1,NegativeWidth=1,PositiveHeight=2,NegativeHeight=1)
    ActiveRegionSize=0.05
    Debug=(bDrawDebug=false,bDraw2D=false,bDrawLocalAxes=true,DebugScale=25,DebugSegments=20,DebugThickness=0.25)
    DriverAxis=(X=-1,Y=0,Z=0)
    DriverItem=(Type=Bone,Name="upperarm_r")
    FalloffRegionScaleFactors=(PositiveWidth=0.5,NegativeWidth=0.5,PositiveHeight=1,NegativeHeight=1)
    FalloffSize=0.3
    FlipWidthScaling=true
    OptionalParentItem=(Type=Bone,Name="spine_04")
    RotationOffset=(X=-50,Y=35,Z=20)
   )
   > Set Transforms for Bone_1_1_1_1_1 = RigVMFunctionReferenceNode(
    Joint=(Type=Bone,Name="dyn_rapel_row_r")
    Rotate X=0
    Rotate Y=-60
    Rotate Z=0
    Scale X=0
    Scale Y=0
    Scale Z=0
    Translate X=-1
    Translate Y=0
    Translate Z=3
    Weight=@n6.OutputParam
   )
   > n7 = &RigUnit_SphericalPoseReader(
    ActiveRegionScaleFactors=(PositiveWidth=1,NegativeWidth=1,PositiveHeight=1,NegativeHeight=1)
    ActiveRegionSize=0.02
    Debug=(bDrawDebug=false,bDraw2D=false,bDrawLocalAxes=true,DebugScale=25,DebugSegments=12,DebugThickness=0.15)
    DriverAxis=(X=-1,Y=0,Z=0)
    DriverItem=(Type=Bone,Name="clavicle_r")
    FalloffRegionScaleFactors=(PositiveWidth=1,NegativeWidth=1,PositiveHeight=1,NegativeHeight=1)
    FalloffSize=0.15
    FlipWidthScaling=false
    OptionalParentItem=(Type=Bone,Name="spine_05")
    RotationOffset=(X=0,Y=130,Z=20)
   )
   > Set Transforms for Bone_1_1_1_1_1_1 = RigVMFunctionReferenceNode(
    Joint=(Type=Bone,Name="dyn_rapel_high_r")
    Rotate X=0
    Rotate Y=-80
    Rotate Z=0
    Scale X=0
    Scale Y=0
    Scale Z=0
    Translate X=0
    Translate Y=0
    Translate Z=0
    Weight=@n7.OutputParam
   )
   > Set Transforms for Bone_1_1_1_1_1_1_1 = RigVMFunctionReferenceNode(
    Joint=(Type=Bone,Name="dyn_rapel_r")
    Rotate X=0
    Rotate Y=-40
    Rotate Z=0
    Scale X=0
    Scale Y=0
    Scale Z=0
    Translate X=0
    Translate Y=0
    Translate Z=0
    Weight=@n7.OutputParam
   )
   > &RigUnit_DebugTransformMutableItemSpace(
    Color=(R=1,G=0,B=0,A=1)
    DebugDrawSettings=(DepthPriority=SDPG_Foreground,Lifetime=-1)
    Mode=Axes
    Scale=3
    Space=(Type=Bone)
    Thickness=0.3
    WorldOffset=(Rotation=(X=0,Y=0,Z=0,W=1),Translation=(X=0,Y=0,Z=0),Scale3D=(X=1,Y=1,Z=1))
    bEnabled=True
    Transform=&RigUnit_GetTransform(Item=(Type=Bone,Name="dyn_rapel_r")).Transform
   )
   > &RigUnit_DebugTransformMutableItemSpace(
    Color=(R=1,G=0,B=0,A=1)
    DebugDrawSettings=(DepthPriority=SDPG_Foreground,Lifetime=-1)
    Mode=Axes
    Scale=3
    Space=(Type=Bone)
    Thickness=0.3
    WorldOffset=(Rotation=(X=0,Y=0,Z=0,W=1),Translation=(X=0,Y=0,Z=0),Scale3D=(X=1,Y=1,Z=1))
    bEnabled=True
    Transform=&RigUnit_GetTransform(Item=(Type=Bone,Name="dyn_rapel_row_r")).Transform
   )
   > &RigUnit_DebugTransformMutableItemSpace(
    Color=(R=1,G=0,B=0,A=1)
    DebugDrawSettings=(DepthPriority=SDPG_Foreground,Lifetime=-1)
    Mode=Axes
    Scale=3
    Space=(Type=Bone)
    Thickness=0.3
    WorldOffset=(Rotation=(X=0,Y=0,Z=0,W=1),Translation=(X=0,Y=0,Z=0),Scale3D=(X=1,Y=1,Z=1))
    bEnabled=True
    Transform=&RigUnit_GetTransform(Item=(Type=Bone,Name="dyn_rapel_high_r")).Transform
   )
  ]}
]
hierarchy [
 root {loc=<0 0 0>, rot=<0 0 0 1>}
 pelvis {parent=root, loc=<0 -0.5301 98.0571>, rot=<0.0127 0.707 -0.0127 -0.707>}
 spine_01 {parent=pelvis, loc=<6.1208 -0.9328 0>, rot=<0 0 0.034 -0.9994>}
 spine_02 {parent=spine_01, loc=<6.5996 -0.2095 0>, rot=<0 0 0.0238 -0.9997>}
 spine_03 {parent=spine_02, loc=<6.4914 -0.1237 0>, rot=<0 0 0.0149 -0.9999>}
 spine_04 {parent=spine_03, loc=<6.9598 0.0794 0>, rot=<0 0 -0.0132 -0.9999>}
 spine_05 {parent=spine_04, loc=<6.9987 0 0>, rot=<0 0 -0.0015 -1>}
 clavicle_l {parent=spine_05, loc=<5.4831 -0.1053 -3.2946>, rot=<0.0465 -0.7289 -0.009 -0.6829>}
 upperarm_l {parent=clavicle_l, loc=<9.8658 -1.2136 -0.4105>, rot=<0.0164 -0.3245 -0.0269 -0.9454>}
 lowerarm_l {parent=upperarm_l, loc=<26.465 -0.8345 0.259>, rot=<0.0036 0.0161 -0.2298 -0.9731>}
 hand_l {parent=lowerarm_l, loc=<22.7355 -0.3203 -0.6105>, rot=<-0.6784 -0.0223 0.0735 -0.7307>}
 index_metacarpal_l {parent=hand_l, loc=<1.6535 0.0183 -2.96>, rot=<0.0158 -0.0124 -0.0173 -0.9996>}
 index_01_l {parent=index_metacarpal_l, loc=<8.3008 0.0357 -0.0524>, rot=<-0.0044 0.0485 0.0924 -0.9945>}
 index_02_l {parent=index_01_l, loc=<3.3587 -0.3004 0.0452>, rot=<-0.002 0.0027 0.3225 -0.9466>}
 index_03_l {parent=index_02_l, loc=<2.5578 -0.123 -0.0252>, rot=<-0.002 0.0067 0.2558 -0.9667>}
 middle_metacarpal_l {parent=hand_l, loc=<2.1856 -0.1725 -0.5984>, rot=<0.006 0.0168 -0.0329 -0.9993>}
 middle_01_l {parent=middle_metacarpal_l, loc=<7.3275 -0.2447 0.0129>, rot=<-0.0208 0.0042 0.0858 -0.9961>}
 middle_02_l {parent=middle_01_l, loc=<4.0123 0.1401 0.1175>, rot=<-0.0218 0.0258 0.3512 -0.9357>}
 middle_03_l {parent=middle_02_l, loc=<2.6077 0.0534 0.1373>, rot=<0.0082 0.0188 0.1708 -0.9851>}
 pinky_metacarpal_l {parent=hand_l, loc=<1.7616 -0.9753 3.3814>, rot=<-0.0397 0.0814 -0.069 -0.9935>}
 pinky_01_l {parent=pinky_metacarpal_l, loc=<7.1709 -0.4831 -0.1094>, rot=<-0.0654 -0.0035 0.2242 -0.9723>}
 pinky_02_l {parent=pinky_01_l, loc=<2.8274 0.0893 -0.1043>, rot=<0.0037 0.0011 0.2854 -0.9584>}
 pinky_03_l {parent=pinky_02_l, loc=<2.2046 0.0307 -0.0078>, rot=<-0.0002 0.0098 0.1445 -0.9895>}
 ring_metacarpal_l {parent=hand_l, loc=<2.2074 -0.3625 1.3638>, rot=<-0.0284 0.0722 -0.0458 -0.9959>}
 ring_01_l {parent=ring_metacarpal_l, loc=<7.0295 -0.6197 -0.1446>, rot=<-0.0489 -0.0252 0.1892 -0.9804>}
 ring_02_l {parent=ring_01_l, loc=<3.6174 0.4647 -0.1027>, rot=<0 0 0.2806 -0.9598>}
 ring_03_l {parent=ring_02_l, loc=<2.4152 0.114 0.0199>, rot=<0 0 0.1189 -0.9929>}
 thumb_01_l {parent=hand_l, loc=<1.3118 -0.3191 -2.459>, rot=<0.6546 -0.4042 0.0664 -0.6354>}
 thumb_02_l {parent=thumb_01_l, loc=<5.9336 -0.0448 0.0401>, rot=<-0.0037 0.0213 0.298 -0.9543>}
 thumb_03_l {parent=thumb_02_l, loc=<2.2223 0.3682 -0.0907>, rot=<0.0118 -0.013 0.0066 -0.9998>}
 lowerarm_twist_01_l {parent=lowerarm_l, loc=<18.316 0 0>, rot=<0 0 0 1>}
 Dyn_MasterBracelet_r {parent=lowerarm_twist_01_l, loc=<2.1784 -1.9158 2.6406>, rot=<0.4423 -0.181 0.475 -0.7389>}
 Dyn_Bracelet_r1 {parent=Dyn_MasterBracelet_r, loc=<0.184 0 -0.1598>, rot=<0 0 0 1>}
 lowerarm_twist_02_l {parent=lowerarm_l, loc=<10.6707 0 0>, rot=<0 0 0 1>}
 upperarm_twist_01_l {parent=upperarm_l, loc=<7.2135 -0.1134 0>, rot=<0 0 0 1>}
 upperarm_twist_02_l {parent=upperarm_l, loc=<19.7998 -0.1134 0>, rot=<0 0 0 1>}
 clavicle_r {parent=spine_05, loc=<5.4831 -0.1053 3.2946>, rot=<0.7289 0.0465 -0.6829 0.009>}
 upperarm_r {parent=clavicle_r, loc=<-9.8658 1.2136 0.4104>, rot=<0.0164 -0.3245 -0.0269 -0.9454>}
 lowerarm_r {parent=upperarm_r, loc=<-26.4651 0.8345 -0.259>, rot=<0.0036 0.0161 -0.2298 -0.9731>}
 hand_r {parent=lowerarm_r, loc=<-22.7356 0.3203 0.6106>, rot=<-0.6784 -0.0223 0.0735 -0.7307>}
 index_metacarpal_r {parent=hand_r, loc=<-1.6534 -0.0181 2.9601>, rot=<0.0158 -0.0124 -0.0173 -0.9996>}
 index_01_r {parent=index_metacarpal_r, loc=<-8.3008 -0.0357 0.0524>, rot=<-0.0044 0.0485 0.0924 -0.9945>}
 index_02_r {parent=index_01_r, loc=<-3.3589 0.3003 -0.0453>, rot=<-0.002 0.0027 0.3225 -0.9466>}
 index_03_r {parent=index_02_r, loc=<-2.5578 0.1231 0.0251>, rot=<-0.002 0.0067 0.2558 -0.9667>}
 middle_metacarpal_r {parent=hand_r, loc=<-2.1856 0.1726 0.5984>, rot=<0.006 0.0168 -0.0329 -0.9993>}
 middle_01_r {parent=middle_metacarpal_r, loc=<-7.3273 0.2447 -0.0129>, rot=<-0.0208 0.0042 0.0858 -0.9961>}
 middle_02_r {parent=middle_01_r, loc=<-4.0124 -0.14 -0.1175>, rot=<-0.0218 0.0258 0.3512 -0.9357>}
 middle_03_r {parent=middle_02_r, loc=<-2.6077 -0.0533 -0.1373>, rot=<0.0082 0.0188 0.1708 -0.9851>}
 pinky_metacarpal_r {parent=hand_r, loc=<-1.7617 0.9753 -3.3814>, rot=<-0.0397 0.0814 -0.069 -0.9935>}
 pinky_01_r {parent=pinky_metacarpal_r, loc=<-7.1707 0.4832 0.1094>, rot=<-0.0654 -0.0035 0.2242 -0.9723>}
 pinky_02_r {parent=pinky_01_r, loc=<-2.8274 -0.0892 0.1043>, rot=<0.0037 0.0011 0.2854 -0.9584>}
 pinky_03_r {parent=pinky_02_r, loc=<-2.2045 -0.031 0.0078>, rot=<-0.0002 0.0098 0.1445 -0.9895>}
 ring_metacarpal_r {parent=hand_r, loc=<-2.2073 0.3627 -1.3638>, rot=<-0.0284 0.0722 -0.0458 -0.9959>}
 ring_01_r {parent=ring_metacarpal_r, loc=<-7.0297 0.6197 0.1446>, rot=<-0.0489 -0.0252 0.1892 -0.9804>}
 ring_02_r {parent=ring_01_r, loc=<-3.6171 -0.4648 0.1027>, rot=<0 0 0.2806 -0.9598>}
 ring_03_r {parent=ring_02_r, loc=<-2.4153 -0.1141 -0.0199>, rot=<0 0 0.1189 -0.9929>}
 thumb_01_r {parent=hand_r, loc=<-1.3119 0.3192 2.459>, rot=<0.6546 -0.4042 0.0664 -0.6354>}
 thumb_02_r {parent=thumb_01_r, loc=<-5.9335 0.0447 -0.0401>, rot=<-0.0037 0.0213 0.298 -0.9543>}
 thumb_03_r {parent=thumb_02_r, loc=<-2.2223 -0.3682 0.0907>, rot=<0.0118 -0.013 0.0066 -0.9998>}
 lowerarm_twist_01_r {parent=lowerarm_r, loc=<-18.3161 0 0>, rot=<0 0 0 1>}
 Dyn_MasterBracelet_l {parent=lowerarm_twist_01_r, loc=<-2.2447 2.1389 -2.7544>, rot=<0.153 0.4601 0.7178 0.4997>}
 Dyn_Bracelet_l1 {parent=Dyn_MasterBracelet_l, loc=<0.2157 0 0>, rot=<0 0 0 1>}
 lowerarm_twist_02_r {parent=lowerarm_r, loc=<-10.6707 0 0>, rot=<0 0 0 1>}
 upperarm_twist_01_r {parent=upperarm_r, loc=<-7.2135 0.1134 0>, rot=<0 0 0 1>}
 upperarm_twist_02_r {parent=upperarm_r, loc=<-19.7998 0.1134 0>, rot=<0 0 0 1>}
 neck_01 {parent=spine_05, loc=<11.3138 0.1529 0>, rot=<0 0 -0.2003 -0.9797>}
 neck_02 {parent=neck_01, loc=<4.5627 0 0>, rot=<0 0 0.0504 -0.9987>}
 head {parent=neck_02, loc=<5.3914 0 0>, rot=<0 0 0.115 -0.9934>}
 dyn_lapel_l_MSTR {parent=spine_05, loc=<1.8464 9.6698 -6.9362>, rot=<0.5166 -0.6602 -0.2961 -0.4577>}
 dyn_lapel_l {parent=dyn_lapel_l_MSTR, loc=<0.3 0 0>, rot=<0 0 0 1>}
 dyn_rapel_r_MSTR {parent=spine_05, loc=<1.8469 9.6691 6.3902>, rot=<0.4577 0.2961 0.6602 -0.5166>}
 dyn_rapel_r {parent=dyn_rapel_r_MSTR, loc=<0.3 0 0>, rot=<0 0 0 1>}
 dyn_lapel_high_l_MSTR {parent=spine_05, loc=<8.6249 4.4381 -5.8221>, rot=<0.4009 -0.7029 -0.2125 -0.5478>}
 dyn_lapel_high_l {parent=dyn_lapel_high_l_MSTR, loc=<0.3 0 0>, rot=<0 0 0 1>}
 dyn_rapel_high_r_MSTR {parent=spine_05, loc=<8.6135 4.3911 6.4009>, rot=<0.5478 0.2125 0.7029 -0.4009>}
 dyn_rapel_high_r {parent=dyn_rapel_high_r_MSTR, loc=<0.3 0.0003 0.0003>, rot=<0 0 0 1>}
 dyn_lapel_low_l_MSTR {parent=spine_05, loc=<-3.5153 12.0071 -6.7625>, rot=<0.6005 -0.5975 -0.3809 -0.3705>}
 dyn_lapel_low_l {parent=dyn_lapel_low_l_MSTR, loc=<0.3 0 0>, rot=<0 0 0 1>}
 dyn_rapel_row_r_MSTR {parent=spine_05, loc=<-3.5137 12.0019 6.5601>, rot=<0.3705 0.3809 0.5975 -0.6005>}
 dyn_rapel_row_r {parent=dyn_rapel_row_r_MSTR, loc=<0.2999 0.0007 0>, rot=<0 0 0 1>}
 Dyn_Master_Necklace {parent=spine_05, loc=<7.9916 5.0355 0>, rot=<0 0 0.966 0.2585>}
 Dyn_Necklace1 {parent=Dyn_Master_Necklace, loc=<0.3327 0 0>, rot=<0 0 0 1>}
 dyn_coat_l_MSTR {parent=spine_04, loc=<-0.3926 11.5658 -6.6781>, rot=<0.0881 -0.8693 -0.4851 0.0335>}
 dyn_coat_l_01 {parent=dyn_coat_l_MSTR, loc=<0.3 0 0>, rot=<0 0 0 1>}
 dyn_coat_l_02 {parent=dyn_coat_l_01, loc=<11.567 0 0>, rot=<-0.2943 0.076 0.003 -0.9527>}
 dyn_coat_l_03 {parent=dyn_coat_l_02, loc=<11.3331 -2.6511 1.5274>, rot=<-0.3806 0.0292 0.0472 -0.9231>}
 dyn_coat_l_04 {parent=dyn_coat_l_03, loc=<15.8782 0 0>, rot=<0 -0.0045 0 -1>}
 dyn_coat_l_05 {parent=dyn_coat_l_04, loc=<20.2787 0 0.0147>, rot=<0 -0.016 0 -0.9999>}
 dyn_coat_l_06 {parent=dyn_coat_l_05, loc=<18.03 0 0>, rot=<0 0 0 1>}
 dyn_coat_r_MSTR {parent=spine_04, loc=<-0.4119 11.5722 6.7235>, rot=<-0.0335 0.4851 0.8693 -0.0881>}
 dyn_coat_r_01 {parent=dyn_coat_r_MSTR, loc=<0.2997 0 -0.0001>, rot=<0 0 0 1>}
 dyn_coat_r_02 {parent=dyn_coat_r_01, loc=<11.5674 0 0>, rot=<0.2943 0.076 -0.003 -0.9527>}
 dyn_coat_r_03 {parent=dyn_coat_r_02, loc=<11.3033 2.6787 1.5236>, rot=<0.3806 0.0292 -0.0472 -0.9231>}
 dyn_coat_r_04 {parent=dyn_coat_r_03, loc=<15.8781 0 0>, rot=<0 -0.0045 0 -1>}
 dyn_coat_r_05 {parent=dyn_coat_r_04, loc=<20.2787 0 0.0148>, rot=<0 -0.016 0 -0.9999>}
 dyn_coat_r_06 {parent=dyn_coat_r_05, loc=<18.0299 0 0>, rot=<0 0 0 -1>}
 dyn_coat_bk_l_MSTR {parent=spine_01, loc=<2.0638 -5.4204 -9.521>, rot=<0.069 -0.8982 0.416 -0.124>}
 dyn_coat_bk_l_01 {parent=dyn_coat_bk_l_MSTR, loc=<0.3 0 0>, rot=<0 0 0 1>}
 dyn_coat_bk_l_02 {parent=dyn_coat_bk_l_01, loc=<11.0848 0 0>, rot=<0 -0.0641 0 -0.9979>}
 dyn_coat_bk_l_04 {parent=dyn_coat_bk_l_02, loc=<14.588 0 0.0559>, rot=<0 -0.0116 0 -0.9999>}
 dyn_coat_bk_l_05 {parent=dyn_coat_bk_l_04, loc=<13.6078 0 0>, rot=<0 0.036 0 -0.9994>}
 dyn_coat_bk_l_06 {parent=dyn_coat_bk_l_05, loc=<14.9961 0 0>, rot=<0 0 0 1>}
 dyn_coat_bk_MSTR {parent=spine_01, loc=<1.0729 -9.4669 0>, rot=<0.1027 -0.6996 0.6996 -0.1027>}
 dyn_coat_bk_01 {parent=dyn_coat_bk_MSTR, loc=<0.3 0 0>, rot=<0 0 0 1>}
 dyn_coat_bk_02 {parent=dyn_coat_bk_01, loc=<10.4809 0 0>, rot=<0 -0.064 0 -0.998>}
 dyn_coat_bk_04 {parent=dyn_coat_bk_02, loc=<12.8611 0 0.2968>, rot=<0 -0.014 0 -0.9999>}
 dyn_coat_bk_split_01 {parent=dyn_coat_bk_04, loc=<15.1935 0 -0.4447>, rot=<0 -0.0025 0 -1>}
 dyn_coat_bk_split_02 {parent=dyn_coat_bk_split_01, loc=<13.982 0 0>, rot=<0 0 0 1>}
 dyn_coat_bk_r_MSTR {parent=spine_01, loc=<2.0635 -5.4204 9.521>, rot=<0.124 -0.416 0.8982 -0.069>}
 dyn_coat_bk_r_01 {parent=dyn_coat_bk_r_MSTR, loc=<0.3 0 0>, rot=<0 0 0 1>}
 dyn_coat_bk_r_02 {parent=dyn_coat_bk_r_01, loc=<11.0844 0 0.0001>, rot=<0 -0.0641 0 -0.9979>}
 dyn_coat_bk_r_04 {parent=dyn_coat_bk_r_02, loc=<14.588 0 0.0559>, rot=<0 -0.0116 0 -0.9999>}
 dyn_coat_bk_r_05 {parent=dyn_coat_bk_r_04, loc=<13.6079 0 -0.0001>, rot=<0 0.036 0 -0.9994>}
 dyn_coat_bk_r_06 {parent=dyn_coat_bk_r_05, loc=<14.9961 -0.0001 0>, rot=<0 0 0 1>}
 thigh_l {parent=pelvis, loc=<-1.0855 -0.2109 -8.7053>, rot=<0.0009 0.0355 0.0155 -0.9993>}
 calf_l {parent=thigh_l, loc=<-42.5284 -0.0019 -0.7297>, rot=<0.0006 -0.0161 -0.0391 -0.9991>}
 calf_twist_01_l {parent=calf_l, loc=<-35.8483 -1.9197 -0.5794>, rot=<0.0237 0.0095 -0.0035 -0.9997>}
 calf_twist_02_l {parent=calf_l, loc=<-19.1546 -1.7943 -0.2638>, rot=<0.0237 0.0095 -0.0035 -0.9997>}
 foot_l {parent=calf_l, loc=<-46.7402 -2.5014 -0.592>, rot=<0.0427 -0.0238 0.0353 -0.9982>}
 ball_l {parent=foot_l, loc=<-3.5658 10.8066 -0.1544>, rot=<0.0001 -0.0001 -0.7063 -0.7079>}
 thigh_twist_01_l {parent=thigh_l, loc=<-27.952 0 0>, rot=<0.0013 0 0 -1>}
 thigh_r {parent=pelvis, loc=<-1.0853 -0.2109 8.7053>, rot=<0.0355 -0.0009 0.9993 0.0155>}
 calf_r {parent=thigh_r, loc=<42.5286 0.0019 0.7297>, rot=<0.0006 -0.0161 -0.0391 -0.9991>}
 calf_twist_01_r {parent=calf_r, loc=<35.8485 1.9197 0.5794>, rot=<0.0237 0.0095 -0.0035 -0.9997>}
 calf_twist_02_r {parent=calf_r, loc=<19.1546 1.7943 0.2638>, rot=<0.0237 0.0095 -0.0035 -0.9997>}
 foot_r {parent=calf_r, loc=<46.7403 2.5014 0.592>, rot=<0.0427 -0.0238 0.0353 -0.9982>}
 ball_r {parent=foot_r, loc=<3.5658 -10.8066 0.1544>, rot=<0.0001 -0.0001 -0.7063 -0.7079>}
 thigh_twist_01_r {parent=thigh_r, loc=<27.952 0 0>, rot=<0.0013 0 0 -1>}
 dyn_jacket {parent=pelvis, loc=<3.663 -0.7698 0>, rot=<-0.0127 -0.707 0.0127 -0.707>}
 dyn_jacketSide_mstr_l {parent=dyn_jacket, loc=<12.4347 -0.4538 3.1408>, rot=<0.0433 -0.6274 -0.0248 -0.7771>}
 dyn_jacketSide_01_l {parent=dyn_jacketSide_mstr_l, loc=<0 0 0>, rot=<0 0 0 1>}
 dyn_jacketSide_02_l {parent=dyn_jacketSide_01_l, loc=<11.4368 -0.2896 0.9164>, rot=<0.0158 -0.0827 0.0171 -0.9963>}
 dyn_jacketSide_03_l {parent=dyn_jacketSide_02_l, loc=<21.8442 -1.1704 2.1486>, rot=<-0.0063 -0.032 0.0016 -0.9995>}
 dyn_jacketSide_04_l {parent=dyn_jacketSide_03_l, loc=<15.8676 -1.4681 1.7809>, rot=<0.0084 -0.0171 -0.0388 -0.9991>}
 dyn_jacketSide_05_l {parent=dyn_jacketSide_04_l, loc=<11.7224 -2.2079 1.5321>, rot=<0 0 0 1>}
 dyn_jacketBack_mstr_l {parent=dyn_jacket, loc=<4.969 -9.3491 3.832>, rot=<-0.3935 -0.4753 0.4506 -0.6451>}
 dyn_jacketBack_01_l {parent=dyn_jacketBack_mstr_l, loc=<0 0 0>, rot=<0 0 0.0219 -0.9998>}
 dyn_jacketBack_02_l {parent=dyn_jacketBack_01_l, loc=<11.5489 0.6957 -0.6835>, rot=<0.027 -0.083 0.0026 -0.9962>}
 dyn_jacketBack_03_l {parent=dyn_jacketBack_02_l, loc=<20.4637 -0.7727 0.4371>, rot=<-0.0184 -0.0221 -0.0008 -0.9996>}
 dyn_jacketBack_04_l {parent=dyn_jacketBack_03_l, loc=<16.7184 -1.2439 1.1788>, rot=<-0.0016 -0.0541 -0.0001 -0.9985>}
 dyn_jacketBack_05_l {parent=dyn_jacketBack_04_l, loc=<9.3159 -0.7562 1.3733>, rot=<0 0 0 1>}
 dyn_jacketBack_mstr_r {parent=dyn_jacket, loc=<-4.97 -9.3589 3.8373>, rot=<-0.4842 -0.3825 0.6354 -0.4643>}
 dyn_jacketBack_01_r {parent=dyn_jacketBack_mstr_r, loc=<0 0 0>, rot=<0 0 -0.022 -0.9998>}
 dyn_jacketBack_02_r {parent=dyn_jacketBack_01_r, loc=<11.5098 -1.2062 -0.6269>, rot=<-0.027 -0.083 -0.0026 -0.9962>}
 dyn_jacketBack_03_r {parent=dyn_jacketBack_02_r, loc=<20.467 -0.1231 0.5492>, rot=<-0.001 -0.0323 0 -0.9995>}
 dyn_jacketBack_04_r {parent=dyn_jacketBack_03_r, loc=<16.7322 0.5505 1.4506>, rot=<0.0016 -0.054 0.0001 -0.9985>}
 dyn_jacketBack_05_r {parent=dyn_jacketBack_04_r, loc=<9.3033 0.1169 1.6572>, rot=<0 0 0 1>}
 dyn_jacketSide_mstr_r {parent=dyn_jacket, loc=<-12.6758 -0.3727 3.1622>, rot=<-0.6282 0.0333 0.7773 0.0124>}
 dyn_jacketSide_01_r {parent=dyn_jacketSide_mstr_r, loc=<0 0 0>, rot=<0 0 -0.0122 -0.9999>}
 dyn_jacketSide_02_r {parent=dyn_jacketSide_01_r, loc=<11.4097 0.1614 0.6463>, rot=<0.0553 -0.0805 -0.0078 -0.9952>}
 dyn_jacketSide_03_r {parent=dyn_jacketSide_02_r, loc=<21.8428 0.761 2.3443>, rot=<0.0063 -0.0317 -0.0014 -0.9995>}
 dyn_jacketSide_04_r {parent=dyn_jacketSide_03_r, loc=<15.8619 1.1607 2.0582>, rot=<-0.0093 -0.0217 0.0025 -0.9997>}
 dyn_jacketSide_05_r {parent=dyn_jacketSide_04_r, loc=<11.7201 1.9127 1.8917>, rot=<0 0 0 1>}
 dyn_jacketFront_mstr_l {parent=dyn_jacket, loc=<8.4988 9.0234 4.3044>, rot=<0.017 -0.6449 -0.0574 -0.7619>}
 dyn_jacketFront_01_l {parent=dyn_jacketFront_mstr_l, loc=<0 0 0>, rot=<0 0 0 1>}
 dyn_jacketFront_02_l {parent=dyn_jacketFront_01_l, loc=<14.1151 -1.3085 0.8109>, rot=<0.0198 -0.0696 0.0711 -0.9948>}
 dyn_jacketFront_03_l {parent=dyn_jacketFront_02_l, loc=<20.7528 -0.0253 1.8669>, rot=<-0.0062 0.0009 0.0018 -1>}
 dyn_jacketFront_04_l {parent=dyn_jacketFront_03_l, loc=<16.4804 0.2375 1.2577>, rot=<0.0093 -0.0031 -0.0033 -0.9999>}
 dyn_jacketFront_05_l {parent=dyn_jacketFront_04_l, loc=<18.5816 0.057 1.7045>, rot=<0 0 0 1>}
 dyn_jacketFront_mstr_r {parent=dyn_jacket, loc=<-8.5566 9.1359 4.3105>, rot=<-0.6451 0.0148 0.7631 0.0357>}
 dyn_jacketFront_01_r {parent=dyn_jacketFront_mstr_r, loc=<0 0 0>, rot=<0 0 0 1>}
 dyn_jacketFront_02_r {parent=dyn_jacketFront_01_r, loc=<14.1066 1.1469 0.8742>, rot=<0.0506 -0.0555 -0.0605 -0.9953>}
 dyn_jacketFront_03_r {parent=dyn_jacketFront_02_r, loc=<20.7593 -0.2887 1.7687>, rot=<0.0063 0.0012 -0.0016 -1>}
 dyn_jacketFront_04_r {parent=dyn_jacketFront_03_r, loc=<16.4839 -0.4483 1.142>, rot=<-0.0093 -0.0036 0.0027 -0.9999>}
 dyn_jacketFront_05_r {parent=dyn_jacketFront_04_r, loc=<18.583 -0.3632 1.6367>, rot=<0 0 0 1>}
]
