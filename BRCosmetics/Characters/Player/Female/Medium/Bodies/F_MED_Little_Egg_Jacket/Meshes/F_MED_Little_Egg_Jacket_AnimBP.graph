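# Animation Blueprint for the female medium Little Egg Jacket cosmetic, adding secondary motion to coat, lapels, bracelets and necklace.
# It copies the attached parent's pose, then applies male-jacket Control Rigs and lapel AnimDynamics into a startup cache.
# Cached branches drive coat tails with trail controllers or RigidBody simulation plus leg constraints.
# State-driven bone offsets correct skydiving, crouching, surface swimming and raised arms, with AnimDynamics swinging the jewelry.
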AnimBlueprint /BRCosmetics/Characters/Player/Female/Medium/Bodies/F_MED_Little_Egg_Jacket/Meshes/F_MED_Little_Egg_Jacket_AnimBP
target_skeleton=&/F_MED_Little_Egg_Jacket_Skeleton
graphs [
 AnimGraph {nodes [
   Blend Poses by bool(BlendPose_1=Use cached pose 'trailCache'(), bActiveValue=Get bRigidBodyEnabled()) {blendType=Linear}
   n1 = "Constraint - Dyn_MasterBracelet_r (PPPPPPPP)"!(
    ComponentPose
    Alpha=1
    ConstraintWeights_0=0.0177
    ConstraintWeights_1:float=0.7906
    ConstraintWeights_2:float=0.0271
    ConstraintWeights_3:float=0.1604
    ConstraintWeights_4:float=0.0007
    ConstraintWeights_5:float=0.0008
    ConstraintWeights_6:float=0.0006
    ConstraintWeights_7:float=0.0021
   ) -> (Pose) {
    alphaBoolBlend {blendOption=Linear}
    boneToModify {boneName=Dyn_MasterBracelet_r}
    constraintSetup [
     {targetBone {boneName=lowerarm_twist_02_r}}
     {targetBone {boneName=lowerarm_twist_01_r}}
     {targetBone {boneName=lowerarm_r}}
     {targetBone {boneName=hand_r}}
     {targetBone {boneName=ring_metacarpal_r}}
     {targetBone {boneName=middle_metacarpal_r}}
     {targetBone {boneName=index_metacarpal_r}}
     {targetBone {boneName=thumb_01_r}}
    ]
    constraintWeights [1,1,1,1,1,1,1,1]
   }
   "Constraint - Dyn_MasterBracelet_l (PPPPPPPP)"!(
    ComponentPose=@n1
    Alpha=1
    ConstraintWeights_0=0.1335
    ConstraintWeights_1:float=0.0003
    ConstraintWeights_2:float=0.0003
    ConstraintWeights_3:float=0.0025
    ConstraintWeights_4:float=0.0002
    ConstraintWeights_5:float=0.817
    ConstraintWeights_6:float=0.0292
    ConstraintWeights_7:float=0.0169
   ) -> (Pose) {
    alphaBoolBlend {blendOption=Linear}
    boneToModify {boneName=Dyn_MasterBracelet_l}
    constraintSetup [
     {targetBone {boneName=hand_l}}
     {targetBone {boneName=ring_metacarpal_l}}
     {targetBone {boneName=middle_metacarpal_l}}
     {targetBone {boneName=thumb_01_l}}
     {targetBone {boneName=pinky_metacarpal_l}}
     {targetBone {boneName=lowerarm_twist_01_l}}
     {targetBone {boneName=lowerarm_l}}
     {targetBone {boneName=lowerarm_twist_02_l}}
    ]
    constraintWeights [1,1,1,1,1,1,1,1]
   }
   n2 = "Transform (Modify) Bone - Bone: dyn_coat_bk_r_02"(Alpha=1, Translation="0,0,10", Rotation="0,0,0", Scale="1,1,1") {
    translationMode=BMM_Additive
    translationSpace=BCS_BoneSpace
    alphaBoolBlend {blendOption=Linear}
    boneToModify {boneName=dyn_coat_bk_r_02}
    translation {z=10}
   }
   n3 = "Transform (Modify) Bone - Bone: dyn_coat_bk_02"(ComponentPose=@n2, Alpha=1, Translation="0,0,10", Rotation="0,0,0", Scale="1,1,1") {
    translationMode=BMM_Additive
    translationSpace=BCS_BoneSpace
    alphaBoolBlend {blendOption=Linear}
    boneToModify {boneName=dyn_coat_bk_02}
    translation {z=10}
   }
   n4 = "Transform (Modify) Bone - Bone: dyn_coat_bk_l_02"(ComponentPose=@n3, Alpha=1, Translation="0,0,10", Rotation="0,0,0", Scale="1,1,1") {
    translationMode=BMM_Additive
    translationSpace=BCS_BoneSpace
    alphaBoolBlend {blendOption=Linear}
    boneToModify {boneName=dyn_coat_bk_l_02}
    translation {z=10}
   }
   n5 = "Transform (Modify) Bone - Bone: dyn_coat_side_l_01"(ComponentPose=@n4, Alpha=1, Translation="0,0,10", Rotation="0,0,0", Scale="1,1,1") {
    translationMode=BMM_Additive
    translationSpace=BCS_BoneSpace
    alphaBoolBlend {blendOption=Linear}
    boneToModify {boneName=dyn_coat_side_l_01}
    translation {z=10}
   }
   "Transform (Modify) Bone - Bone: dyn_coat_side_r_01"(ComponentPose=@n5, Alpha=1, Translation="0,0,10", Rotation="0,0,0", Scale="1,1,1") {
    translationMode=BMM_Additive
    translationSpace=BCS_BoneSpace
    alphaBoolBlend {blendOption=Linear}
    boneToModify {boneName=dyn_coat_side_r_01}
    translation {z=10}
   }
   n6 = "Constraint - Dyn_Master_Necklace (PPP)"!(
    ComponentPose=Local To Component(LocalPose=Use cached pose 'RbanCache'())
    Alpha=1
    ConstraintWeights_0=0.934
    ConstraintWeights_1:float=0.0313
    ConstraintWeights_2:float=0.0347
   ) -> (Pose) {
    alphaBoolBlend {blendOption=Linear}
    boneToModify {boneName=Dyn_Master_Necklace}
    constraintSetup [{targetBone {boneName=spine_05}}, {targetBone {boneName=clavicle_l}}, {targetBone {boneName=clavicle_r}}]
    constraintWeights [1,1,1]
   }
   n7 = "Transform (Modify) Bone - Bone: dyn_lapel_high_l_MSTR"(ComponentPose=@n6, Alpha=Get Left Arm Up() -> (Left Arm Up:double), Translation="0,0,5", Rotation="30,0,0", Scale="0.8,0.8,0.8") {
    rotationMode=BMM_Additive
    rotationSpace=BCS_BoneSpace
    scaleMode=BMM_Additive
    scaleSpace=BCS_BoneSpace
    translationMode=BMM_Additive
    translationSpace=BCS_BoneSpace
    alphaBoolBlend {blendOption=Linear}
    boneToModify {boneName=dyn_lapel_high_l_MSTR}
    rotation {pitch=30}
    scale=<0.8 0.8 0.8>
    translation {z=5}
   }
   n8 = "Transform (Modify) Bone - Bone: dyn_rapel_high_r_MSTR"(ComponentPose=@n7, Alpha=Get Right Arm Up() -> (Right Arm Up:double), Translation="0,0,5", Rotation="50,0,0", Scale="0.8,0.8,0.8") {
    rotationMode=BMM_Additive
    rotationSpace=BCS_BoneSpace
    scaleMode=BMM_Additive
    scaleSpace=BCS_BoneSpace
    translationMode=BMM_Additive
    translationSpace=BCS_BoneSpace
    alphaBoolBlend {blendOption=Linear}
    boneToModify {boneName=dyn_rapel_high_r_MSTR}
    rotation {pitch=50}
    scale=<0.8 0.8 0.8>
    translation {z=5}
   }
   n9 = Get GravityOvcride() -> (GravityOvcride:Vector)
   n10 = Anim Dynamics - Bone: Dyn_Bracelet_r1!(ComponentPose=@n8, Alpha=1, GravityOverride:Vector=@n9, ExternalForce="0,0,0") -> (Pose) {
    bOverrideLinearDamping=true
    bUseGravityOverride=true
    linearDampingOverride=0.9998
    simulationSpace=RootRelative
    alphaBoolBlend {blendOption=Linear}
    boundBone {boneName=Dyn_Bracelet_r1}
    componentAppliedLinearAccClamp=<3000 3000 3000>
    componentLinearAccScale=<0.2 0.2 0.35>
    componentLinearVelScale=<0.2 0.2 0.35>
    gravityOverride {z=-980}
    physicsBodyDefinitions [{
      boundBone {boneName=Dyn_Bracelet_r1}
      boxExtents=<20 10 20>
      constraintSetup {angularLimitsMax=<25 55 25>, angularLimitsMin=<-25 -55 -55>, angularTarget=<0 0 0>}
      localJointOffset=<3 0 0>
     }]
   }
   n11 = Anim Dynamics - Bone: Dyn_Bracelet_l1!(ComponentPose=@n10, Alpha=1, GravityOverride:Vector=@n9, ExternalForce="0,0,0") -> (Pose) {
    bOverrideLinearDamping=true
    bUseGravityOverride=true
    bUsePlanarLimit=false
    linearDampingOverride=0.9998
    simulationSpace=RootRelative
    alphaBoolBlend {blendOption=Linear}
    boundBone {boneName=Dyn_Bracelet_l1}
    componentAppliedLinearAccClamp=<3000 3000 3000>
    componentLinearAccScale=<0.2 0.2 0.35>
    componentLinearVelScale=<0.2 0.2 0.35>
    gravityOverride {z=-980}
    physicsBodyDefinitions [{
      boundBone {boneName=Dyn_Bracelet_l1}
      boxExtents=<20 10 20>
      constraintSetup {angularLimitsMax=<25 55 25>, angularLimitsMin=<-25 -55 -55>, angularTarget=<0 0 0>}
      localJointOffset=<3 0 0>
     }]
    sphericalLimits [{limitRadius=5, drivingBone {boneName=hand_l}, sphereLocalOffset=<0 0 0>}]
   }
   n12 = Anim Dynamics - Bone: Dyn_Master_Necklace!(ComponentPose=@n11, Alpha=1, GravityOverride:Vector=@n9, ExternalForce="0,0,0") -> (Pose) {
    bOverrideLinearDamping=true
    bUseGravityOverride=true
    bUsePlanarLimit=false
    linearDampingOverride=0.9999
    simulationSpace=RootRelative
    alphaBoolBlend {blendOption=Linear}
    boundBone {boneName=Dyn_Master_Necklace}
    componentAppliedLinearAccClamp=<3000 3000 3000>
    componentLinearAccScale=<0.2 0.2 0.25>
    componentLinearVelScale=<0.2 0.2 0.25>
    gravityOverride {z=-980}
    physicsBodyDefinitions [{
      boundBone {boneName=Dyn_Master_Necklace}
      boxExtents=<25 10 25>
      constraintSetup {angularLimitsMax=<1 10 1>, angularLimitsMin=<-1 -10 -30>, angularTarget=<0 0 0>}
      localJointOffset=<4 0 0>
     }]
    sphericalLimits [{limitRadius=5, drivingBone {boneName=hand_l}, sphereLocalOffset=<0 0 0>}]
   }
   n13 = "Transform (Modify) Bone - Bone: dyn_rapel_high_r_MSTR"!(ComponentPose=@n12, bAlphaBoolEnabled:bool=True, Translation="0,2,1", Rotation="-3,0,-10", Scale="1,1,1") -> (Pose) {
    alphaInputType=Bool
    rotationMode=BMM_Additive
    rotationSpace=BCS_BoneSpace
    translationMode=BMM_Additive
    translationSpace=BCS_BoneSpace
    alphaBoolBlend {blendOption=Linear}
    boneToModify {boneName=dyn_rapel_high_r_MSTR}
    rotation {pitch=-3, roll=-10}
    translation {y=2, z=1}
   }
   n14 = "Transform (Modify) Bone - Bone: dyn_lapel_high_l_MSTR"!(ComponentPose=@n13, bAlphaBoolEnabled:bool=True, Translation="0,0,1.5", Rotation="0,0,5", Scale="1,1,1") -> (Pose) {
    alphaInputType=Bool
    rotationMode=BMM_Additive
    rotationSpace=BCS_BoneSpace
    translationMode=BMM_Additive
    translationSpace=BCS_BoneSpace
    alphaBoolBlend {blendOption=Linear}
    boneToModify {boneName=dyn_lapel_high_l_MSTR}
    rotation {roll=5}
    translation {z=1.5}
   }
   n15 = Get bIsSurfaceSwimming()
   n16 = "Constraint - dyn_coat_bk_r_02 (P)"!(ComponentPose=@n14, bAlphaBoolEnabled:bool=@n15, ConstraintWeights_0=1) -> (Pose) {
    alphaInputType=Bool
    alphaBoolBlend {blendOption=Linear}
    boneToModify {boneName=dyn_coat_bk_r_02}
    constraintSetup [{targetBone {boneName=pelvis}}]
    constraintWeights [1]
   }
   n17 = "Constraint - dyn_coat_bk_l_02 (P)"!(ComponentPose=@n16, bAlphaBoolEnabled:bool=@n15, ConstraintWeights_0=1) -> (Pose) {
    alphaInputType=Bool
    alphaBoolBlend {blendOption=Linear}
    boneToModify {boneName=dyn_coat_bk_l_02}
    constraintSetup [{targetBone {boneName=pelvis}}]
    constraintWeights [1]
   }
   n18 = "Constraint - dyn_coat_bk_02 (P)"!(ComponentPose=@n17, bAlphaBoolEnabled:bool=@n15, ConstraintWeights_0=1) -> (Pose) {
    alphaInputType=Bool
    alphaBoolBlend {blendOption=Linear}
    boneToModify {boneName=dyn_coat_bk_02}
    constraintSetup [{targetBone {boneName=pelvis}}]
    constraintWeights [1]
   }
   n19 = "Transform (Modify) Bone - Bone: dyn_coat_bk_02"!(ComponentPose=@n18, bAlphaBoolEnabled:bool=@n15, Translation="0,0,0.5", Rotation="5,0,0", Scale="1,1,1") -> (Pose) {
    alphaInputType=Bool
    rotationMode=BMM_Additive
    rotationSpace=BCS_BoneSpace
    scaleMode=BMM_Additive
    scaleSpace=BCS_BoneSpace
    translationMode=BMM_Additive
    translationSpace=BCS_BoneSpace
    alphaBoolBlend {blendOption=Linear}
    boneToModify {boneName=dyn_coat_bk_02}
    rotation {pitch=5}
    translation {z=0.5}
   }
   n20 = "Transform (Modify) Bone - Bone: dyn_coat_bk_l_02"!(ComponentPose=@n19, bAlphaBoolEnabled:bool=@n15, Translation="0,0,0.5", Rotation="5,0,0", Scale="1,1,1") -> (Pose) {
    alphaInputType=Bool
    rotationMode=BMM_Additive
    rotationSpace=BCS_BoneSpace
    scaleMode=BMM_Additive
    scaleSpace=BCS_BoneSpace
    translationMode=BMM_Additive
    translationSpace=BCS_BoneSpace
    alphaBoolBlend {blendOption=Linear}
    boneToModify {boneName=dyn_coat_bk_l_02}
    rotation {pitch=5}
    translation {z=0.5}
   }
   n21 = "Transform (Modify) Bone - Bone: dyn_coat_bk_r_02"!(ComponentPose=@n20, bAlphaBoolEnabled:bool=@n15, Translation="0,0,0.5", Rotation="5,0,0", Scale="1,1,1") -> (Pose) {
    alphaInputType=Bool
    rotationMode=BMM_Additive
    rotationSpace=BCS_BoneSpace
    scaleMode=BMM_Additive
    scaleSpace=BCS_BoneSpace
    translationMode=BMM_Additive
    translationSpace=BCS_BoneSpace
    alphaBoolBlend {blendOption=Linear}
    boneToModify {boneName=dyn_coat_bk_r_02}
    rotation {pitch=5}
    translation {z=0.5}
   }
   n22 = "Transform (Modify) Bone - Bone: dyn_coat_r_02"(ComponentPose=@n21, Alpha=Get FE Coat Alpha() -> (FE Coat Alpha:double), Translation="-0.75,0,0.25", Rotation="0,0,10", Scale="1,1,1") {
    rotationMode=BMM_Additive
    rotationSpace=BCS_BoneSpace
    translationMode=BMM_Additive
    translationSpace=BCS_BoneSpace
    alphaBoolBlend {blendOption=Linear}
    boneToModify {boneName=dyn_coat_r_02}
    rotation {roll=10}
    translation {x=-0.75, z=0.25}
   }
   Output Pose(Result=Component To Local(ComponentPose=@n22)) {layerGroup=None}
   n23 = Control Rig(Source=Copy Pose From Mesh() {bUseAttachedParent=true}) {
    alphaBoolBlend {blendOption=Linear}
    controlRigAssetReference {blueprintRigClass=&/M_MED_Little_Egg_Jacket_Deform_CtrlRig_twists.M_MED_Little_Egg_Jacket_Deform_CtrlRig_twists_C}
   }
   n24 = Anim Dynamics - Bone: dyn_lapel_l(ComponentPose=Local To Component(LocalPose=@n23), Alpha=1, ExternalForce="0,0,0") {
    angularDampingOverride=1
    angularSpringConstant=20000
    bAngularSpring=true
    bOverrideAngularDamping=true
    bUseGravityOverride=true
    gravityScale=0.35
    simulationSpace=RootRelative
    alphaBoolBlend {blendOption=Linear}
    boundBone {boneName=dyn_lapel_l}
    componentLinearAccScale=<1 1 1>
    componentLinearVelScale=<1 1 1>
    gravityOverride {z=-980}
    physicsBodyDefinitions [{
      sphereCollisionRadius=0
      boundBone {boneName=dyn_lapel_l}
      boxExtents=<20 30 30>
      constraintSetup {angularLimitsMax=<0 0 0>, angularLimitsMin=<0 -30 0>, angularTarget=<1 0 0>}
      localJointOffset=<4 0 0>
     }]
   }
   n25 = Anim Dynamics - Bone: dyn_rapel_r(ComponentPose=@n24, Alpha=1, ExternalForce="0,0,0") {
    angularDampingOverride=1
    angularSpringConstant=20000
    bAngularSpring=true
    bOverrideAngularDamping=true
    bUseGravityOverride=true
    gravityScale=0.35
    simulationSpace=RootRelative
    alphaBoolBlend {blendOption=Linear}
    boundBone {boneName=dyn_rapel_r}
    componentLinearAccScale=<1 1 1>
    componentLinearVelScale=<1 1 1>
    gravityOverride {z=-980}
    physicsBodyDefinitions [{
      sphereCollisionRadius=0
      boundBone {boneName=dyn_rapel_r}
      boxExtents=<20 30 30>
      constraintSetup {angularLimitsMax=<0 0 0>, angularLimitsMin=<0 -30 0>, angularTarget=<1 0 0>}
      localJointOffset=<4 0 0>
     }]
   }
   n26 = Anim Dynamics - Bone: dyn_lapel_high_l(ComponentPose=@n25, Alpha=1, ExternalForce="0,0,0") {
    angularDampingOverride=1
    angularSpringConstant=20000
    bAngularSpring=true
    bOverrideAngularDamping=true
    bUseGravityOverride=true
    gravityScale=0.35
    simulationSpace=RootRelative
    alphaBoolBlend {blendOption=Linear}
    boundBone {boneName=dyn_lapel_high_l}
    componentLinearAccScale=<1 1 1>
    componentLinearVelScale=<1 1 1>
    gravityOverride {z=-980}
    physicsBodyDefinitions [{
      sphereCollisionRadius=0
      boundBone {boneName=dyn_lapel_high_l}
      boxExtents=<20 30 10>
      constraintSetup {angularLimitsMax=<0 0 0>, angularLimitsMin=<0 -30 0>, angularTarget=<1 0 0>}
      localJointOffset=<6 0 0>
     }]
    relativeSpaceBone {boneName=clavicle_l}
   }
   n27 = Anim Dynamics - Bone: dyn_rapel_high_r(ComponentPose=@n26, Alpha=1, ExternalForce="0,0,0") {
    angularDampingOverride=1
    angularSpringConstant=20000
    bAngularSpring=true
    bOverrideAngularDamping=true
    bUseGravityOverride=true
    gravityScale=0.35
    simulationSpace=RootRelative
    alphaBoolBlend {blendOption=Linear}
    boundBone {boneName=dyn_rapel_high_r}
    componentLinearAccScale=<1 1 1>
    componentLinearVelScale=<1 1 1>
    gravityOverride {z=-980}
    physicsBodyDefinitions [{
      sphereCollisionRadius=0
      boundBone {boneName=dyn_rapel_high_r}
      boxExtents=<20 30 30>
      constraintSetup {angularLimitsMax=<0 0 0>, angularLimitsMin=<0 -30 0>, angularTarget=<1 0 0>}
      localJointOffset=<6 0 0>
     }]
   }
   n28 = Anim Dynamics - Bone: dyn_lapel_low_l(ComponentPose=@n27, Alpha=1, ExternalForce="0,0,0") {
    angularDampingOverride=1
    angularSpringConstant=20000
    bAngularSpring=true
    bOverrideAngularDamping=true
    bUseGravityOverride=true
    gravityScale=0.35
    simulationSpace=RootRelative
    alphaBoolBlend {blendOption=Linear}
    boundBone {boneName=dyn_lapel_low_l}
    componentLinearAccScale=<1 1 1>
    componentLinearVelScale=<1 1 1>
    gravityOverride {z=-980}
    physicsBodyDefinitions [{
      sphereCollisionRadius=0
      boundBone {boneName=dyn_lapel_low_l}
      boxExtents=<20 30 30>
      constraintSetup {angularLimitsMax=<0 0 0>, angularLimitsMin=<0 -30 0>, angularTarget=<1 0 0>}
      localJointOffset=<4 0 0>
     }]
   }
   n29 = Anim Dynamics - Bone: dyn_rapel_row_r(ComponentPose=@n28, Alpha=1, ExternalForce="0,0,0") {
    angularDampingOverride=1
    angularSpringConstant=20000
    bAngularSpring=true
    bOverrideAngularDamping=true
    bUseGravityOverride=true
    gravityScale=0.35
    simulationSpace=RootRelative
    alphaBoolBlend {blendOption=Linear}
    boundBone {boneName=dyn_rapel_row_r}
    componentLinearAccScale=<1 1 1>
    componentLinearVelScale=<1 1 1>
    gravityOverride {z=-980}
    physicsBodyDefinitions [{
      sphereCollisionRadius=0
      boundBone {boneName=dyn_rapel_row_r}
      boxExtents=<20 30 30>
      constraintSetup {angularLimitsMax=<0 0 0>, angularLimitsMin=<0 -30 0>, angularTarget=<1 0 0>}
      localJointOffset=<4 0 0>
     }]
   }
   n30 = Control Rig(Source=Component To Local(ComponentPose=@n29)) {
    alphaBoolBlend {blendOption=Linear}
    controlRigAssetReference {blueprintRigClass=&/M_MED_Little_Egg_Jacket_CtrlRig.M_MED_Little_Egg_Jacket_CtrlRig_C}
   }
   Save cached pose 'StartupCache'(Pose=@n30) {cacheName=StartupCache}
   n31 = Get Bwd Movement() -> (Bwd Movement:double)
   n32 = Get TrailControllerRot() -> (TrailControllerRot:Rotator)
   n33 = "Transform (Modify) Bone - Bone: dyn_coat_l_MSTR"(ComponentPose=Local To Component(LocalPose=Use cached pose 'StartupCache'()), Alpha=1, Translation="0,0,0", Rotation=@n32, Scale="1,1,1") {rotationMode=BMM_Additive, rotationSpace=BCS_BoneSpace, alphaBoolBlend {blendOption=Linear}, boneToModify {boneName=dyn_coat_l_MSTR}}
   n34 = "Transform (Modify) Bone - Bone: dyn_coat_bk_MSTR"(ComponentPose=@n33, Alpha=1, Translation="0,0,0", Rotation=@n32, Scale="1,1,1") {rotationMode=BMM_Additive, rotationSpace=BCS_BoneSpace, alphaBoolBlend {blendOption=Linear}, boneToModify {boneName=dyn_coat_bk_MSTR}}
   n35 = "Transform (Modify) Bone - Bone: dyn_coat_r_MSTR"(ComponentPose=@n34, Alpha=1, Translation="0,0,0", Rotation=@n32, Scale="1,1,1") {rotationMode=BMM_Additive, rotationSpace=BCS_BoneSpace, alphaBoolBlend {blendOption=Linear}, boneToModify {boneName=dyn_coat_r_MSTR}}
   n36 = Trail controller - Bone: dyn_coat_l_06(ComponentPose=@n35, Alpha=@n31) {
    bUsePlanarLimit=true
    chainLength=6
    debugLifeTime=0.1
    maxDeltaTime=0.333
    alphaScaleBiasClamp {bInterpResult=true}
    rotationLimits [
     {limitMax=<180 180 180>}
     {limitMax=<180 180 180>}
     {limitMax=<180 180 180>}
     {limitMax=<180 180 180>}
     {limitMax=<180 180 180>}
     {limitMax=<180 180 180>}
    ]
    rotationOffsets [<0 0 0>, <0 0 0>, <0 0 0>, <0 0 0>, <0 0 0>, <0 0 0>]
    trailBone {boneName=dyn_coat_l_06}
    trailRelaxationSpeed {editorCurveData {keys [<at -0.0134 19.3841 RCIM_Linear>,<at 0.5079 12.8986 RCIM_Linear>,<at 0.9955 9.6044 RCIM_Linear>]}}
   }
   n37 = Trail controller - Bone: dyn_coat_bk_04(ComponentPose=@n36, Alpha=@n31) {
    bUsePlanarLimit=true
    chainLength=3
    debugLifeTime=0.1
    maxDeltaTime=0.0333
    alphaScaleBiasClamp {bInterpResult=true}
    rotationLimits [{limitMax=<180 180 180>}, {limitMax=<180 180 180>}, {limitMax=<180 90 180>, limitMin=<-180 -90 -180>}]
    rotationOffsets [<0 0 0>, <0 0 0>, <0 0 0>]
    trailBone {boneName=dyn_coat_bk_04}
    trailRelaxationSpeed {editorCurveData {keys [<at 0 15 RCIM_Linear>,<at 0.5289 13.1838 RCIM_Linear>,<at 1 6 RCIM_Linear>]}}
   }
   n38 = Trail controller - Bone: dyn_coat_r_06(ComponentPose=@n37, Alpha=@n31) {
    bUsePlanarLimit=true
    chainLength=6
    debugLifeTime=0.1
    maxDeltaTime=0.0333
    alphaScaleBiasClamp {bInterpResult=true}
    rotationLimits [
     {limitMax=<180 180 180>}
     {limitMax=<180 180 180>}
     {limitMax=<180 90 180>, limitMin=<-180 -90 -180>}
     {limitMax=<180 180 180>}
     {limitMax=<180 180 180>}
     {limitMax=<180 180 180>}
    ]
    rotationOffsets [<0 0 0>, <0 0 0>, <0 0 0>, <0 0 0>, <0 0 0>, <0 0 0>]
    trailBone {boneName=dyn_coat_r_06}
    trailRelaxationSpeed {editorCurveData {keys [<at 0 15 RCIM_Linear>,<at 0.5289 13.1838 RCIM_Linear>,<at 1 6 RCIM_Linear>]}}
   }
   Save cached pose 'trailCache'(Pose=Component To Local(ComponentPose=@n38)) {cacheName=trailCache}
   n39 = "Transform (Modify) Bone - Bone: dyn_coat_l_01"(ComponentPose=Local To Component(), Alpha=1, Translation="0,0,0", Rotation="5,0,0", Scale="1,1,1") {
    rotationMode=BMM_Additive
    rotationSpace=BCS_BoneSpace
    alphaBoolBlend {blendOption=Linear}
    boneToModify {boneName=dyn_coat_l_01}
    rotation {pitch=5}
   }
   n40 = "Transform (Modify) Bone - Bone: dyn_coat_l_03"(ComponentPose=@n39, Alpha=1, Translation="0,0,0", Rotation="-10,0,0", Scale="1,1,1") {
    rotationMode=BMM_Additive
    rotationSpace=BCS_BoneSpace
    alphaBoolBlend {blendOption=Linear}
    boneToModify {boneName=dyn_coat_l_03}
    rotation {pitch=-10}
   }
   n41 = "Transform (Modify) Bone - Bone: dyn_coat_l_05"(ComponentPose=@n40, Alpha=1, Translation="0,0,0", Rotation="-25,0,0", Scale="1,1,1") {
    rotationMode=BMM_Additive
    rotationSpace=BCS_BoneSpace
    alphaBoolBlend {blendOption=Linear}
    boneToModify {boneName=dyn_coat_l_05}
    rotation {pitch=-25}
   }
   n42 = "Transform (Modify) Bone - Bone: dyn_coat_bk_01"(ComponentPose=@n41, Alpha=1, Translation="0,0,0", Rotation="5,0,0", Scale="1,1,1") {
    rotationMode=BMM_Additive
    rotationSpace=BCS_BoneSpace
    alphaBoolBlend {blendOption=Linear}
    boneToModify {boneName=dyn_coat_bk_01}
    rotation {pitch=5}
   }
   n43 = "Transform (Modify) Bone - Bone: dyn_coat_bk_02"!(ComponentPose=@n42, Alpha:float=1, Translation="0,0,0", Rotation="-10,0,0", Scale="1,1,1") -> (Pose) {
    rotationMode=BMM_Additive
    rotationSpace=BCS_BoneSpace
    alphaBoolBlend {blendOption=Linear}
    boneToModify {boneName=dyn_coat_bk_02}
    rotation {pitch=-10}
   }
   n44 = "Transform (Modify) Bone - Bone: dyn_coat_bk_04"(ComponentPose=@n43, Alpha=1, Translation="0,0,0", Rotation="-25,0,0", Scale="1,1,1") {
    rotationMode=BMM_Additive
    rotationSpace=BCS_BoneSpace
    alphaBoolBlend {blendOption=Linear}
    boneToModify {boneName=dyn_coat_bk_04}
    rotation {pitch=-25}
   }
   n45 = "Transform (Modify) Bone - Bone: dyn_coat_r_01"(ComponentPose=@n44, Alpha=1, Translation="0,0,0", Rotation="5,0,0", Scale="1,1,1") {
    rotationMode=BMM_Additive
    rotationSpace=BCS_BoneSpace
    alphaBoolBlend {blendOption=Linear}
    boneToModify {boneName=dyn_coat_r_01}
    rotation {pitch=5}
   }
   n46 = "Transform (Modify) Bone - Bone: dyn_coat_r_03"(ComponentPose=@n45, Alpha=1, Translation="0,0,0", Rotation="-10,0,0", Scale="1,1,1") {
    rotationMode=BMM_Additive
    rotationSpace=BCS_BoneSpace
    alphaBoolBlend {blendOption=Linear}
    boneToModify {boneName=dyn_coat_r_03}
    rotation {pitch=-10}
   }
   n47 = "Transform (Modify) Bone - Bone: dyn_coat_r_05"(ComponentPose=@n46, Alpha=1, Translation="0,0,0", Rotation="-25,0,0", Scale="1,1,1") {
    rotationMode=BMM_Additive
    rotationSpace=BCS_BoneSpace
    alphaBoolBlend {blendOption=Linear}
    boneToModify {boneName=dyn_coat_r_05}
    rotation {pitch=-25}
   }
   n48 = Trail controller - Bone: dyn_coat_l_06!(ComponentPose=@n47, Alpha=1, RelaxationSpeedScale:float=2, FakeVelocity:Vector="0,0,0") -> (Pose) {
    bShowTrailLocation=true
    bUsePlanarLimit=true
    chainLength=6
    debugLifeTime=0.1
    maxDeltaTime=0.0333
    relaxationSpeedScale=1.2
    alphaScaleBiasClamp {bInterpResult=true, interpSpeedDecreasing=25, interpSpeedIncreasing=25}
    rotationLimits [
     {limitMax=<180 180 180>}
     {limitMax=<180 180 180>}
     {limitMax=<180 180 180>}
     {limitMax=<180 180 180>}
     {limitMax=<180 180 180>}
     {limitMax=<180 180 180>}
    ]
    rotationOffsets [<0 0 0>, <0 0 0>, <0 0 0>, <0 0 0>, <0 0 0>, <0 0 0>]
    trailBone {boneName=dyn_coat_l_06}
    trailRelaxationSpeed {editorCurveData {keys [<at 0 15 RCIM_Linear>,<at 1 6 RCIM_Linear>]}}
   }
   n49 = Trail controller - Bone: dyn_coat_bk_04!(ComponentPose=@n48, Alpha=1, RelaxationSpeedScale:float=2, FakeVelocity:Vector="0,0,0") -> (Pose) {
    bShowTrailLocation=true
    bUsePlanarLimit=true
    chainLength=3
    debugLifeTime=0.1
    maxDeltaTime=0.0333
    relaxationSpeedScale=1.2
    alphaScaleBiasClamp {bInterpResult=true, interpSpeedDecreasing=25, interpSpeedIncreasing=25}
    rotationLimits [{limitMax=<180 180 180>}, {limitMax=<180 180 180>}, {limitMax=<180 180 180>}]
    rotationOffsets [<0 0 0>, <0 0 0>, <0 0 0>]
    trailBone {boneName=dyn_coat_bk_04}
    trailRelaxationSpeed {editorCurveData {keys [<at 0 15 RCIM_Linear>,<at 1 6 RCIM_Linear>]}}
   }
   n50 = Trail controller - Bone: dyn_coat_r_06!(ComponentPose=@n49, Alpha=1, RelaxationSpeedScale:float=2, FakeVelocity:Vector="0,0,0") -> (Pose) {
    bShowTrailLocation=true
    bUsePlanarLimit=true
    chainLength=6
    debugLifeTime=0.1
    maxDeltaTime=0.0333
    relaxationSpeedScale=1.2
    alphaScaleBiasClamp {bInterpResult=true, interpSpeedDecreasing=25, interpSpeedIncreasing=25}
    rotationLimits [
     {limitMax=<180 180 180>}
     {limitMax=<180 180 180>}
     {limitMax=<180 180 180>}
     {limitMax=<180 180 180>}
     {limitMax=<180 180 180>}
     {limitMax=<180 180 180>}
    ]
    rotationOffsets [<0 0 0>, <0 0 0>, <0 0 0>, <0 0 0>, <0 0 0>, <0 0 0>]
    trailBone {boneName=dyn_coat_r_06}
    trailRelaxationSpeed {editorCurveData {keys [<at 0 15 RCIM_Linear>,<at 1 6 RCIM_Linear>]}}
   }
   Save cached pose 'dbnoCache'(Pose=Component To Local(ComponentPose=@n50)) {cacheName=dbnoCache}
   n51 = Get Skydive Alpha() -> (Skydive Alpha:double)
   n52 = "Transform (Modify) Bone - Bone: dyn_coat_l_MSTR"(
    ComponentPose=Local To Component(LocalPose=Use cached pose 'StartupCache'())
    Alpha=@n51
    Translation="0,-4,0"
    Rotation="0,-25,0"
    Scale="1,1,1"
   ) {
    rotationMode=BMM_Additive
    rotationSpace=BCS_BoneSpace
    translationMode=BMM_Additive
    translationSpace=BCS_BoneSpace
    alphaBoolBlend {blendOption=Linear}
    boneToModify {boneName=dyn_coat_l_MSTR}
    rotation {yaw=-25}
    translation {y=-4}
   }
   n53 = "Transform (Modify) Bone - Bone: dyn_coat_r_MSTR"(ComponentPose=@n52, Alpha=@n51, Translation="0,4,0", Rotation="0,25,0", Scale="1,1,1") {
    rotationMode=BMM_Additive
    rotationSpace=BCS_BoneSpace
    translationMode=BMM_Additive
    translationSpace=BCS_BoneSpace
    alphaBoolBlend {blendOption=Linear}
    boneToModify {boneName=dyn_coat_r_MSTR}
    rotation {yaw=25}
    translation {y=4}
   }
   n54 = "Transform (Modify) Bone - Bone: dyn_coat_side_l_MSTR"(ComponentPose=@n53, Alpha=@n51, Translation="0,2,2", Rotation="0,0,0", Scale="1,1,1") {
    rotationMode=BMM_Additive
    rotationSpace=BCS_BoneSpace
    translationMode=BMM_Additive
    translationSpace=BCS_BoneSpace
    alphaBoolBlend {blendOption=Linear}
    boneToModify {boneName=dyn_coat_side_l_MSTR}
    translation {y=2, z=2}
   }
   n55 = "Transform (Modify) Bone - Bone: dyn_coat_side_r_MSTR"(ComponentPose=@n54, Alpha=@n51, Translation="0,-2,2", Rotation="0,0,0", Scale="1,1,1") {
    rotationMode=BMM_Additive
    rotationSpace=BCS_BoneSpace
    translationMode=BMM_Additive
    translationSpace=BCS_BoneSpace
    alphaBoolBlend {blendOption=Linear}
    boneToModify {boneName=dyn_coat_side_r_MSTR}
    translation {y=-2, z=2}
   }
   n56 = Get CrouchFixAlpha() -> (CrouchFixAlpha:double)
   n57 = "Transform (Modify) Bone - Bone: dyn_coat_r_MSTR"(ComponentPose=@n55, Alpha=@n56, Translation="0,0,0", Rotation="15,45,0", Scale="1,1,1") {
    rotationMode=BMM_Additive
    rotationSpace=BCS_BoneSpace
    translationMode=BMM_Additive
    translationSpace=BCS_BoneSpace
    alphaBoolBlend {blendOption=Linear}
    alphaScaleBiasClamp {bInterpResult=true, interpSpeedDecreasing=4, interpSpeedIncreasing=1000}
    boneToModify {boneName=dyn_coat_r_MSTR}
    rotation {pitch=15, yaw=45}
   }
   n58 = "Transform (Modify) Bone - Bone: dyn_coat_l_MSTR"(ComponentPose=@n57, Alpha=@n56, Translation="0,0,0", Rotation="15,-45,0", Scale="1,1,1") {
    rotationMode=BMM_Additive
    rotationSpace=BCS_BoneSpace
    translationMode=BMM_Additive
    translationSpace=BCS_BoneSpace
    alphaBoolBlend {blendOption=Linear}
    alphaScaleBiasClamp {bInterpResult=true, interpSpeedDecreasing=4, interpSpeedIncreasing=1000}
    boneToModify {boneName=dyn_coat_l_MSTR}
    rotation {pitch=15, yaw=-45}
   }
   n59 = "Transform (Modify) Bone - Bone: dyn_coat_r_MSTR"(ComponentPose=@n58, Alpha=Get Thigh R Out Alpha() -> (Thigh R Out Alpha:double), Translation="0,3,3", Rotation="-15,-15,0", Scale="1,1,1") {
    rotationMode=BMM_Additive
    rotationSpace=BCS_BoneSpace
    translationMode=BMM_Additive
    translationSpace=BCS_BoneSpace
    alphaBoolBlend {blendOption=Linear}
    alphaScaleBiasClamp {interpSpeedDecreasing=4, interpSpeedIncreasing=1000}
    boneToModify {boneName=dyn_coat_r_MSTR}
    rotation {pitch=-15, yaw=-15}
    translation {y=3, z=3}
   }
   n60 = "Transform (Modify) Bone - Bone: dyn_coat_l_MSTR"(ComponentPose=@n59, Alpha=Get Thigh L Out Alpha() -> (Thigh L Out Alpha:double), Translation="0,-2,2", Rotation="-15,-15,0", Scale="1,1,1") {
    rotationMode=BMM_Additive
    rotationSpace=BCS_BoneSpace
    translationMode=BMM_Additive
    translationSpace=BCS_BoneSpace
    alphaBoolBlend {blendOption=Linear}
    alphaScaleBiasClamp {interpSpeedDecreasing=4, interpSpeedIncreasing=1000}
    boneToModify {boneName=dyn_coat_l_MSTR}
    rotation {pitch=-15, yaw=-15}
    translation {y=-2, z=2}
   }
   n61 = Look At - Bone: dyn_coat_l_01(ComponentPose=@n60, Alpha=1) {
    alphaBoolBlend {blendOption=Linear}
    boneToModify {boneName=dyn_coat_l_01}
    lookAtLocation=<-189.9219 -39.5305 -70.363>
    lookAtTarget {boneReference {boneName=pelvis}}
    lookAt_Axis {axis=<1 0>}
   }
   n62 = Look At - Bone: dyn_coat_r_01(ComponentPose=@n61, Alpha=1) {
    alphaBoolBlend {blendOption=Linear}
    boneToModify {boneName=dyn_coat_r_01}
    lookAtLocation=<-193.4731 -29.2708 58.66>
    lookAtTarget {boneReference {boneName=pelvis}}
    lookAt_Axis {axis=<1 0>}
   }
   n63 = Get Skydive Vector() -> (Skydive Vector:Vector)
   n64 = RigidBody!(
    ComponentPose=@n62
    Alpha
    OverrideWorldGravity:Vector=Get GravityOvcride()
    ExternalForce
    ComponentLinearAccScale:Vector=@n63
    ComponentLinearVelScale:Vector=@n63
   ) -> (Pose) {
    bClampLinearTranslationLimitToRefPose=true
    bOverrideWorldGravity=true
    overridePhysicsAsset=&/F_MED_Little_Egg_Jacket_Physics_2
    simulationSpace=BaseBoneSpace
    alphaBoolBlend {blendOption=Linear}
    baseBoneRef {boneName=spine_01}
    componentAppliedLinearAccClamp=<1500 1500 1500>
    componentLinearAccScale=<0.2 0.2 0.35>
    componentLinearVelScale=<0.2 0.2 0.35>
    overrideWorldGravity {z=-980}
    simSpaceSettings {
     maxAngularAcceleration=1000
     maxAngularVelocity=3
     maxLinearAcceleration=700
     maxLinearVelocity=500
     velocityScaleZ=0
     worldAlpha=1
     externalLinearDragV=<0.1 0.1>
    }
   }
   n65 = "Constraint - dyn_coat_bk_l_MSTR (PPPPPP)"!(
    ComponentPose=@n64
    Alpha=1
    ConstraintWeights_0=0.0475
    ConstraintWeights_1:float=0
    ConstraintWeights_2:float=0
    ConstraintWeights_3:float=0
    ConstraintWeights_4:float=0.0452
    ConstraintWeights_5:float=0.9072
   ) -> (Pose) {
    alphaBoolBlend {blendOption=Linear}
    boneToModify {boneName=dyn_coat_bk_l_MSTR}
    constraintSetup [
     {targetBone {boneName=spine_02}}
     {targetBone {boneName=spine_03}}
     {targetBone {boneName=spine_04}}
     {targetBone {boneName=thigh_twist_01_l}}
     {targetBone {boneName=spine_01}}
     {targetBone {boneName=pelvis}}
    ]
    constraintWeights [1,1,1,1,1,1]
   }
   n66 = "Constraint - dyn_coat_bk_MSTR (PPPPPPPP)"!(
    ComponentPose=@n65
    Alpha=1
    ConstraintWeights_0=0.0498
    ConstraintWeights_1:float=0.0001
    ConstraintWeights_2:float=0.0548
    ConstraintWeights_3:float=0.0001
    ConstraintWeights_4:float=0.0595
    ConstraintWeights_5:float=0
    ConstraintWeights_6:float=0
    ConstraintWeights_7:float=0.8358
   ) -> (Pose) {
    alphaBoolBlend {blendOption=Linear}
    boneToModify {boneName=dyn_coat_bk_MSTR}
    constraintSetup [
     {targetBone {boneName=spine_02}}
     {targetBone {boneName=thigh_twist_01_r}}
     {targetBone {boneName=spine_03}}
     {targetBone {boneName=thigh_twist_01_l}}
     {targetBone {boneName=spine_01}}
     {targetBone {boneName=calf_l}}
     {targetBone {boneName=thigh_l}}
     {targetBone {boneName=pelvis}}
    ]
    constraintWeights [1,1,1,1,1,1,1,1]
   }
   n67 = "Constraint - dyn_coat_bk_r_MSTR (PPPPPPPP)"!(
    ComponentPose=@n66
    Alpha=1
    ConstraintWeights_0=0.108
    ConstraintWeights_1:float=0
    ConstraintWeights_2:float=0
    ConstraintWeights_3:float=0.0174
    ConstraintWeights_4:float=0.0055
    ConstraintWeights_5:float=0
    ConstraintWeights_6:float=0.1003
    ConstraintWeights_7:float=0.7688
   ) -> (Pose) {
    alphaBoolBlend {blendOption=Linear}
    boneToModify {boneName=dyn_coat_bk_r_MSTR}
    constraintSetup [
     {targetBone {boneName=spine_02}}
     {targetBone {boneName=thigh_twist_01_r}}
     {targetBone {boneName=thigh_r}}
     {targetBone {boneName=spine_03}}
     {targetBone {boneName=spine_04}}
     {targetBone {boneName=thigh_twist_01_l}}
     {targetBone {boneName=spine_01}}
     {targetBone {boneName=pelvis}}
    ]
    constraintWeights [1,1,1,1,1,1,1,1]
   }
   n68 = "Transform (Modify) Bone - Bone: dyn_coat_l_03"(ComponentPose=@n67, Alpha=Get Thigh L Out Alpha(), Translation="0,3,10", Rotation="0,0,0", Scale="1,1,1") {
    rotationMode=BMM_Additive
    rotationSpace=BCS_BoneSpace
    translationMode=BMM_Additive
    translationSpace=BCS_BoneSpace
    alphaBoolBlend {blendOption=Linear}
    boneToModify {boneName=dyn_coat_l_03}
    translation {y=3, z=10}
   }
   n69 = "Transform (Modify) Bone - Bone: dyn_coat_r_03"(ComponentPose=@n68, Alpha=Get Thigh R Out Alpha(), Translation="0,-2,12", Rotation="0,0,0", Scale="1,1,1") {
    rotationMode=BMM_Additive
    rotationSpace=BCS_BoneSpace
    translationMode=BMM_Additive
    translationSpace=BCS_BoneSpace
    alphaBoolBlend {blendOption=Linear}
    boneToModify {boneName=dyn_coat_r_03}
    translation {y=-2, z=12}
   }
   n70 = Get Skydive Alpha()
   n71 = "Transform (Modify) Bone - Bone: dyn_coat_bk_r_MSTR"(ComponentPose=@n69, Alpha=@n70, Translation="0,0,4", Rotation="5,0,0", Scale="1,1,1") {
    rotationMode=BMM_Additive
    rotationSpace=BCS_BoneSpace
    translationMode=BMM_Additive
    translationSpace=BCS_BoneSpace
    alphaBoolBlend {blendOption=Linear}
    boneToModify {boneName=dyn_coat_bk_r_MSTR}
    rotation {pitch=5}
    translation {z=4}
   }
   n72 = "Transform (Modify) Bone - Bone: dyn_coat_bk_MSTR"(ComponentPose=@n71, Alpha=@n70, Translation="0,0,4", Rotation="5,0,0", Scale="1,1,1") {
    rotationMode=BMM_Additive
    rotationSpace=BCS_BoneSpace
    translationMode=BMM_Additive
    translationSpace=BCS_BoneSpace
    alphaBoolBlend {blendOption=Linear}
    boneToModify {boneName=dyn_coat_bk_MSTR}
    rotation {pitch=5}
    translation {z=4}
   }
   n73 = "Transform (Modify) Bone - Bone: dyn_coat_bk_l_MSTR"(ComponentPose=@n72, Alpha=@n70, Translation="0,0,4", Rotation="5,0,0", Scale="1,1,1") {
    rotationMode=BMM_Additive
    rotationSpace=BCS_BoneSpace
    translationMode=BMM_Additive
    translationSpace=BCS_BoneSpace
    alphaBoolBlend {blendOption=Linear}
    boneToModify {boneName=dyn_coat_bk_l_MSTR}
    rotation {pitch=5}
    translation {z=4}
   }
   n74 = NOT Boolean(A=Get bIsCrouching())
   n75 = "Transform (Modify) Bone - Bone: dyn_coat_bk_r_MSTR"!(ComponentPose=@n73, bAlphaBoolEnabled:bool=@n74, Translation="0,0,1", Rotation="0,0,0", Scale="1,1,1") -> (Pose) {
    alphaInputType=Bool
    rotationMode=BMM_Additive
    rotationSpace=BCS_BoneSpace
    translationMode=BMM_Additive
    translationSpace=BCS_BoneSpace
    alphaBoolBlend {blendOption=Linear}
    boneToModify {boneName=dyn_coat_bk_r_MSTR}
    translation {z=1}
   }
   n76 = "Transform (Modify) Bone - Bone: dyn_coat_bk_MSTR"!(ComponentPose=@n75, bAlphaBoolEnabled:bool=@n74, Translation="0,0,1", Rotation="0,0,0", Scale="1,1,1") -> (Pose) {
    alphaInputType=Bool
    rotationMode=BMM_Additive
    rotationSpace=BCS_BoneSpace
    translationMode=BMM_Additive
    translationSpace=BCS_BoneSpace
    alphaBoolBlend {blendOption=Linear}
    boneToModify {boneName=dyn_coat_bk_MSTR}
    translation {z=1}
   }
   n77 = "Transform (Modify) Bone - Bone: dyn_coat_bk_l_MSTR"!(ComponentPose=@n76, bAlphaBoolEnabled:bool=@n74, Translation="0,0,1", Rotation="0,0,0", Scale="1,1,1") -> (Pose) {
    alphaInputType=Bool
    rotationMode=BMM_Additive
    rotationSpace=BCS_BoneSpace
    translationMode=BMM_Additive
    translationSpace=BCS_BoneSpace
    alphaBoolBlend {blendOption=Linear}
    boneToModify {boneName=dyn_coat_bk_l_MSTR}
    translation {z=1}
   }
   n78 = "Constraint - dyn_coat_r_03 (P)"(ComponentPose=@n77, Alpha=0.5, ConstraintWeights_0=1) {
    alphaBoolBlend {blendOption=Linear}
    alphaScaleBiasClamp {bClampResult=true, bInterpResult=true, clampMax=0.5, interpSpeedDecreasing=5, interpSpeedIncreasing=8}
    boneToModify {boneName=dyn_coat_r_03}
    constraintSetup [{targetBone {boneName=thigh_r}}]
    constraintWeights [1]
   }
   n79 = "Constraint - dyn_coat_r_04 (P)"(ComponentPose=@n78, Alpha=0.5, ConstraintWeights_0=1) {
    alphaBoolBlend {blendOption=Linear}
    alphaScaleBiasClamp {bClampResult=true, bInterpResult=true, clampMax=0.5, interpSpeedDecreasing=5, interpSpeedIncreasing=8}
    boneToModify {boneName=dyn_coat_r_04}
    constraintSetup [{targetBone {boneName=thigh_r}}]
    constraintWeights [1]
   }
   n80 = "Constraint - dyn_coat_r_05 (P)"(ComponentPose=@n79, Alpha=0.5, ConstraintWeights_0=1) {
    alphaBoolBlend {blendOption=Linear}
    alphaScaleBiasClamp {bClampResult=true, bInterpResult=true, clampMax=0.5, interpSpeedDecreasing=5, interpSpeedIncreasing=8}
    boneToModify {boneName=dyn_coat_r_05}
    constraintSetup [{targetBone {boneName=calf_r}}]
    constraintWeights [1]
   }
   n81 = "Constraint - dyn_coat_r_06 (P)"(ComponentPose=@n80, Alpha=0.5, ConstraintWeights_0=1) {
    alphaBoolBlend {blendOption=Linear}
    alphaScaleBiasClamp {bClampResult=true, bInterpResult=true, clampMax=0.5, interpSpeedDecreasing=5, interpSpeedIncreasing=8}
    boneToModify {boneName=dyn_coat_r_06}
    constraintSetup [{targetBone {boneName=calf_r}}]
    constraintWeights [1]
   }
   n82 = "Constraint - dyn_coat_l_03 (P)"(ComponentPose=@n81, Alpha=0.5, ConstraintWeights_0=1) {
    alphaBoolBlend {blendOption=Linear}
    alphaScaleBiasClamp {bClampResult=true, bInterpResult=true, clampMax=0.5, interpSpeedDecreasing=5, interpSpeedIncreasing=8}
    boneToModify {boneName=dyn_coat_l_03}
    constraintSetup [{targetBone {boneName=thigh_l}}]
    constraintWeights [1]
   }
   n83 = "Constraint - dyn_coat_l_04 (P)"(ComponentPose=@n82, Alpha=0.5, ConstraintWeights_0=1) {
    alphaBoolBlend {blendOption=Linear}
    alphaScaleBiasClamp {bClampResult=true, bInterpResult=true, clampMax=0.5, interpSpeedDecreasing=5, interpSpeedIncreasing=8}
    boneToModify {boneName=dyn_coat_l_04}
    constraintSetup [{targetBone {boneName=thigh_l}}]
    constraintWeights [1]
   }
   n84 = "Constraint - dyn_coat_l_05 (P)"(ComponentPose=@n83, Alpha=0.5, ConstraintWeights_0=1) {
    alphaBoolBlend {blendOption=Linear}
    alphaScaleBiasClamp {bClampResult=true, bInterpResult=true, clampMax=0.5, interpSpeedDecreasing=5, interpSpeedIncreasing=8}
    boneToModify {boneName=dyn_coat_l_05}
    constraintSetup [{targetBone {boneName=calf_l}}]
    constraintWeights [1]
   }
   n85 = "Constraint - dyn_coat_l_06 (P)"(ComponentPose=@n84, Alpha=0.5, ConstraintWeights_0=1) {
    alphaBoolBlend {blendOption=Linear}
    alphaScaleBiasClamp {bClampResult=true, bInterpResult=true, clampMax=0.5, interpSpeedDecreasing=5, interpSpeedIncreasing=8}
    boneToModify {boneName=dyn_coat_l_06}
    constraintSetup [{targetBone {boneName=calf_l}}]
    constraintWeights [1]
   }
   n86 = "Constraint - dyn_coat_bk_l_02 (P)"!(ComponentPose=@n85, Alpha:float=0.5, ConstraintWeights_0=1) -> (Pose) {
    alphaBoolBlend {blendOption=Linear}
    alphaScaleBiasClamp {bClampResult=true, bInterpResult=true, clampMax=0.4, interpSpeedDecreasing=5, interpSpeedIncreasing=8}
    boneToModify {boneName=dyn_coat_bk_l_02}
    constraintSetup [{targetBone {boneName=thigh_l}}]
    constraintWeights [1]
   }
   n87 = "Constraint - dyn_coat_bk_l_04 (P)"(ComponentPose=@n86, Alpha=0.5, ConstraintWeights_0=1) {
    alphaBoolBlend {blendOption=Linear}
    alphaScaleBiasClamp {bClampResult=true, bInterpResult=true, clampMax=0.4, interpSpeedDecreasing=5, interpSpeedIncreasing=8}
    boneToModify {boneName=dyn_coat_bk_l_04}
    constraintSetup [{targetBone {boneName=thigh_l}}]
    constraintWeights [1]
   }
   n88 = "Constraint - dyn_coat_bk_l_05 (P)"(ComponentPose=@n87, Alpha=0.5, ConstraintWeights_0=1) {
    alphaBoolBlend {blendOption=Linear}
    alphaScaleBiasClamp {bClampResult=true, bInterpResult=true, clampMax=0.4, interpSpeedDecreasing=5, interpSpeedIncreasing=8}
    boneToModify {boneName=dyn_coat_bk_l_05}
    constraintSetup [{targetBone {boneName=calf_l}}]
    constraintWeights [1]
   }
   n89 = "Constraint - dyn_coat_bk_l_06 (P)"(ComponentPose=@n88, Alpha=0.5, ConstraintWeights_0=1) {
    alphaBoolBlend {blendOption=Linear}
    alphaScaleBiasClamp {bClampResult=true, bInterpResult=true, clampMax=0.4, interpSpeedDecreasing=5, interpSpeedIncreasing=8}
    boneToModify {boneName=dyn_coat_bk_l_06}
    constraintSetup [{targetBone {boneName=calf_l}}]
    constraintWeights [1]
   }
   n90 = "Constraint - dyn_coat_bk_r_02 (P)"!(ComponentPose=@n89, Alpha:float=0.5, ConstraintWeights_0=1) -> (Pose) {
    alphaBoolBlend {blendOption=Linear}
    alphaScaleBiasClamp {bClampResult=true, bInterpResult=true, clampMax=0.4, interpSpeedDecreasing=5, interpSpeedIncreasing=8}
    boneToModify {boneName=dyn_coat_bk_r_02}
    constraintSetup [{targetBone {boneName=thigh_r}}]
    constraintWeights [1]
   }
   n91 = "Constraint - dyn_coat_bk_r_04 (P)"(ComponentPose=@n90, Alpha=0.5, ConstraintWeights_0=1) {
    alphaBoolBlend {blendOption=Linear}
    alphaScaleBiasClamp {bClampResult=true, bInterpResult=true, clampMax=0.4, interpSpeedDecreasing=5, interpSpeedIncreasing=8}
    boneToModify {boneName=dyn_coat_bk_r_04}
    constraintSetup [{targetBone {boneName=thigh_r}}]
    constraintWeights [1]
   }
   n92 = "Constraint - dyn_coat_bk_r_05 (P)"(ComponentPose=@n91, Alpha=0.5, ConstraintWeights_0=1) {
    alphaBoolBlend {blendOption=Linear}
    alphaScaleBiasClamp {bClampResult=true, bInterpResult=true, clampMax=0.4, interpSpeedDecreasing=5, interpSpeedIncreasing=8}
    boneToModify {boneName=dyn_coat_bk_r_05}
    constraintSetup [{targetBone {boneName=calf_r}}]
    constraintWeights [1]
   }
   n93 = "Constraint - dyn_coat_bk_r_06 (P)"(ComponentPose=@n92, Alpha=0.5, ConstraintWeights_0=1) {
    alphaBoolBlend {blendOption=Linear}
    alphaScaleBiasClamp {bClampResult=true, bInterpResult=true, clampMax=0.4, interpSpeedDecreasing=5, interpSpeedIncreasing=8}
    boneToModify {boneName=dyn_coat_bk_r_06}
    constraintSetup [{targetBone {boneName=calf_r}}]
    constraintWeights [1]
   }
   n94 = "Constraint - dyn_coat_bk_02 (PP)"!(ComponentPose=@n93, Alpha=0.5, ConstraintWeights_0=0.5, ConstraintWeights_1:float=0.5) -> (Pose) {
    alphaBoolBlend {blendOption=Linear}
    alphaScaleBiasClamp {bClampResult=true, bInterpResult=true, clampMax=0.4, interpSpeedDecreasing=5, interpSpeedIncreasing=8}
    boneToModify {boneName=dyn_coat_bk_02}
    constraintSetup [{targetBone {boneName=thigh_r}}, {targetBone {boneName=thigh_l}}]
    constraintWeights [1,1]
   }
   n95 = "Constraint - dyn_coat_bk_04 (PP)"!(ComponentPose=@n94, Alpha=0.5, ConstraintWeights_0=0.5, ConstraintWeights_1:float=0.5) -> (Pose) {
    alphaBoolBlend {blendOption=Linear}
    alphaScaleBiasClamp {bClampResult=true, bInterpResult=true, clampMax=0.4, interpSpeedDecreasing=5, interpSpeedIncreasing=8}
    boneToModify {boneName=dyn_coat_bk_04}
    constraintSetup [{targetBone {boneName=thigh_r}}, {targetBone {boneName=thigh_l}}]
    constraintWeights [1,1]
   }
   n96 = "Constraint - dyn_coat_bk_split_01 (PP)"!(ComponentPose=@n95, Alpha=0.5, ConstraintWeights_0=0.5, ConstraintWeights_1:float=0.5) -> (Pose) {
    alphaBoolBlend {blendOption=Linear}
    alphaScaleBiasClamp {bClampResult=true, bInterpResult=true, clampMax=0.4, interpSpeedDecreasing=5, interpSpeedIncreasing=8}
    boneToModify {boneName=dyn_coat_bk_split_01}
    constraintSetup [{targetBone {boneName=calf_r}}, {targetBone {boneName=calf_l}}]
    constraintWeights [1,1]
   }
   n97 = "Constraint - dyn_coat_bk_split_02 (PP)"!(ComponentPose=@n96, Alpha=0.5, ConstraintWeights_0=0.5, ConstraintWeights_1:float=0.5) -> (Pose) {
    alphaBoolBlend {blendOption=Linear}
    alphaScaleBiasClamp {bClampResult=true, bInterpResult=true, clampMax=0.4, interpSpeedDecreasing=5, interpSpeedIncreasing=8}
    boneToModify {boneName=dyn_coat_bk_split_02}
    constraintSetup [{targetBone {boneName=calf_r}}, {targetBone {boneName=calf_l}}]
    constraintWeights [1,1]
   }
   n98 = "Constraint - dyn_coat_side_l_02 (PP)"!(ComponentPose=@n97, Alpha=1, ConstraintWeights_0=0.5, ConstraintWeights_1:float=0.5) -> (Pose) {
    alphaBoolBlend {blendOption=Linear}
    alphaScaleBiasClamp {bClampResult=true, bInterpResult=true, clampMax=0.5, interpSpeedDecreasing=5, interpSpeedIncreasing=8}
    boneToModify {boneName=dyn_coat_side_l_02}
    constraintSetup [{targetBone {boneName=dyn_coat_bk_l_02}}, {targetBone {boneName=dyn_coat_l_04}}]
    constraintWeights [1,1]
   }
   n99 = "Constraint - dyn_coat_side_r_02 (PP)"!(ComponentPose=@n98, Alpha=1, ConstraintWeights_0=0.5, ConstraintWeights_1:float=0.5) -> (Pose) {
    alphaBoolBlend {blendOption=Linear}
    alphaScaleBiasClamp {bClampResult=true, bInterpResult=true, clampMax=0.5, interpSpeedDecreasing=5, interpSpeedIncreasing=8}
    boneToModify {boneName=dyn_coat_side_r_02}
    constraintSetup [{targetBone {boneName=dyn_coat_bk_r_02}}, {targetBone {boneName=dyn_coat_r_04}}]
    constraintWeights [1,1]
   }
   n100 = "Constraint - dyn_coat_side_l_03 (PP)"!(ComponentPose=@n99, Alpha=1, ConstraintWeights_0=0.5, ConstraintWeights_1:float=0.5) -> (Pose) {
    alphaBoolBlend {blendOption=Linear}
    alphaScaleBiasClamp {bClampResult=true, bInterpResult=true, clampMax=0.5, interpSpeedDecreasing=5, interpSpeedIncreasing=8}
    boneToModify {boneName=dyn_coat_side_l_03}
    constraintSetup [{targetBone {boneName=dyn_coat_bk_l_04}}, {targetBone {boneName=dyn_coat_l_04}}]
    constraintWeights [1,1]
   }
   n101 = "Constraint - dyn_coat_side_r_03 (PP)"!(ComponentPose=@n100, Alpha=1, ConstraintWeights_0=0.5, ConstraintWeights_1:float=0.5) -> (Pose) {
    alphaBoolBlend {blendOption=Linear}
    alphaScaleBiasClamp {bClampResult=true, bInterpResult=true, clampMax=0.5, interpSpeedDecreasing=5, interpSpeedIncreasing=8}
    boneToModify {boneName=dyn_coat_side_r_03}
    constraintSetup [{targetBone {boneName=dyn_coat_bk_r_04}}, {targetBone {boneName=dyn_coat_r_04}}]
    constraintWeights [1,1]
   }
   n102 = "Constraint - dyn_coat_side_l_04 (PP)"!(ComponentPose=@n101, Alpha=1, ConstraintWeights_0=0.5, ConstraintWeights_1:float=0.5) -> (Pose) {
    alphaBoolBlend {blendOption=Linear}
    alphaScaleBiasClamp {bClampResult=true, bInterpResult=true, clampMax=0.5, interpSpeedDecreasing=5, interpSpeedIncreasing=8}
    boneToModify {boneName=dyn_coat_side_l_04}
    constraintSetup [{targetBone {boneName=dyn_coat_bk_l_05}}, {targetBone {boneName=dyn_coat_l_05}}]
    constraintWeights [1,1]
   }
   n103 = "Constraint - dyn_coat_side_r_04 (PP)"!(ComponentPose=@n102, Alpha=1, ConstraintWeights_0=0.5, ConstraintWeights_1:float=0.5) -> (Pose) {
    alphaBoolBlend {blendOption=Linear}
    alphaScaleBiasClamp {bClampResult=true, bInterpResult=true, clampMax=0.5, interpSpeedDecreasing=5, interpSpeedIncreasing=8}
    boneToModify {boneName=dyn_coat_side_r_04}
    constraintSetup [{targetBone {boneName=dyn_coat_bk_r_05}}, {targetBone {boneName=dyn_coat_r_05}}]
    constraintWeights [1,1]
   }
   n104 = "Constraint - dyn_coat_side_l_05 (PP)"!(ComponentPose=@n103, Alpha=1, ConstraintWeights_0=0.5, ConstraintWeights_1:float=0.5) -> (Pose) {
    alphaBoolBlend {blendOption=Linear}
    alphaScaleBiasClamp {bClampResult=true, bInterpResult=true, clampMax=0.5, interpSpeedDecreasing=5, interpSpeedIncreasing=8}
    boneToModify {boneName=dyn_coat_side_l_05}
    constraintSetup [{targetBone {boneName=dyn_coat_bk_l_06}}, {targetBone {boneName=dyn_coat_l_06}}]
    constraintWeights [1,1]
   }
   n105 = "Constraint - dyn_coat_side_r_05 (PP)"!(ComponentPose=@n104, Alpha=1, ConstraintWeights_0=0.5, ConstraintWeights_1:float=0.5) -> (Pose) {
    alphaBoolBlend {blendOption=Linear}
    alphaScaleBiasClamp {bClampResult=true, bInterpResult=true, clampMax=0.5, interpSpeedDecreasing=5, interpSpeedIncreasing=8}
    boneToModify {boneName=dyn_coat_side_r_05}
    constraintSetup [{targetBone {boneName=dyn_coat_bk_r_06}}, {targetBone {boneName=dyn_coat_r_06}}]
    constraintWeights [1,1]
   }
   n106 = "Transform (Modify) Bone - Bone: dyn_coat_bk_split_01"!(ComponentPose=@n105, bAlphaBoolEnabled:bool=Get bIsMoving2D(), Translation="0,0,0", Rotation="0,0,20", Scale="1,1,1") -> (Pose) {
    alphaInputType=Bool
    rotationMode=BMM_Additive
    translationMode=BMM_Additive
    alphaBoolBlend {blendInTime=1, blendOption=Linear, blendOutTime=1}
    alphaScaleBiasClamp {bInterpResult=true}
    boneToModify {boneName=dyn_coat_bk_split_01}
    rotation {roll=20}
   }
   n107 = Get CrouchFixAlpha()
   n108 = "Transform (Modify) Bone - Bone: dyn_coat_r_MSTR"(
    ComponentPose=Local To Component(LocalPose=Use cached pose 'StartupCache'())
    Alpha=@n107
    Translation="0,0,0"
    Rotation="15,45,0"
    Scale="1,1,1"
   ) {
    rotationMode=BMM_Additive
    rotationSpace=BCS_BoneSpace
    translationMode=BMM_Additive
    translationSpace=BCS_BoneSpace
    alphaBoolBlend {blendOption=Linear}
    alphaScaleBiasClamp {bInterpResult=true, interpSpeedDecreasing=4, interpSpeedIncreasing=1000}
    boneToModify {boneName=dyn_coat_r_MSTR}
    rotation {pitch=15, yaw=45}
   }
   n109 = "Transform (Modify) Bone - Bone: dyn_coat_l_MSTR"(ComponentPose=@n108, Alpha=@n107, Translation="0,0,0", Rotation="15,-45,0", Scale="1,1,1") {
    rotationMode=BMM_Additive
    rotationSpace=BCS_BoneSpace
    translationMode=BMM_Additive
    translationSpace=BCS_BoneSpace
    alphaBoolBlend {blendOption=Linear}
    alphaScaleBiasClamp {bInterpResult=true, interpSpeedDecreasing=4, interpSpeedIncreasing=1000}
    boneToModify {boneName=dyn_coat_l_MSTR}
    rotation {pitch=15, yaw=-45}
   }
   n110 = "Transform (Modify) Bone - Bone: dyn_coat_r_MSTR"(ComponentPose=@n109, Alpha=1, Translation="-1,2,1", Rotation="15,0,10", Scale="1,1,1") {
    rotationMode=BMM_Additive
    rotationSpace=BCS_BoneSpace
    translationMode=BMM_Additive
    translationSpace=BCS_BoneSpace
    alphaBoolBlend {blendOption=Linear}
    boneToModify {boneName=dyn_coat_r_MSTR}
    rotation {pitch=15, roll=10}
    translation=<-1 2 1>
   }
   n111 = "Transform (Modify) Bone - Bone: dyn_coat_r_MSTR"(ComponentPose=@n110, Alpha=Get Thigh_R_Rot_Crouch() -> (Thigh_R_Rot_Crouch:double), Translation="-2,0,0", Rotation="5,15,10", Scale="1,1,1") {
    rotationMode=BMM_Additive
    rotationSpace=BCS_BoneSpace
    translationMode=BMM_Additive
    translationSpace=BCS_BoneSpace
    alphaBoolBlend {blendOption=Linear}
    boneToModify {boneName=dyn_coat_r_MSTR}
    rotation=<rot 5 15 10>
    translation {x=-2}
   }
   n112 = Get Thigh L Rot Crouch() -> (Thigh L Rot Crouch:double)
   n113 = "Transform (Modify) Bone - Bone: dyn_coat_l_MSTR"(ComponentPose=@n111, Alpha=@n112, Translation="-3,-5,0", Rotation="0,-45,-30", Scale="1,1,1") {
    rotationMode=BMM_Additive
    rotationSpace=BCS_BoneSpace
    translationMode=BMM_Additive
    translationSpace=BCS_BoneSpace
    alphaBoolBlend {blendOption=Linear}
    boneToModify {boneName=dyn_coat_l_MSTR}
    rotation {roll=-30, yaw=-45}
    translation=<-3 -5>
   }
   n114 = "Transform (Modify) Bone - Bone: dyn_coat_l_02"(ComponentPose=@n113, Alpha=@n112, Translation="-1,0,0", Rotation="0,0,0", Scale="1,1,1") {
    rotationMode=BMM_Additive
    rotationSpace=BCS_BoneSpace
    translationMode=BMM_Additive
    translationSpace=BCS_BoneSpace
    alphaBoolBlend {blendOption=Linear}
    boneToModify {boneName=dyn_coat_l_02}
    translation {x=-1}
   }
   n115 = "Transform (Modify) Bone - Bone: dyn_coat_bk_r_MSTR"!(ComponentPose=@n114, Alpha:float=1, Translation="-1,0,0", Rotation="0,0,0", Scale="1,1,1") -> (Pose) {
    rotationMode=BMM_Additive
    rotationSpace=BCS_BoneSpace
    translationMode=BMM_Additive
    translationSpace=BCS_BoneSpace
    alphaBoolBlend {blendOption=Linear}
    boneToModify {boneName=dyn_coat_bk_r_MSTR}
    translation {x=-1}
   }
   n116 = "Transform (Modify) Bone - Bone: dyn_coat_bk_MSTR"!(ComponentPose=@n115, Alpha:float=1, Translation="-1,0,0", Rotation="0,0,0", Scale="1,1,1") -> (Pose) {
    rotationMode=BMM_Additive
    rotationSpace=BCS_BoneSpace
    translationMode=BMM_Additive
    translationSpace=BCS_BoneSpace
    alphaBoolBlend {blendOption=Linear}
    boneToModify {boneName=dyn_coat_bk_MSTR}
    translation {x=-1}
   }
   n117 = "Transform (Modify) Bone - Bone: dyn_coat_bk_l_MSTR"!(ComponentPose=@n116, Alpha:float=1, Translation="-1,0,0", Rotation="0,0,0", Scale="1,1,1") -> (Pose) {
    rotationMode=BMM_Additive
    rotationSpace=BCS_BoneSpace
    translationMode=BMM_Additive
    translationSpace=BCS_BoneSpace
    alphaBoolBlend {blendOption=Linear}
    boneToModify {boneName=dyn_coat_bk_l_MSTR}
    translation {x=-1}
   }
   n118 = Get bIsSliding()
   n119 = "Transform (Modify) Bone - Bone: dyn_coat_bk_r_MSTR"!(ComponentPose=@n117, bAlphaBoolEnabled:bool=@n118, Translation="-5,0,0", Rotation="90,0,0", Scale="1,1,1") -> (Pose) {
    alphaInputType=Bool
    rotationMode=BMM_Additive
    rotationSpace=BCS_BoneSpace
    translationMode=BMM_Additive
    translationSpace=BCS_BoneSpace
    alphaBoolBlend {blendOption=Linear}
    boneToModify {boneName=dyn_coat_bk_r_MSTR}
    rotation {pitch=90}
    translation {x=-5}
   }
   n120 = "Transform (Modify) Bone - Bone: dyn_coat_bk_MSTR"!(ComponentPose=@n119, bAlphaBoolEnabled:bool=@n118, Translation="-5,0,0", Rotation="80,0,0", Scale="1,1,1") -> (Pose) {
    alphaInputType=Bool
    rotationMode=BMM_Additive
    rotationSpace=BCS_BoneSpace
    translationMode=BMM_Additive
    translationSpace=BCS_BoneSpace
    alphaBoolBlend {blendOption=Linear}
    boneToModify {boneName=dyn_coat_bk_MSTR}
    rotation {pitch=80}
    translation {x=-5}
   }
   n121 = "Transform (Modify) Bone - Bone: dyn_coat_bk_l_MSTR"!(ComponentPose=@n120, bAlphaBoolEnabled:bool=@n118, Translation="-5,0,0", Rotation="70,0,0", Scale="1,1,1") -> (Pose) {
    alphaInputType=Bool
    rotationMode=BMM_Additive
    rotationSpace=BCS_BoneSpace
    translationMode=BMM_Additive
    translationSpace=BCS_BoneSpace
    alphaBoolBlend {blendOption=Linear}
    boneToModify {boneName=dyn_coat_bk_l_MSTR}
    rotation {pitch=70}
    translation {x=-5}
   }
   n122 = RigidBody!(ComponentPose=@n121, Alpha, OverrideWorldGravity=Get GravityOvcride(), ExternalForce) -> (Pose) {
    bClampLinearTranslationLimitToRefPose=true
    bOverrideWorldGravity=true
    overridePhysicsAsset=&/F_MED_Little_Egg_Jacket_Physics_Crouch
    simulationSpace=BaseBoneSpace
    alphaBoolBlend {blendOption=Linear}
    baseBoneRef {boneName=spine_01}
    componentAppliedLinearAccClamp=<1500 1500 1500>
    componentLinearAccScale=<0.1 0.1 0.1>
    componentLinearVelScale=<0.1 0.1 0.1>
    overrideWorldGravity {z=-980}
    simSpaceSettings {
     maxAngularAcceleration=1000
     maxAngularVelocity=3
     maxLinearAcceleration=700
     maxLinearVelocity=500
     velocityScaleZ=0
     worldAlpha=1
     externalLinearDragV=<0.1 0.1>
    }
   }
   n123 = "Constraint - dyn_coat_side_l_01 (PP)"!(ComponentPose=@n122, Alpha=1, ConstraintWeights_0=0.5, ConstraintWeights_1:float=0.5) -> (Pose) {
    alphaBoolBlend {blendOption=Linear}
    boneToModify {boneName=dyn_coat_side_l_01}
    constraintSetup [{targetBone {boneName=dyn_coat_bk_l_01}}, {targetBone {boneName=dyn_coat_l_03}}]
    constraintWeights [1,1]
   }
   n124 = "Constraint - dyn_coat_side_r_01 (PP)"!(ComponentPose=@n123, Alpha=1, ConstraintWeights_0=0.5, ConstraintWeights_1:float=0.5) -> (Pose) {
    alphaBoolBlend {blendOption=Linear}
    boneToModify {boneName=dyn_coat_side_r_01}
    constraintSetup [{targetBone {boneName=dyn_coat_bk_r_01}}, {targetBone {boneName=dyn_coat_r_03}}]
    constraintWeights [1,1]
   }
   n125 = "Constraint - dyn_coat_side_l_02 (PP)"(ComponentPose=@n124, Alpha=1, ConstraintWeights_0=0.5, ConstraintWeights_1=0.5) {
    alphaBoolBlend {blendOption=Linear}
    boneToModify {boneName=dyn_coat_side_l_02}
    constraintSetup [{targetBone {boneName=dyn_coat_bk_l_02}}, {targetBone {boneName=dyn_coat_l_04}}]
    constraintWeights [1,1]
   }
   n126 = "Constraint - dyn_coat_side_r_02 (PP)"(ComponentPose=@n125, Alpha=1, ConstraintWeights_0=0.5, ConstraintWeights_1=0.5) {
    alphaBoolBlend {blendOption=Linear}
    boneToModify {boneName=dyn_coat_side_r_02}
    constraintSetup [{targetBone {boneName=dyn_coat_bk_r_02}}, {targetBone {boneName=dyn_coat_r_04}}]
    constraintWeights [1,1]
   }
   n127 = "Constraint - dyn_coat_side_l_03 (PP)"(ComponentPose=@n126, Alpha=1, ConstraintWeights_0=0.5, ConstraintWeights_1=0.5) {
    alphaBoolBlend {blendOption=Linear}
    boneToModify {boneName=dyn_coat_side_l_03}
    constraintSetup [{targetBone {boneName=dyn_coat_bk_l_04}}, {targetBone {boneName=dyn_coat_l_04}}]
    constraintWeights [1,1]
   }
   n128 = "Constraint - dyn_coat_side_r_03 (PP)"(ComponentPose=@n127, Alpha=1, ConstraintWeights_0=0.5, ConstraintWeights_1=0.5) {
    alphaBoolBlend {blendOption=Linear}
    boneToModify {boneName=dyn_coat_side_r_03}
    constraintSetup [{targetBone {boneName=dyn_coat_bk_r_04}}, {targetBone {boneName=dyn_coat_r_04}}]
    constraintWeights [1,1]
   }
   n129 = "Constraint - dyn_coat_side_l_04 (PP)"(ComponentPose=@n128, Alpha=1, ConstraintWeights_0=0.5, ConstraintWeights_1=0.5) {
    alphaBoolBlend {blendOption=Linear}
    boneToModify {boneName=dyn_coat_side_l_04}
    constraintSetup [{targetBone {boneName=dyn_coat_bk_l_05}}, {targetBone {boneName=dyn_coat_l_05}}]
    constraintWeights [1,1]
   }
   n130 = "Constraint - dyn_coat_side_r_04 (PP)"(ComponentPose=@n129, Alpha=1, ConstraintWeights_0=0.5, ConstraintWeights_1=0.5) {
    alphaBoolBlend {blendOption=Linear}
    boneToModify {boneName=dyn_coat_side_r_04}
    constraintSetup [{targetBone {boneName=dyn_coat_bk_r_05}}, {targetBone {boneName=dyn_coat_r_05}}]
    constraintWeights [1,1]
   }
   Save cached pose 'RbanCache'(Pose=Blend Poses by bool(
     BlendPose_0=Component To Local(ComponentPose=@n130)
     BlendPose_1=Component To Local(ComponentPose=@n106)
     BlendTime_0=0.3
     BlendTime_1=0.3
     bActiveValue=Get bIsCrouching()
    ) {blendType=Linear}) {cacheName=RbanCache}
   Use cached pose 'StartupCache'()
  ]}
 EventGraph {nodes [
   n1 = Event BlueprintUpdateAnimation()
   > Set Delta Time Local(Delta Time Local:double=@n1.DeltaTimeX) -> (Output_Get:double)
   > n2 = Sequence!(execute) -> (then_0, then_1, then_2:Exec, then_3:Exec, then_4:Exec, then_5:Exec, then_6:Exec)
   > n3 = Branch(Condition=Get bClothEnabled())
   > n4 = Sequence!(execute) -> (then_0, then_1)
   n5 = Cloth Controller Function DEPRECATED!(
    execute=@n4.then_1
    CharacterPartAnimInstance:Custom Character Part Anim Instance Object Reference=Self-Reference()
    ClothingAssetName:String=LittleEgg_Jacket_Cloth
    Cloth Paramaters:Cloth Params Struct Structure=Get Cloth Paramaters 0() -> (Cloth Paramaters 0:Cloth Params Struct Structure)
    Cloth Core Settings:Cloth Core Setting Struct Structure=Get Cloth Core Settings 0() -> (Cloth Core Settings 0:Cloth Core Setting Struct Structure)
    ClothUpAxis:Vector="1,0,0"
    OnGround_Moving_SpeedLimit:double=2
    OnGround_Moving_AngularVel_Max:double=0.05
    OnGround_Moving_FictitiousVel_Max:double=0.6
    LinVelScale_FwdBwd_Clamp:double=0
    LinVelScale_UpDwn_Clamp:double=0.7
    DeltaTime:double=Get Delta Time Local() -> (Delta Time Local:double)
    GameTimeInSeconds:double=GetGameTimeInSeconds()
    WindGustsParams:Wind Gust Params Struct Structure=Get Wind Gust Params() -> (Wind Gust Params:Wind Gust Params Struct Structure)
    WindGustLocalSettings:Wind Gust Local Setting Struct Structure=Get Gust Local Settings() -> (Gust Local Settings:Wind Gust Local Setting Struct Structure)
   ) -> (
    then
    CurrentStateOut:String
    Cloth Settings Out:Cloth Setting Struct Structure
    Cloth Core Settings Out:Cloth Core Setting Struct Structure
    WindGustSettings Out:Wind Gust Setting Struct Structure
    WindGustLocalSettings Out:Wind Gust Local Setting Struct Structure
    GustStrength - WindVelocityOut:Vector
    LinearVelVector-DELETE LATER:Vector
   )
   > Set Gust Local Settings(Gust Local Settings:Wind Gust Local Setting Struct Structure=@n5.WindGustLocalSettings Out) -> (Output_Get:Wind Gust Local Setting Struct Structure)
   n6 = Gravity Override State Function DEPRECATED(
    execute=[@n3.else, @n4.then_0]
    CharacterPartAnimInstance=Self-Reference()
    GameTimeInSeconds=GetGameTimeInSeconds()
    GravityOverrideParams=Get Gravity Override Params()
   )
   > Set GravityOvcride(GravityOvcride:Vector=@n6) -> (Output_Get:Vector)
   n7 = Branch(execute=@n2.then_2, Condition=Get bIsSwimSprinting() -> (bIsSwimSprinting:bool))
   > Set TrailControllerRot(TrailControllerRot:Rotator="15, 0, 0") -> (Output_Get:Rotator)
   n8 = Branch(execute=@n7.else, Condition=Get bIsSkyDiving())
   > n9 = Branch(Condition=Get bIsParachuteOpened())
   > Set TrailControllerRot(TrailControllerRot="45,0, 0")
   n10 = Branch(execute=@n9.else, Condition="float > float"(A=Get PawnVelocityZ(), B=50) {is=title K2Node_CallFunction})
   > Set Bwd Movement(Bwd Movement:double=float * float(A=Get Bwd Movement(), B=0)) -> (Output_Get:double)
   > Set TrailControllerRot(TrailControllerRot="45,0, 0")
   n11 = Branch(execute=@n10.else, Condition=Get bIsSkydiveDiveMode())
   > Set TrailControllerRot(TrailControllerRot="15,0, 0")
   Set TrailControllerRot(execute=@n11.else, TrailControllerRot="60,0, 0")
   Set TrailControllerRot(execute=@n8.else, TrailControllerRot="15,0, 0")
   > Set Bwd Movement(Bwd Movement=float * float(A=Get Bwd Movement(), B=1))
   n12 = GetOwnerSkeletalMeshComponent()
   n13 = BreakRotator(InRot=BreakTransform(InTransform=GetDeltaTransformFromRefPose(self=@n12, BoneName=thigh_l, BaseName=pelvis)).Rotation)
   Set Thigh L Rot Crouch(execute=@n2.then_5, Thigh L Rot Crouch:double=MapRangeClamped(Value=@n13.Yaw, InRangeA=-40, InRangeB=-100)) -> (Output_Get:double)
   n14 = SelectFloat(A=1, B=0.5, bPickA=Get bIsMoving2D())
   > Set Thigh L Out Alpha(Thigh L Out Alpha:double=MapRangeClamped(Value=@n13.Pitch, InRangeA=-5, InRangeB=30, OutRangeB=@n14)) -> (Output_Get:double)
   n15 = GetOwnerSkeletalMeshComponent()
   n16 = BreakRotator(InRot=BreakTransform(InTransform=GetDeltaTransformFromRefPose(self=@n15, BoneName=thigh_r, BaseName=pelvis)).Rotation)
   > Set Thigh_R_Rot_Crouch(Thigh_R_Rot_Crouch:double=MapRangeClamped(Value=@n16.Yaw, InRangeA=-40, InRangeB=-100)) -> (Output_Get:double)
   > Set Thigh R Out Alpha(Thigh R Out Alpha:double=MapRangeClamped(Value=@n16.Pitch, InRangeA=-5, InRangeB=30, OutRangeB=@n14)) -> (Output_Get:double)
   > n17 = Branch(Condition=OR Boolean(A=Get bIsSkydiveDiveMode(), B=Get bIsSkydiveFloating()))
   > Set Skydive Alpha(Skydive Alpha:double=1) -> (Output_Get:double)
   > Set Skydive Vector(Skydive Vector:Vector="0.1,0.1,0.15") -> (Output_Get:Vector)
   Set Skydive Alpha(execute=@n17.else, Skydive Alpha=0)
   > Set Skydive Vector(Skydive Vector="0.2,0.2,0.35")
   n18 = Get Skel Mesh()
   Set Right Arm Up(
    execute=@n2.then_3
    Right Arm Up:double=GetAlphaFromReferenceTransformDelta(SkeletalMeshComponent=@n18, BoneName=upperarm_r, BaseName=clavicle_r, InputRange=(X=20,Y=100), TwistAxis="0,0,1")
   ) -> (Output_Get:double)
   Set Left Arm Up(
    execute=@n2.then_4
    Left Arm Up:double=GetAlphaFromReferenceTransformDelta(SkeletalMeshComponent=@n18, BoneName=upperarm_l, BaseName=clavicle_l, InputRange=(X=10,Y=90), TwistAxis="0,0,1")
   ) -> (Output_Get:double)
   n19 = PlayerMovement(execute=@n2.then_1, AnimInstance=Self-Reference(), XDotProductCoefficient=1, YDotProductCoefficient=1, ZDotProductCoefficient=1)
   n20 = MapRangeClamped(Value=@n19.YMovement, InRangeA=0, InRangeB=1)
   n21 = "Absolute (Float)"(A=@n19)
   > Set Bwd Movement(Bwd Movement=MapRangeClamped(Value=float + float(A=@n21, B=@n20) {is=title K2Node_CommutativeAssociativeBinaryOperator}, InRangeA=0, InRangeB=1.4))
   n22 = NOT Boolean(A=Get bIsPlayingEmote())
   n23 = Get bIsInFrontEnd()
   Set FE Coat Alpha(execute=@n2.then_6, FE Coat Alpha:double=SelectFloat(A=1, B=0, bPickA=AND Boolean(A=@n23, B=@n22))) -> (Output_Get:double)
   Event BlueprintInitializeAnimation()
   > Set Skel Mesh(Skel Mesh=GetOwnerSkeletalMeshComponent())
   Event OnBeginCrouching()
   > Set CrouchFixAlpha(CrouchFixAlpha:double=1) -> (Output_Get:double)
   > Delay(Duration=0.1)
   > Set CrouchFixAlpha(CrouchFixAlpha=0)
   Event OnEndCrouching()
   > Set CrouchFixAlpha(CrouchFixAlpha=1)
   > Delay(Duration=0.1)
   > Set CrouchFixAlpha(CrouchFixAlpha=0)
   Event OnBeginSkydiving()
   > Set TrailControllerRot(TrailControllerRot="-65,0,0")
   Event OnEndSkydiving()
   > Set TrailControllerRot(TrailControllerRot="-10,0,0")
   DrawVectorFromJointFunction!(
    execute
    SkelMeshComponent:Scene Component Object Reference
    JointName:Name=pelvis
    LineEnd:Vector=Break Cloth Setting Struct!(ClothSettingStruct:"Cloth Setting Struct Structure (by ref)"=@n5.Cloth Settings Out) -> (
     LinearVelocityScale_7_#1:Vector=~
     AngularVelocityScale_6_#2:float=~
     FictitiousAngularScale_10_#3:float=~
     GravityOverride_15_#4:Vector=~
     DragCoefficient_23_#5:Vector 2D Structure
     LiftCoefficient_24_#6:Vector 2D Structure
     JointName_27_#7:Name
    ) {is=title K2Node_BreakStruct}.GravityOverride_15_#4
    JointOffset:Vector="0, 0, 0"
    LineColor:Linear Color Structure=(R=1,G=0,B=0,A=1)
    Thickness:double=2
    Foreground:EDrawDebugSceneDepthPriorityGroup Enum=Foreground
   ) -> (then)
   Reroute Node(InputPin:Wildcard) -> (OutputPin:Wildcard)
  ]}
]
variables [
 Bwd Movement {type=number}
 Cloth Core Settings 0 {
  type=ClothCoreSettingStruct
  default {
   collisionThickness=1.5
   dampingCoefficient=0.1
   frictionCoefficient=0.5
   localDampingCoefficient=0.1
   numberIterations=2
   numberSubsteps=1
   selfCollisionThickness=0
   animDriveDamping=<0 1>
   animDriveStiffness=<0 1>
   areaStiffness=<1 1>
   bendingStiffness=<1 1>
   edgeStiffness=<1 1>
   tetherStiffness=<1 1>
  }
 }
 Cloth Paramaters 0 {
  type=ClothParamsStruct
  default {
   dBNO {
    angularVelocityScale=0.6
    fictitiousAngularScale=0.1
    jointName=pelvis
    dragCoefficient=<0.2 1>
    gravityOverride=<-980 400 0>
    liftCoefficient=<0 0.5>
    linearVelocityScale=<0.3 0.3 0.1>
   }
   dBNO_Carried {
    angularVelocityScale=0.6
    fictitiousAngularScale=0.1
    jointName=pelvis
    dragCoefficient=<0.2 1>
    gravityOverride=<-980 -200 0>
    liftCoefficient=<0 0.5>
    linearVelocityScale=<0.3 0.3 0.1>
   }
   dBNO_Impulse {
    angularVelocityScale=0
    fictitiousAngularScale=0
    jointName=pelvis
    dragCoefficient=<0 0.2>
    gravityOverride=<-3000 -2000 0>
    liftCoefficient=<0 0.1>
    linearVelocityScale=<0.3 0.1 0.3>
   }
   dBNO_RollBwd {
    angularVelocityScale=0
    fictitiousAngularScale=0
    jointName=pelvis
    dragCoefficient=<0 0.2>
    gravityOverride=<-2000 0 0>
    liftCoefficient=<0 0.1>
    linearVelocityScale=<0.3 0.1 0.3>
   }
   dBNO_RollFwd {
    angularVelocityScale=0
    fictitiousAngularScale=0
    jointName=pelvis
    dragCoefficient=<0 0.2>
    gravityOverride=<-980 0 0>
    liftCoefficient=<0 0.1>
    linearVelocityScale=<0.3 0.1 0.3>
   }
   dBNO_Tumbling {
    angularVelocityScale=0
    fictitiousAngularScale=0
    jointName=pelvis
    dragCoefficient=<0 0.2>
    gravityOverride=<-2500 -300 0>
    liftCoefficient=<0 0.1>
    linearVelocityScale=<0.3 0.1 0.3>
   }
   emoteOrMelee {
    angularVelocityScale=0.6
    fictitiousAngularScale=0.1
    jointName=root
    dragCoefficient=<2.5 2.5>
    gravityOverride=<0 0 -980>
    liftCoefficient=<0.5 0.5>
    linearVelocityScale=<0.6 0.6 0.5>
   }
   front_End {
    angularVelocityScale=0.05
    fictitiousAngularScale=0.6
    jointName=root
    dragCoefficient=<0.2 0.2>
    gravityOverride=<0 -200 -980>
    liftCoefficient=<0.1 0.1>
    linearVelocityScale=<0.1 0.1 0.1>
   }
   onGround_Crouch_Moving {
    angularVelocityScale=0.1
    fictitiousAngularScale=0.3
    jointName=root
    dragCoefficient=<2 2>
    gravityOverride=<0 0 -980>
    liftCoefficient=<0.5 0.5>
    linearVelocityScale=<0.6 0.6 0.5>
   }
   onGround_Falling {
    angularVelocityScale=0.6
    fictitiousAngularScale=0
    jointName=root
    dragCoefficient=<0 0.2>
    gravityOverride=<0 -500 -980>
    liftCoefficient=<0 0.1>
    linearVelocityScale=<2 1 0.3>
   }
   onGround_Moving {
    angularVelocityScale=0.6
    fictitiousAngularScale=0.1
    jointName=root
    dragCoefficient=<2 2>
    gravityOverride=<0 0 -980>
    liftCoefficient=<0.5 0.5>
    linearVelocityScale=<0.6 0.6 0.5>
   }
   onGround_Standing {
    angularVelocityScale=0.05
    fictitiousAngularScale=1.2
    jointName=root
    dragCoefficient=<2.5 2.5>
    gravityOverride=<0 0 -980>
    liftCoefficient=<0.5 0.5>
    linearVelocityScale=<0.6 0.6 0.5>
   }
   rail_Grinding {
    angularVelocityScale=0.2
    fictitiousAngularScale=0
    jointName=root
    dragCoefficient=<0.1 0.2>
    gravityOverride=<50 -500 -400>
    liftCoefficient=<0 0.5>
    linearVelocityScale=<0.1 0.1 0.1>
   }
   riding_Moving {
    angularVelocityScale=0.05
    fictitiousAngularScale=1
    jointName=root
    dragCoefficient=<0.5 0.5>
    gravityOverride=<0 0 -980>
    liftCoefficient=<0.5 0.5>
    linearVelocityScale=<0.4 0.6 0.3>
   }
   riding_Moving_Targeting {
    angularVelocityScale=0.05
    fictitiousAngularScale=0.7
    jointName=root
    dragCoefficient=<0.5 0.5>
    gravityOverride=<0 0 -980>
    liftCoefficient=<0.5 0.5>
    linearVelocityScale=<0.2 0.6 0.2>
   }
   riding_Standing {
    angularVelocityScale=0.05
    fictitiousAngularScale=1
    jointName=root
    dragCoefficient=<0.5 0.5>
    gravityOverride=<0 -450 -980>
    liftCoefficient=<0.5 0.5>
    linearVelocityScale=<0.6 0.6 0.3>
   }
   skydive_DivingDown {
    angularVelocityScale=0.1
    fictitiousAngularScale=0
    jointName=pelvis
    dragCoefficient=<0.3 0.3>
    gravityOverride=<-450 -100 0>
    liftCoefficient=<0.9 0.9>
    linearVelocityScale=<0.1 0.1 0.1>
   }
   skydive_FloatingDown {
    angularVelocityScale=0.1
    fictitiousAngularScale=0
    jointName=root
    dragCoefficient=<0.3 0.3>
    gravityOverride=<0 -800 200>
    liftCoefficient=<1.3 1.3>
    linearVelocityScale=<0.2 0.2 0.1>
   }
   skydive_GoingUp {
    angularVelocityScale=0.1
    fictitiousAngularScale=0
    jointName=root
    dragCoefficient=<0.01 0.01>
    gravityOverride=<0 -2000 400>
    liftCoefficient=<0.05 0.05>
    linearVelocityScale=<0.1 0.1 0.01>
   }
   skydive_Parachuting {
    angularVelocityScale=0.6
    fictitiousAngularScale=0.1
    jointName=root
    dragCoefficient=<0.8 0.8>
    gravityOverride=<0 -600 -500>
    liftCoefficient=<0.8 0.8>
    linearVelocityScale=<0.2 0.2 0.3>
   }
   skydive_Surfing {
    angularVelocityScale=0.6
    fictitiousAngularScale=0.1
    jointName=root
    dragCoefficient=<1.5 1.5>
    gravityOverride=<0 -600 -500>
    liftCoefficient=<1 1>
    linearVelocityScale=<0.1 0.1 0.3>
   }
   swimming {
    angularVelocityScale=0.6
    fictitiousAngularScale=0
    jointName=root
    dragCoefficient=<2.5 2.5>
    gravityOverride=<0 0 0>
    liftCoefficient=<0.1 0.1>
    linearVelocityScale=<0.6 0.6 0.5>
   }
   tactical_Sprint {
    angularVelocityScale=0.6
    fictitiousAngularScale=0.1
    jointName=root
    dragCoefficient=<2 2>
    gravityOverride=<0 0 -980>
    liftCoefficient=<0.5 0.5>
    linearVelocityScale=<0.6 0.6 0.5>
   }
   ziplining {
    angularVelocityScale=0.6
    fictitiousAngularScale=0
    jointName=root
    dragCoefficient=<0 1.5>
    gravityOverride=<0 0 -980>
    liftCoefficient=<0 0.3>
    linearVelocityScale=<0.4 0.4 0.5>
   }
  }
 }
 CrouchFixAlpha {type=number}
 Delta Time Local {type=number}
 FE Coat Alpha {type=number}
 Gravity Override Params {type=GravityOverrideParamsStruct}
 GravityOvcride {type=Vector}
 Gust Local Settings {type=WindGustLocalSettingStruct, default {elapsedTime=0, finalGustStrengthLocal=0, gustDuration=0, gustStrength=0, timeBetweenGusts=0}}
 Gust Local Settings_0 {type=WindGustLocalSettingStruct, default {elapsedTime=0, finalGustStrengthLocal=0, gustDuration=0, gustStrength=0, timeBetweenGusts=0}}
 Left Arm Up {type=number}
 Right Arm Up {type=number}
 Skel Mesh {type=SkeletalMeshComponent}
 Skydive Alpha {type=number}
 Skydive Vector {type=Vector}
 Thigh L Out Alpha {type=number}
 Thigh L Rot Crouch {type=number}
 Thigh R Out Alpha {type=number}
 Thigh_R_Rot_Crouch {type=number}
 TrailControllerRot {type=Rotator}
 Wind Gust Params {
  type=WindGustParamsStruct
  default {
   dBNO {
    maxDownInterpSpeed=5
    maxGustDuration=2
    maxGustStrength=7
    maxTimeBetweenGusts=7
    maxUpInterpSpeed=5
    minDownInterpSpeed=2
    minGustDuration=0.5
    minGustStrength=3
    minTimeBetweenGusts=0
    minUpInterpSpeed=2
    windTurbulenceScalarA=0.1
    windTurbulenceScalarB=1
    windAmplitude=<0 0 0>
    windFrequency=<0 0 0>
    windOffset=<0 0 0>
   }
   dBNO_Carried {
    maxDownInterpSpeed=5
    maxGustDuration=2
    maxGustStrength=7
    maxTimeBetweenGusts=7
    maxUpInterpSpeed=5
    minDownInterpSpeed=2
    minGustDuration=0.5
    minGustStrength=3
    minTimeBetweenGusts=0
    minUpInterpSpeed=2
    windTurbulenceScalarA=0.1
    windTurbulenceScalarB=1
    windAmplitude=<0 0 0>
    windFrequency=<0 0 0>
    windOffset=<0 0 0>
   }
   dBNO_Impulse {
    maxDownInterpSpeed=5
    maxGustDuration=2
    maxGustStrength=7
    maxTimeBetweenGusts=7
    maxUpInterpSpeed=5
    minDownInterpSpeed=2
    minGustDuration=0.5
    minGustStrength=3
    minTimeBetweenGusts=0
    minUpInterpSpeed=2
    windTurbulenceScalarA=0.1
    windTurbulenceScalarB=1
    windAmplitude=<0 0 0>
    windFrequency=<0 0 0>
    windOffset=<0 0 0>
   }
   dBNO_RollBwd {
    maxDownInterpSpeed=5
    maxGustDuration=2
    maxGustStrength=7
    maxTimeBetweenGusts=7
    maxUpInterpSpeed=5
    minDownInterpSpeed=2
    minGustDuration=0.5
    minGustStrength=3
    minTimeBetweenGusts=0
    minUpInterpSpeed=2
    windTurbulenceScalarA=0.1
    windTurbulenceScalarB=1
    windAmplitude=<0 0 0>
    windFrequency=<0 0 0>
    windOffset=<0 0 0>
   }
   dBNO_RollFwd {
    maxDownInterpSpeed=5
    maxGustDuration=2
    maxGustStrength=7
    maxTimeBetweenGusts=7
    maxUpInterpSpeed=5
    minDownInterpSpeed=2
    minGustDuration=0.5
    minGustStrength=3
    minTimeBetweenGusts=0
    minUpInterpSpeed=2
    windTurbulenceScalarA=0.1
    windTurbulenceScalarB=1
    windAmplitude=<0 0 0>
    windFrequency=<0 0 0>
    windOffset=<0 0 0>
   }
   dBNO_Tumbling {
    maxDownInterpSpeed=5
    maxGustDuration=2
    maxGustStrength=7
    maxTimeBetweenGusts=7
    maxUpInterpSpeed=5
    minDownInterpSpeed=2
    minGustDuration=0.5
    minGustStrength=3
    minTimeBetweenGusts=0
    minUpInterpSpeed=2
    windTurbulenceScalarA=0.1
    windTurbulenceScalarB=1
    windAmplitude=<0 0 0>
    windFrequency=<0 0 0>
    windOffset=<0 0 0>
   }
   emoteOrMelee {
    maxDownInterpSpeed=5
    maxGustDuration=2
    maxGustStrength=7
    maxTimeBetweenGusts=7
    maxUpInterpSpeed=5
    minDownInterpSpeed=2
    minGustDuration=0.5
    minGustStrength=3
    minTimeBetweenGusts=0
    minUpInterpSpeed=2
    windTurbulenceScalarA=0.1
    windTurbulenceScalarB=1
    windAmplitude=<0 0 0>
    windFrequency=<0 0 0>
    windOffset=<0 0 0>
   }
   front_End {
    maxDownInterpSpeed=5
    maxGustDuration=2
    maxGustStrength=7
    maxTimeBetweenGusts=7
    maxUpInterpSpeed=5
    minDownInterpSpeed=2
    minGustDuration=0.5
    minGustStrength=3
    minTimeBetweenGusts=0
    minUpInterpSpeed=2
    windTurbulenceScalarA=0.1
    windTurbulenceScalarB=1
    windAmplitude=<0 0 0>
    windFrequency=<0 0 0>
    windOffset=<0 0 0>
   }
   onGround_Crouch_Moving {
    maxDownInterpSpeed=5
    maxGustDuration=2
    maxGustStrength=7
    maxTimeBetweenGusts=7
    maxUpInterpSpeed=5
    minDownInterpSpeed=2
    minGustDuration=0.5
    minGustStrength=3
    minTimeBetweenGusts=0
    minUpInterpSpeed=2
    windTurbulenceScalarA=0.1
    windTurbulenceScalarB=1
    windAmplitude=<0 0 0>
    windFrequency=<0 0 0>
    windOffset=<0 0 0>
   }
   onGround_Falling {
    maxDownInterpSpeed=5
    maxGustDuration=2
    maxGustStrength=7
    maxTimeBetweenGusts=7
    maxUpInterpSpeed=5
    minDownInterpSpeed=2
    minGustDuration=0.5
    minGustStrength=3
    minTimeBetweenGusts=0
    minUpInterpSpeed=2
    windTurbulenceScalarA=0.1
    windTurbulenceScalarB=1
    windAmplitude=<0 0 0>
    windFrequency=<0 0 0>
    windOffset=<0 0 0>
   }
   onGround_Moving {
    maxDownInterpSpeed=5
    maxGustDuration=2
    maxGustStrength=7
    maxTimeBetweenGusts=7
    maxUpInterpSpeed=5
    minDownInterpSpeed=2
    minGustDuration=0.5
    minGustStrength=3
    minTimeBetweenGusts=0
    minUpInterpSpeed=2
    windTurbulenceScalarA=0.1
    windTurbulenceScalarB=1
    windAmplitude=<0 0 0>
    windFrequency=<0 0 0>
    windOffset=<0 0 0>
   }
   onGround_Standing {
    maxDownInterpSpeed=5
    maxGustDuration=2
    maxGustStrength=7
    maxTimeBetweenGusts=7
    maxUpInterpSpeed=5
    minDownInterpSpeed=2
    minGustDuration=0.5
    minGustStrength=3
    minTimeBetweenGusts=0
    minUpInterpSpeed=2
    windTurbulenceScalarA=0.1
    windTurbulenceScalarB=1
    windAmplitude=<0 0 0>
    windFrequency=<0 0 0>
    windOffset=<0 0 0>
   }
   rail_Grinding {
    maxDownInterpSpeed=5
    maxGustDuration=8
    maxGustStrength=7
    maxTimeBetweenGusts=8
    maxUpInterpSpeed=5
    minDownInterpSpeed=2
    minGustDuration=0.2
    minGustStrength=3
    minTimeBetweenGusts=1
    minUpInterpSpeed=2
    windTurbulenceScalarA=0.1
    windTurbulenceScalarB=0.8
    windAmplitude=<20 65 42>
    windFrequency=<5 6 8.6>
    windOffset=<50 -500 -400>
   }
   riding_Moving {
    maxDownInterpSpeed=5
    maxGustDuration=2
    maxGustStrength=7
    maxTimeBetweenGusts=7
    maxUpInterpSpeed=5
    minDownInterpSpeed=2
    minGustDuration=0.5
    minGustStrength=3
    minTimeBetweenGusts=0
    minUpInterpSpeed=2
    windTurbulenceScalarA=0.1
    windTurbulenceScalarB=1
    windAmplitude=<0 0 0>
    windFrequency=<0 0 0>
    windOffset=<0 0 0>
   }
   riding_Moving_Targeting {
    maxDownInterpSpeed=5
    maxGustDuration=2
    maxGustStrength=7
    maxTimeBetweenGusts=7
    maxUpInterpSpeed=5
    minDownInterpSpeed=2
    minGustDuration=0.5
    minGustStrength=3
    minTimeBetweenGusts=0
    minUpInterpSpeed=2
    windTurbulenceScalarA=0.1
    windTurbulenceScalarB=1
    windAmplitude=<0 0 0>
    windFrequency=<0 0 0>
    windOffset=<0 0 0>
   }
   riding_Standing {
    maxDownInterpSpeed=5
    maxGustDuration=2
    maxGustStrength=7
    maxTimeBetweenGusts=7
    maxUpInterpSpeed=5
    minDownInterpSpeed=2
    minGustDuration=0.5
    minGustStrength=3
    minTimeBetweenGusts=0
    minUpInterpSpeed=2
    windTurbulenceScalarA=0.1
    windTurbulenceScalarB=1
    windAmplitude=<0 0 0>
    windFrequency=<0 0 0>
    windOffset=<0 0 0>
   }
   skydive_DivingDown {
    maxDownInterpSpeed=5
    maxGustDuration=2
    maxGustStrength=7
    maxTimeBetweenGusts=7
    maxUpInterpSpeed=5
    minDownInterpSpeed=2
    minGustDuration=0.5
    minGustStrength=3
    minTimeBetweenGusts=0
    minUpInterpSpeed=2
    windTurbulenceScalarA=0.1
    windTurbulenceScalarB=1
    windAmplitude=<0 0 0>
    windFrequency=<0 0 0>
    windOffset=<0 0 0>
   }
   skydive_FloatingDown {
    maxDownInterpSpeed=5
    maxGustDuration=2
    maxGustStrength=7
    maxTimeBetweenGusts=7
    maxUpInterpSpeed=5
    minDownInterpSpeed=2
    minGustDuration=0.5
    minGustStrength=3
    minTimeBetweenGusts=0
    minUpInterpSpeed=2
    windTurbulenceScalarA=0.1
    windTurbulenceScalarB=1
    windAmplitude=<0 0 0>
    windFrequency=<0 0 0>
    windOffset=<0 0 0>
   }
   skydive_GoingUp {
    maxDownInterpSpeed=5
    maxGustDuration=2
    maxGustStrength=7
    maxTimeBetweenGusts=7
    maxUpInterpSpeed=5
    minDownInterpSpeed=2
    minGustDuration=0.5
    minGustStrength=3
    minTimeBetweenGusts=0
    minUpInterpSpeed=2
    windTurbulenceScalarA=0.1
    windTurbulenceScalarB=1
    windAmplitude=<0 0 0>
    windFrequency=<0 0 0>
    windOffset=<0 0 0>
   }
   skydive_Parachuting {
    maxDownInterpSpeed=5
    maxGustDuration=2
    maxGustStrength=7
    maxTimeBetweenGusts=7
    maxUpInterpSpeed=5
    minDownInterpSpeed=2
    minGustDuration=0.5
    minGustStrength=3
    minTimeBetweenGusts=0
    minUpInterpSpeed=2
    windTurbulenceScalarA=0.1
    windTurbulenceScalarB=1
    windAmplitude=<0 0 0>
    windFrequency=<0 0 0>
    windOffset=<0 0 0>
   }
   skydive_Surfing {
    maxDownInterpSpeed=5
    maxGustDuration=2
    maxGustStrength=7
    maxTimeBetweenGusts=7
    maxUpInterpSpeed=5
    minDownInterpSpeed=2
    minGustDuration=0.5
    minGustStrength=3
    minTimeBetweenGusts=0
    minUpInterpSpeed=2
    windTurbulenceScalarA=0.1
    windTurbulenceScalarB=1
    windAmplitude=<0 0 0>
    windFrequency=<0 0 0>
    windOffset=<0 0 0>
   }
   swimming {
    maxDownInterpSpeed=5
    maxGustDuration=2
    maxGustStrength=7
    maxTimeBetweenGusts=7
    maxUpInterpSpeed=5
    minDownInterpSpeed=2
    minGustDuration=0.5
    minGustStrength=3
    minTimeBetweenGusts=0
    minUpInterpSpeed=2
    windTurbulenceScalarA=0.1
    windTurbulenceScalarB=1
    windAmplitude=<0 0 0>
    windFrequency=<0 0 0>
    windOffset=<0 0 0>
   }
   tactical_Sprint {
    maxDownInterpSpeed=5
    maxGustDuration=2
    maxGustStrength=7
    maxTimeBetweenGusts=7
    maxUpInterpSpeed=5
    minDownInterpSpeed=2
    minGustDuration=0.5
    minGustStrength=3
    minTimeBetweenGusts=0
    minUpInterpSpeed=2
    windTurbulenceScalarA=0.1
    windTurbulenceScalarB=1
    windAmplitude=<0 0 0>
    windFrequency=<0 0 0>
    windOffset=<0 0 0>
   }
   ziplining {
    maxDownInterpSpeed=5
    maxGustDuration=2
    maxGustStrength=7
    maxTimeBetweenGusts=7
    maxUpInterpSpeed=5
    minDownInterpSpeed=2
    minGustDuration=0.5
    minGustStrength=3
    minTimeBetweenGusts=0
    minUpInterpSpeed=2
    windTurbulenceScalarA=0.1
    windTurbulenceScalarB=1
    windAmplitude=<0 0 0>
    windFrequency=<0 0 0>
    windOffset=<0 0 0>
   }
  }
 }
]
