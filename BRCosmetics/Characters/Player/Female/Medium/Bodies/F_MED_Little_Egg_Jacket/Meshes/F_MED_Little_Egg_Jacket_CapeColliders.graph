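# Physics asset for the female medium Little Egg Jacket cosmetic, named as cape colliders and covering the body and its dynamic coat bones.
# A head sphere, capsules on pelvis, spine, neck and clavicles, and a wide root box provide body shapes; arm and leg bodies are empty.
# Five coat chains (back left, centre, right, both sides) use identical capsules and hang from spine_01 through angularly free constraints, with the centre chain continuing into split bones.
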PhysicsAsset /BRCosmetics/Characters/Player/Female/Medium/Bodies/F_MED_Little_Egg_Jacket/Meshes/F_MED_Little_Egg_Jacket_CapeColliders
bodies {
 calf_l []
 calf_r []
 clavicle_l [{shapeType=Capsule, length=5, radius=10, shapeName=clavicle_l_capsule, center=<6 0 0>, rotation=<rot 90 0 0>}]
 clavicle_r [{shapeType=Capsule, length=5, radius=10, shapeName=clavicle_r_capsule, center=<-6 0 0>, rotation=<rot 90 0 0>}]
 dyn_coat_bk_01 [{shapeType=Capsule, length=5, radius=10, shapeName=dyn_coat_bk_01_capsule, center=<0 0 -6>, rotation=<rot 90 0 0>}]
 dyn_coat_bk_02 [{shapeType=Capsule, length=5, radius=10, shapeName=dyn_coat_bk_02_capsule, center=<0 0 -6>, rotation=<rot 90 0 0>}]
 dyn_coat_bk_04 [{shapeType=Capsule, length=5, radius=10, shapeName=dyn_coat_bk_04_capsule, center=<0 0 -6>, rotation=<rot 90 0 0>}]
 dyn_coat_bk_MSTR [{shapeType=Capsule, length=5, radius=10, shapeName=dyn_coat_bk_MSTR_capsule, center=<0 0 -6>, rotation=<rot 90 0 0>}]
 dyn_coat_bk_l_01 [{shapeType=Capsule, length=5, radius=10, shapeName=dyn_coat_bk_l_01_capsule, center=<0 0 -6>, rotation=<rot 90 0 0>}]
 dyn_coat_bk_l_02 [{shapeType=Capsule, length=5, radius=10, shapeName=dyn_coat_bk_l_02_capsule, center=<0 0 -6>, rotation=<rot 90 0 0>}]
 dyn_coat_bk_l_04 [{shapeType=Capsule, length=5, radius=10, shapeName=dyn_coat_bk_l_04_capsule, center=<0 0 -6>, rotation=<rot 90 0 0>}]
 dyn_coat_bk_l_05 [{shapeType=Capsule, length=5, radius=10, shapeName=dyn_coat_bk_l_05_capsule, center=<0 0 -6>, rotation=<rot 90 0 0>}]
 dyn_coat_bk_l_06 [{shapeType=Capsule, length=5, radius=10, shapeName=dyn_coat_bk_l_06_capsule, center=<0 0 -6>, rotation=<rot 90 0 0>}]
 dyn_coat_bk_l_MSTR [{shapeType=Capsule, length=5, radius=10, shapeName=dyn_coat_bk_l_MSTR_capsule, center=<0 0 -6>, rotation=<rot 90 0 0>}]
 dyn_coat_bk_r_01 [{shapeType=Capsule, length=5, radius=10, shapeName=dyn_coat_bk_r_01_capsule, center=<0 0 -6>, rotation=<rot 90 0 0>}]
 dyn_coat_bk_r_02 [{shapeType=Capsule, length=5, radius=10, shapeName=dyn_coat_bk_r_02_capsule, center=<0 0 -6>, rotation=<rot 90 0 0>}]
 dyn_coat_bk_r_04 [{shapeType=Capsule, length=5, radius=10, shapeName=dyn_coat_bk_r_04_capsule, center=<0 0 -6>, rotation=<rot 90 0 0>}]
 dyn_coat_bk_r_05 [{shapeType=Capsule, length=5, radius=10, shapeName=dyn_coat_bk_r_05_capsule, center=<0 0 -6>, rotation=<rot 90 0 0>}]
 dyn_coat_bk_r_06 [{shapeType=Capsule, length=5, radius=10, shapeName=dyn_coat_bk_r_06_capsule, center=<0 0 -6>, rotation=<rot 90 0 0>}]
 dyn_coat_bk_r_MSTR [{shapeType=Capsule, length=5, radius=10, shapeName=dyn_coat_bk_r_MSTR_capsule, center=<0 0 -6>, rotation=<rot 90 0 0>}]
 dyn_coat_bk_split_01 [{shapeType=Capsule, length=5, radius=10, shapeName=dyn_coat_bk_split_01_capsule, center=<0 0 -6>, rotation=<rot 90 0 0>}]
 dyn_coat_bk_split_02 [{shapeType=Capsule, length=5, radius=10, shapeName=dyn_coat_bk_split_02_capsule, center=<0 0 -6>, rotation=<rot 90 0 0>}]
 dyn_coat_side_l_01 [{shapeType=Capsule, length=5, radius=10, shapeName=dyn_coat_side_l_01_capsule, center=<0 0 -6>, rotation=<rot 90 0 0>}]
 dyn_coat_side_l_02 [{shapeType=Capsule, length=5, radius=10, shapeName=dyn_coat_side_l_02_capsule, center=<0 0 -6>, rotation=<rot 90 0 0>}]
 dyn_coat_side_l_03 [{shapeType=Capsule, length=5, radius=10, shapeName=dyn_coat_side_l_03_capsule, center=<0 0 -6>, rotation=<rot 90 0 0>}]
 dyn_coat_side_l_04 [{shapeType=Capsule, length=5, radius=10, shapeName=dyn_coat_side_l_04_capsule, center=<0 0 -6>, rotation=<rot 90 0 0>}]
 dyn_coat_side_l_05 [{shapeType=Capsule, length=5, radius=10, shapeName=dyn_coat_side_l_05_capsule, center=<0 0 -6>, rotation=<rot 90 0 0>}]
 dyn_coat_side_l_MSTR [{shapeType=Capsule, length=5, radius=10, shapeName=dyn_coat_side_l_MSTR_capsule, center=<0 0 -6>, rotation=<rot 90 0 0>}]
 dyn_coat_side_r_01 [{shapeType=Capsule, length=5, radius=10, shapeName=dyn_coat_side_r_01_capsule, center=<0 0 -6>, rotation=<rot 90 0 0>}]
 dyn_coat_side_r_02 [{shapeType=Capsule, length=5, radius=10, shapeName=dyn_coat_side_r_02_capsule, center=<0 0 -6>, rotation=<rot 90 0 0>}]
 dyn_coat_side_r_03 [{shapeType=Capsule, length=5, radius=10, shapeName=dyn_coat_side_r_03_capsule, center=<0 0 -6>, rotation=<rot 90 0 0>}]
 dyn_coat_side_r_04 [{shapeType=Capsule, length=5, radius=10, shapeName=dyn_coat_side_r_04_capsule, center=<0 0 -6>, rotation=<rot 90 0 0>}]
 dyn_coat_side_r_05 [{shapeType=Capsule, length=5, radius=10, shapeName=dyn_coat_side_r_05_capsule, center=<0 0 -6>, rotation=<rot 90 0 0>}]
 dyn_coat_side_r_MSTR [{shapeType=Capsule, length=5, radius=10, shapeName=dyn_coat_side_r_MSTR_capsule, center=<0 0 -6>, rotation=<rot 90 0 0>}]
 foot_l []
 foot_r []
 head [{shapeType=Sphere, radius=17, shapeName=head_sphere, center=<10.9137 0 0>}]
 lowerarm_l []
 lowerarm_r []
 neck_02 [{shapeType=Capsule, length=13, radius=7, shapeName=neck_02_capsule, rotation=<rot 90 0 0>}]
 pelvis [{shapeType=Capsule, length=8, radius=10.5}]
 root [{shapeType=Box, extentX=327.5, extentY=327.5, extentZ=10, shapeName=root_box}]
 spine_01 [{shapeType=Capsule, length=8, radius=9}]
 spine_02 [{shapeType=Capsule, length=4, radius=8}]
 spine_03 [{shapeType=Capsule, length=4, radius=8, shapeName=spine_03_capsule}]
 spine_04 [{shapeType=Capsule, length=10, radius=12}]
 spine_05 [{shapeType=Capsule, length=10, radius=15, center=<1.4779 0 0>}]
 thigh_l []
 thigh_r []
 upperarm_l []
 upperarm_r []
}
constraints [
 {bone1Name=spine_01, bone2Name=pelvis, swing1LimitDegrees=45, swing2LimitDegrees=45, twistLimitDegrees=45}
 {bone1Name=spine_02, bone2Name=spine_01, swing1LimitDegrees=45, swing2LimitDegrees=45, twistLimitDegrees=45}
 {bone1Name=spine_04, bone2Name=spine_03, swing1LimitDegrees=45, swing2LimitDegrees=45, twistLimitDegrees=45}
 {bone1Name=spine_05, bone2Name=spine_04, swing1LimitDegrees=45, swing2LimitDegrees=45, twistLimitDegrees=45}
 {
  bone1Name=pelvis
  bone2Name=root
  swing1LimitDegrees=45
  swing1Motion=Free
  swing2LimitDegrees=45
  swing2Motion=Free
  twistLimitDegrees=45
  twistMotion=Free
 }
 {
  bone1Name=spine_03
  bone2Name=spine_02
  swing1LimitDegrees=45
  swing1Motion=Free
  swing2LimitDegrees=45
  swing2Motion=Free
  twistLimitDegrees=45
  twistMotion=Free
 }
 {
  bone1Name=clavicle_r
  bone2Name=spine_05
  swing1LimitDegrees=45
  swing1Motion=Free
  swing2LimitDegrees=45
  swing2Motion=Free
  twistLimitDegrees=45
  twistMotion=Free
 }
 {
  bone1Name=clavicle_l
  bone2Name=spine_05
  swing1LimitDegrees=45
  swing1Motion=Free
  swing2LimitDegrees=45
  swing2Motion=Free
  twistLimitDegrees=45
  twistMotion=Free
 }
 {
  bone1Name=upperarm_l
  bone2Name=clavicle_l
  swing1LimitDegrees=45
  swing1Motion=Free
  swing2LimitDegrees=45
  swing2Motion=Free
  twistLimitDegrees=45
  twistMotion=Free
 }
 {
  bone1Name=upperarm_r
  bone2Name=clavicle_r
  swing1LimitDegrees=45
  swing1Motion=Free
  swing2LimitDegrees=45
  swing2Motion=Free
  twistLimitDegrees=45
  twistMotion=Free
 }
 {
  bone1Name=lowerarm_l
  bone2Name=upperarm_l
  swing1LimitDegrees=45
  swing1Motion=Free
  swing2LimitDegrees=45
  swing2Motion=Free
  twistLimitDegrees=45
  twistMotion=Free
 }
 {
  bone1Name=lowerarm_r
  bone2Name=upperarm_r
  swing1LimitDegrees=45
  swing1Motion=Free
  swing2LimitDegrees=45
  swing2Motion=Free
  twistLimitDegrees=45
  twistMotion=Free
 }
 {
  bone1Name=neck_02
  bone2Name=spine_05
  swing1LimitDegrees=45
  swing1Motion=Free
  swing2LimitDegrees=45
  swing2Motion=Free
  twistLimitDegrees=45
  twistMotion=Free
 }
 {
  bone1Name=head
  bone2Name=neck_02
  swing1LimitDegrees=45
  swing1Motion=Free
  swing2LimitDegrees=45
  swing2Motion=Free
  twistLimitDegrees=45
  twistMotion=Free
 }
 {
  bone1Name=thigh_l
  bone2Name=pelvis
  swing1LimitDegrees=45
  swing1Motion=Free
  swing2LimitDegrees=45
  swing2Motion=Free
  twistLimitDegrees=45
  twistMotion=Free
 }
 {
  bone1Name=thigh_r
  bone2Name=pelvis
  swing1LimitDegrees=45
  swing1Motion=Free
  swing2LimitDegrees=45
  swing2Motion=Free
  twistLimitDegrees=45
  twistMotion=Free
 }
 {
  bone1Name=calf_r
  bone2Name=thigh_r
  swing1LimitDegrees=45
  swing1Motion=Free
  swing2LimitDegrees=45
  swing2Motion=Free
  twistLimitDegrees=45
  twistMotion=Free
 }
 {
  bone1Name=calf_l
  bone2Name=thigh_l
  swing1LimitDegrees=45
  swing1Motion=Free
  swing2LimitDegrees=45
  swing2Motion=Free
  twistLimitDegrees=45
  twistMotion=Free
 }
 {
  bone1Name=foot_l
  bone2Name=calf_l
  swing1LimitDegrees=45
  swing1Motion=Free
  swing2LimitDegrees=45
  swing2Motion=Free
  twistLimitDegrees=45
  twistMotion=Free
 }
 {
  bone1Name=foot_r
  bone2Name=calf_r
  swing1LimitDegrees=45
  swing1Motion=Free
  swing2LimitDegrees=45
  swing2Motion=Free
  twistLimitDegrees=45
  twistMotion=Free
 }
 {
  bone1Name=dyn_coat_bk_r_MSTR
  bone2Name=spine_01
  swing1LimitDegrees=45
  swing1Motion=Free
  swing2LimitDegrees=45
  swing2Motion=Free
  twistLimitDegrees=45
  twistMotion=Free
 }
 {
  bone1Name=dyn_coat_bk_r_01
  bone2Name=dyn_coat_bk_r_MSTR
  swing1LimitDegrees=45
  swing1Motion=Free
  swing2LimitDegrees=45
  swing2Motion=Free
  twistLimitDegrees=45
  twistMotion=Free
 }
 {
  bone1Name=dyn_coat_bk_r_02
  bone2Name=dyn_coat_bk_r_01
  swing1LimitDegrees=45
  swing1Motion=Free
  swing2LimitDegrees=45
  swing2Motion=Free
  twistLimitDegrees=45
  twistMotion=Free
 }
 {
  bone1Name=dyn_coat_bk_r_04
  bone2Name=dyn_coat_bk_r_02
  swing1LimitDegrees=45
  swing1Motion=Free
  swing2LimitDegrees=45
  swing2Motion=Free
  twistLimitDegrees=45
  twistMotion=Free
 }
 {
  bone1Name=dyn_coat_bk_r_05
  bone2Name=dyn_coat_bk_r_04
  swing1LimitDegrees=45
  swing1Motion=Free
  swing2LimitDegrees=45
  swing2Motion=Free
  twistLimitDegrees=45
  twistMotion=Free
 }
 {
  bone1Name=dyn_coat_bk_r_06
  bone2Name=dyn_coat_bk_r_05
  swing1LimitDegrees=45
  swing1Motion=Free
  swing2LimitDegrees=45
  swing2Motion=Free
  twistLimitDegrees=45
  twistMotion=Free
 }
 {
  bone1Name=dyn_coat_bk_MSTR
  bone2Name=spine_01
  swing1LimitDegrees=45
  swing1Motion=Free
  swing2LimitDegrees=45
  swing2Motion=Free
  twistLimitDegrees=45
  twistMotion=Free
 }
 {
  bone1Name=dyn_coat_bk_01
  bone2Name=dyn_coat_bk_MSTR
  swing1LimitDegrees=45
  swing1Motion=Free
  swing2LimitDegrees=45
  swing2Motion=Free
  twistLimitDegrees=45
  twistMotion=Free
 }
 {
  bone1Name=dyn_coat_bk_02
  bone2Name=dyn_coat_bk_01
  swing1LimitDegrees=45
  swing1Motion=Free
  swing2LimitDegrees=45
  swing2Motion=Free
  twistLimitDegrees=45
  twistMotion=Free
 }
 {
  bone1Name=dyn_coat_bk_04
  bone2Name=dyn_coat_bk_02
  swing1LimitDegrees=45
  swing1Motion=Free
  swing2LimitDegrees=45
  swing2Motion=Free
  twistLimitDegrees=45
  twistMotion=Free
 }
 {
  bone1Name=dyn_coat_bk_split_01
  bone2Name=dyn_coat_bk_04
  swing1LimitDegrees=45
  swing1Motion=Free
  swing2LimitDegrees=45
  swing2Motion=Free
  twistLimitDegrees=45
  twistMotion=Free
 }
 {
  bone1Name=dyn_coat_bk_split_02
  bone2Name=dyn_coat_bk_split_01
  swing1LimitDegrees=45
  swing1Motion=Free
  swing2LimitDegrees=45
  swing2Motion=Free
  twistLimitDegrees=45
  twistMotion=Free
 }
 {
  bone1Name=dyn_coat_bk_l_MSTR
  bone2Name=spine_01
  swing1LimitDegrees=45
  swing1Motion=Free
  swing2LimitDegrees=45
  swing2Motion=Free
  twistLimitDegrees=45
  twistMotion=Free
 }
 {
  bone1Name=dyn_coat_bk_l_01
  bone2Name=dyn_coat_bk_l_MSTR
  swing1LimitDegrees=45
  swing1Motion=Free
  swing2LimitDegrees=45
  swing2Motion=Free
  twistLimitDegrees=45
  twistMotion=Free
 }
 {
  bone1Name=dyn_coat_bk_l_02
  bone2Name=dyn_coat_bk_l_01
  swing1LimitDegrees=45
  swing1Motion=Free
  swing2LimitDegrees=45
  swing2Motion=Free
  twistLimitDegrees=45
  twistMotion=Free
 }
 {
  bone1Name=dyn_coat_bk_l_04
  bone2Name=dyn_coat_bk_l_02
  swing1LimitDegrees=45
  swing1Motion=Free
  swing2LimitDegrees=45
  swing2Motion=Free
  twistLimitDegrees=45
  twistMotion=Free
 }
 {
  bone1Name=dyn_coat_bk_l_05
  bone2Name=dyn_coat_bk_l_04
  swing1LimitDegrees=45
  swing1Motion=Free
  swing2LimitDegrees=45
  swing2Motion=Free
  twistLimitDegrees=45
  twistMotion=Free
 }
 {
  bone1Name=dyn_coat_bk_l_06
  bone2Name=dyn_coat_bk_l_05
  swing1LimitDegrees=45
  swing1Motion=Free
  swing2LimitDegrees=45
  swing2Motion=Free
  twistLimitDegrees=45
  twistMotion=Free
 }
 {
  bone1Name=dyn_coat_side_l_MSTR
  bone2Name=spine_01
  swing1LimitDegrees=45
  swing1Motion=Free
  swing2LimitDegrees=45
  swing2Motion=Free
  twistLimitDegrees=45
  twistMotion=Free
 }
 {
  bone1Name=dyn_coat_side_l_01
  bone2Name=dyn_coat_side_l_MSTR
  swing1LimitDegrees=45
  swing1Motion=Free
  swing2LimitDegrees=45
  swing2Motion=Free
  twistLimitDegrees=45
  twistMotion=Free
 }
 {
  bone1Name=dyn_coat_side_l_02
  bone2Name=dyn_coat_side_l_01
  swing1LimitDegrees=45
  swing1Motion=Free
  swing2LimitDegrees=45
  swing2Motion=Free
  twistLimitDegrees=45
  twistMotion=Free
 }
 {
  bone1Name=dyn_coat_side_l_03
  bone2Name=dyn_coat_side_l_02
  swing1LimitDegrees=45
  swing1Motion=Free
  swing2LimitDegrees=45
  swing2Motion=Free
  twistLimitDegrees=45
  twistMotion=Free
 }
 {
  bone1Name=dyn_coat_side_l_04
  bone2Name=dyn_coat_side_l_03
  swing1LimitDegrees=45
  swing1Motion=Free
  swing2LimitDegrees=45
  swing2Motion=Free
  twistLimitDegrees=45
  twistMotion=Free
 }
 {
  bone1Name=dyn_coat_side_l_05
  bone2Name=dyn_coat_side_l_04
  swing1LimitDegrees=45
  swing1Motion=Free
  swing2LimitDegrees=45
  swing2Motion=Free
  twistLimitDegrees=45
  twistMotion=Free
 }
 {
  bone1Name=dyn_coat_side_r_MSTR
  bone2Name=spine_01
  swing1LimitDegrees=45
  swing1Motion=Free
  swing2LimitDegrees=45
  swing2Motion=Free
  twistLimitDegrees=45
  twistMotion=Free
 }
 {
  bone1Name=dyn_coat_side_r_01
  bone2Name=dyn_coat_side_r_MSTR
  swing1LimitDegrees=45
  swing1Motion=Free
  swing2LimitDegrees=45
  swing2Motion=Free
  twistLimitDegrees=45
  twistMotion=Free
 }
 {
  bone1Name=dyn_coat_side_r_02
  bone2Name=dyn_coat_side_r_01
  swing1LimitDegrees=45
  swing1Motion=Free
  swing2LimitDegrees=45
  swing2Motion=Free
  twistLimitDegrees=45
  twistMotion=Free
 }
 {
  bone1Name=dyn_coat_side_r_03
  bone2Name=dyn_coat_side_r_02
  swing1LimitDegrees=45
  swing1Motion=Free
  swing2LimitDegrees=45
  swing2Motion=Free
  twistLimitDegrees=45
  twistMotion=Free
 }
 {
  bone1Name=dyn_coat_side_r_04
  bone2Name=dyn_coat_side_r_03
  swing1LimitDegrees=45
  swing1Motion=Free
  swing2LimitDegrees=45
  swing2Motion=Free
  twistLimitDegrees=45
  twistMotion=Free
 }
 {
  bone1Name=dyn_coat_side_r_05
  bone2Name=dyn_coat_side_r_04
  swing1LimitDegrees=45
  swing1Motion=Free
  swing2LimitDegrees=45
  swing2Motion=Free
  twistLimitDegrees=45
  twistMotion=Free
 }
]
