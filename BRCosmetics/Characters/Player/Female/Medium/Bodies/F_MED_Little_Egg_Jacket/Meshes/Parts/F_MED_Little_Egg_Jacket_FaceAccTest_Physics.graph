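# Physics asset for the face accessory test part of the female Little Egg Jacket cosmetic, setting up a dynamic hair chain on the upper body.
# Capsules on the upper spine, neck, clavicles and head form the body, joined by constraints with free swing and twist.
# From the head-linked Dyn_Master_Hair, hair capsules chain in the order 1, 5, 2, 3, 4, with angular limits tight on the second swing axis and wider on twist.
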PhysicsAsset /BRCosmetics/Characters/Player/Female/Medium/Bodies/F_MED_Little_Egg_Jacket/Meshes/Parts/F_MED_Little_Egg_Jacket_FaceAccTest_Physics
bodies {
 Dyn_Hair1 [{shapeType=Capsule, length=7, radius=5.8, shapeName=Dyn_Hair1_capsule, center=<5.7514 -0.3488 0>}]
 Dyn_Hair2 [{shapeType=Capsule, length=9.5132, radius=4, shapeName=Dyn_Hair2_capsule, center=<3 0 0>}]
 Dyn_Hair3 [{shapeType=Capsule, length=5.9043, radius=2.9531, shapeName=Dyn_Hair3_capsule, center=<3 0 0>}]
 Dyn_Hair4 [{shapeType=Capsule, length=4.1535, radius=2.5716, shapeName=Dyn_Hair4_capsule, center=<2 0 0>}]
 Dyn_Hair5 [{shapeType=Capsule, length=12, radius=5.5, shapeName=Dyn_Hair5_capsule, center=<4.9152 0 0>}]
 Dyn_Master_Hair [{shapeType=Capsule, length=0.505, radius=0.505, shapeName=Dyn_Master_Hair_capsule, rotation=<rot 0 -90 0>}]
 clavicle_l [{shapeType=Capsule, length=5, radius=6, shapeName=clavicle_l_capsule, center=<8.4474 -0.6775 -0.5473>, rotation=<rot 90 -172.4475 -172.448>}]
 clavicle_r [{shapeType=Capsule, length=5, radius=6, shapeName=clavicle_r_capsule, center=<-8.5845 0.9179 0.4135>, rotation=<rot 90 -166.7373 -166.7379>}]
 head [{shapeType=Capsule, length=1, radius=1, shapeName=head_capsule}]
 neck_01 [{shapeType=Capsule, length=10, radius=5.5, shapeName=neck_01_capsule, center=<0.767 1.1763 0>, rotation=<rot 90 180 -179.9996>}]
 spine_04 [{shapeType=Capsule, length=8, radius=8, shapeName=spine_04_capsule, center=<-1.907 0.9167 0>}]
 spine_05 [{shapeType=Capsule, length=5, radius=8, shapeName=spine_05_capsule, center=<2.036 0.4378 0>}]
}
constraints [
 {
  bone1Name=clavicle_l
  bone2Name=spine_05
  swing1LimitDegrees=45
  swing1Motion=Free
  swing2LimitDegrees=45
  swing2Motion=Free
  twistLimitDegrees=45
  twistMotion=Free
 }
 {
  bone1Name=clavicle_r
  bone2Name=spine_05
  swing1LimitDegrees=45
  swing1Motion=Free
  swing2LimitDegrees=45
  swing2Motion=Free
  twistLimitDegrees=45
  twistMotion=Free
 }
 {
  bone1Name=neck_01
  bone2Name=spine_05
  swing1LimitDegrees=45
  swing1Motion=Free
  swing2LimitDegrees=45
  swing2Motion=Free
  twistLimitDegrees=45
  twistMotion=Free
 }
 {
  bone1Name=spine_05
  bone2Name=spine_04
  swing1LimitDegrees=45
  swing1Motion=Free
  swing2LimitDegrees=45
  swing2Motion=Free
  twistLimitDegrees=45
  twistMotion=Free
 }
 {
  bone1Name=Dyn_Master_Hair
  bone2Name=head
  swing1LimitDegrees=45
  swing1Motion=Free
  swing2LimitDegrees=45
  swing2Motion=Free
  twistLimitDegrees=45
  twistMotion=Free
 }
 {
  bone1Name=head
  bone2Name=neck_01
  swing1LimitDegrees=45
  swing1Motion=Free
  swing2LimitDegrees=45
  swing2Motion=Free
  twistLimitDegrees=45
  twistMotion=Free
 }
 {bone1Name=Dyn_Hair1, bone2Name=Dyn_Master_Hair, swing1LimitDegrees=55, swing2LimitDegrees=10, twistLimitDegrees=60}
 {bone1Name=Dyn_Hair5, bone2Name=Dyn_Hair1, swing1LimitDegrees=80, swing2LimitDegrees=10, twistLimitDegrees=100}
 {bone1Name=Dyn_Hair2, bone2Name=Dyn_Hair5, swing1LimitDegrees=50, swing2LimitDegrees=10, twistLimitDegrees=90}
 {bone1Name=Dyn_Hair3, bone2Name=Dyn_Hair2, swing1LimitDegrees=30, swing2LimitDegrees=10, twistLimitDegrees=80}
 {bone1Name=Dyn_Hair4, bone2Name=Dyn_Hair3, swing1LimitDegrees=30, swing2LimitDegrees=10, twistLimitDegrees=80}
]
