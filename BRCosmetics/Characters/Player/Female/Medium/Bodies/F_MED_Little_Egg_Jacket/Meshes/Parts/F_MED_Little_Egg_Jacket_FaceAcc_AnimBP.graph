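# Anim Blueprint for the Little Egg Jacket face accessory, copying the attached parent's pose and adding secondary motion to its dynamic hair bones.
# With rigid bodies enabled, it selects backpack physics in Dyn_Hair5 space when backbling is on, otherwise head-space physics.
# With rigid bodies disabled, it falls back to bone constraints alone.
# The event graph detects backbling after customization loads and derives Head Up from head rotation, which the anim graph does not read.
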AnimBlueprint /BRCosmetics/Characters/Player/Female/Medium/Bodies/F_MED_Little_Egg_Jacket/Meshes/Parts/F_MED_Little_Egg_Jacket_FaceAcc_AnimBP
target_skeleton=&/F_MED_Little_Egg_Jacket_FaceAcc_Skeleton
graphs [
 AnimGraph {nodes [
   Component To Local()
   n1 = Get Backbling On() -> (Backbling On:bool)
   "Constraint - Dyn_Hair4 (T)"!(ComponentPose, bAlphaBoolEnabled:bool=@n1, ConstraintWeights_0=1) -> (Pose) {
    alphaInputType=Bool
    alphaBoolBlend {blendOption=Linear}
    boneToModify {boneName=Dyn_Hair4}
    constraintSetup [{transformType=Translation, targetBone {boneName=spine_05}}]
    constraintWeights [1]
   }
   n2 = RigidBody(ComponentPose=Local To Component(LocalPose=Copy Pose From Mesh() {bUseAttachedParent=true})) {
    bClampLinearTranslationLimitToRefPose=true
    bOverrideWorldGravity=true
    overridePhysicsAsset=&/F_MED_Little_Egg_Jacket_FaceAccTest_Physics
    simulationSpace=BaseBoneSpace
    alphaBoolBlend {blendOption=Linear}
    baseBoneRef {boneName=head}
    componentAppliedLinearAccClamp=<1200 2000 2000>
    componentLinearAccScale=<0.33 0.1 0.3>
    componentLinearVelScale=<0.43 0.1 0.3>
    overrideWorldGravity {z=-980}
   }
   n3 = Component To Local(ComponentPose=@n2)
   n4 = "Constraint - Dyn_Hair1 (PP)"!(
    ComponentPose=Local To Component(LocalPose=Copy Pose From Mesh() {bUseAttachedParent=true})
    bAlphaBoolEnabled:bool=True
    ConstraintWeights_0=0.8
    ConstraintWeights_1:float=0.2
   ) -> (Pose) {
    alphaInputType=Bool
    alphaBoolBlend {blendOption=Linear}
    boneToModify {boneName=Dyn_Hair1}
    constraintSetup [{targetBone {boneName=head}}, {targetBone {boneName=spine_05}}]
    constraintWeights [1,1]
   }
   n5 = "Constraint - Dyn_Hair5 (PPPP)"!(
    ComponentPose=@n4
    bAlphaBoolEnabled:bool=True
    ConstraintWeights_0=0.9
    ConstraintWeights_1:float=0.1
    ConstraintWeights_2:float=0
    ConstraintWeights_3:float=0
   ) -> (Pose) {
    alphaInputType=Bool
    alphaBoolBlend {blendOption=Linear}
    boneToModify {boneName=Dyn_Hair5}
    constraintSetup [{targetBone {boneName=spine_05}}, {targetBone {boneName=neck_01}}, {targetBone {boneName=clavicle_l}}, {targetBone {boneName=clavicle_r}}]
    constraintWeights [1,1,1,1]
   }
   n6 = RigidBody!(ComponentPose=@n5, bAlphaBoolEnabled:bool, ExternalForce) -> (Pose) {
    alphaInputType=Bool
    bClampLinearTranslationLimitToRefPose=true
    bOverrideWorldGravity=true
    overridePhysicsAsset=&/F_MED_Little_Egg_Jacket_FaceAccTest_Backpack_Physics
    simulationSpace=BaseBoneSpace
    alphaBoolBlend {blendOption=Linear}
    baseBoneRef {boneName=Dyn_Hair5}
    componentAppliedLinearAccClamp=<1500 1500 1500>
    componentLinearAccScale=<0.3 0.3 0.3>
    componentLinearVelScale=<0.2 0.2 0.2>
    overrideWorldGravity {z=-980}
   }
   n7 = Component To Local(ComponentPose=@n6)
   n8 = "Constraint - Dyn_Hair1 (PP)"(
    ComponentPose=Local To Component(LocalPose=Copy Pose From Mesh() {bUseAttachedParent=true})
    bAlphaBoolEnabled=True
    ConstraintWeights_0=0.8
    ConstraintWeights_1=0.2
   ) {
    alphaInputType=Bool
    alphaBoolBlend {blendOption=Linear}
    boneToModify {boneName=Dyn_Hair1}
    constraintSetup [{targetBone {boneName=head}}, {targetBone {boneName=spine_05}}]
    constraintWeights [1,1]
   }
   n9 = "Constraint - Dyn_Hair5 (PPPP)"(ComponentPose=@n8, bAlphaBoolEnabled=True, ConstraintWeights_0=0.9, ConstraintWeights_1=0.1, ConstraintWeights_2=0, ConstraintWeights_3=0) {
    alphaInputType=Bool
    alphaBoolBlend {blendOption=Linear}
    boneToModify {boneName=Dyn_Hair5}
    constraintSetup [{targetBone {boneName=spine_05}}, {targetBone {boneName=neck_01}}, {targetBone {boneName=clavicle_l}}, {targetBone {boneName=clavicle_r}}]
    constraintWeights [1,1,1,1]
   }
   n10 = "Constraint - Dyn_Hair2 (PP)"!(ComponentPose=@n9, bAlphaBoolEnabled:bool=True, ConstraintWeights_0=0.6, ConstraintWeights_1:float=0.4) -> (Pose) {
    alphaInputType=Bool
    alphaBoolBlend {blendOption=Linear}
    boneToModify {boneName=Dyn_Hair2}
    constraintSetup [{targetBone {boneName=spine_05}}, {targetBone {boneName=spine_04}}]
    constraintWeights [1,1]
   }
   n11 = Get bRigidBodyEnabled()
   Output Pose(Result=Blend Poses by bool(
     BlendPose_0=Component To Local(ComponentPose=@n10)
     BlendPose_1=Blend Poses by bool(BlendPose_0=@n7, BlendPose_1=@n3, bActiveValue=@n1) {blendType=Linear}
     bActiveValue=NOT Boolean(A=@n11)
    ) {blendType=Linear}) {layerGroup=None}
  ]}
 EventGraph {nodes [
   Event BlueprintUpdateAnimation()
   > Set Head Up(Head Up:double=GetAlphaFromReferenceTransformDelta(SkeletalMeshComponent=Get Skel Mesh(), BoneName=head, BaseName=neck_02, InputRange=(X=-5,Y=-20), TwistAxis="0,0,1")) -> (Output_Get:double)
   Event BlueprintInitializeAnimation()
   > Set Skel Mesh(Skel Mesh=GetOwnerSkeletalMeshComponent())
   Event OnCharacterCustomizationLoaded()
   > Do Once()
   n1 = CheckForBackbling()
   > Set Backbling On(Backbling On:bool=OR Boolean(A=@n1, B=@n1.bIsBackblingLong)) -> (Output_Get:bool)
   n2 = Branch()
   > Set BackBling On Set(BackBling On Set:double=1) -> (Output_Get:double)
   > Set BackBling Off Set(BackBling Off Set:double=0) -> (Output_Get:double)
   Set BackBling Off Set(execute=@n2.else, BackBling Off Set=1)
   > Set BackBling On Set(BackBling On Set=0)
   CheckForBackbling()
   Sequence()
   Is Valid()
   Get Pawn Owner() -> (Pawn Owner:Player Pawn Athena Object Reference)
   Set TurnOffOGHair(TurnOffOGHair:double=SelectFloat(A=0, B=1, bPickA=false)) -> (Output_Get:double)
  ]}
]
variables [
 BackBling Off Set {type=number}
 Backbling On {type=boolean}
 BackBling On Set {type=number}
 Hair {type=Vector}
 Head Up {type=number}
 Pawn Owner {type=PlayerPawn_Athena_C}
 Raw Output {type=number}
 Skel Mesh {type=SkeletalMeshComponent}
 TurnOffOGHair {type=number}
]
